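# Niagara eye-glow effect for the Dark Elf (female medium) Enchanting cosmetic, rendering two persistent sprites attached to the skeletal mesh, one per spawn group.
# Each sprite follows its socket with a small camera offset, while color and position offset switch between gameplay and front end.
# Alpha scales with the single-axis gap between two other sampled bones or sockets, fading as they converge.
# A visibility tag is driven by the head-hidden user flag.
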NiagaraSystem /BRCosmetics/Characters/Player/Female/Medium/Bodies/F_MED_Dark_Elf/FX/NS_F_MED_Dark_Elf_Eyes_Enchanting
dependencies {usedDataInterfaces [], usedDynamicInputs [], usedModules [&SystemState]}
emitters {eyes_light_Mercury {
  emitterUpdateScript [
   &EmitterState(Life Cycle Mode=::NewEnumerator1, Loop Behavior=::NewEnumerator1, Loop Duration Mode=::NewEnumerator1, UseLoopDelay=false)
   &SpawnBurst_Instantaneous(Spawn Count=2, Spawn Probability=HlslExpression() {hlslExpression=1.0}, Use Spawn Probability=false, UseLoopCountLimit=false)
  ]
  particleSpawnScript [
   &FN_SetSpawnGroup(Number of Spawn Groups=2)
   &InitializeParticle(
    Color=&LerpLinearColors(
     StartColor=<rgba 0.9911 1 0.3712 0.5>
     EndColor=<rgba 2.9733 3 1.1137 2>
     LerpFactor=&FN_FrontEndFloatModifier(Gameplay=0, Front End=1)
    )
    Sprite Size Mode=::NewEnumerator3
    Uniform Sprite Size=8
    UsePositionOffset=false
    Write Material Random=true
   )
  ]
  particleUpdateScript [
   &ParticleState(
    Kill Particles When Lifetime Has Elapsed=false
    Let Infinitely Lived Particles Die When Emitter Deactivates=true
    Loop Particles Lifetime=true
   )
   &SampleSkeletalMesh(
    AdditionalYawPitchRoll=false
    Bone / Socket Index=2
    Bone Sampling Mode:ENiagaraBoneSamplingMode=::NewEnumerator4
    Position Offset Coordinate Space:ENiagaraCoordinateSpace=::Local
    Skeletal Mesh:NiagaraDataInterfaceSkeletalMesh=$User.SkeletalMesh
   )
   &SampleSkeletalMesh(
    AdditionalYawPitchRoll=false
    Bone / Socket Index=3
    Bone Sampling Mode:ENiagaraBoneSamplingMode=::NewEnumerator4
    Position Offset Coordinate Space:ENiagaraCoordinateSpace=::Local
    Skeletal Mesh:NiagaraDataInterfaceSkeletalMesh=$User.SkeletalMesh
   ) {module=SampleSkeletalMesh001}
   &ScaleColor(
    Scale Alpha=&ClampFloat(Float=&Divide_Float(
      A=&Subtract_Float(
       A=&Subtract_Float(
        A=&MakeFloatFromVector(Vector=$Particles.SampleSkeletalMesh.SampledPosition, Channel=::NewEnumerator2)
        B=&MakeFloatFromVector(Vector=$Particles.SampleSkeletalMesh001.SampledPosition, Channel=::NewEnumerator2)
       )
       B=0.45
      )
      B=0.75
     ))
    ScaleA=true
    ScaleRGB=true
   ) {module=ScaleColor001}
   &CameraOffset(Camera Offset Amount=3)
   &SkeletalMeshLocation(
    Bone / Socket Index=$Particles.SpawnGroup
    MaskSpawnGroup=false
    Mesh Sampling Type=::NewEnumerator5
    Sampled Position Offset=&Lerp_Vector(
     A=&Lerp_Vector(A=<0 0.5 -0.5>, B=<0 0.5 -0.5>, Alpha=&MakeFloatFromInt(Int=$Particles.SpawnGroup))
     B=&Lerp_Vector(A=<0 0.5 0>, B=<0 0.5 0>, Alpha=&MakeFloatFromInt(Int=$Particles.SpawnGroup))
     Alpha=&FN_FrontEndFloatModifier(Gameplay=0, Front End=1)
    )
    Skeletal Mesh=$User.SkeletalMesh
    Socket Sampling Mode=::NewEnumerator5
   )
   /SetVariables(Particles.VisibilityTag=&MakeCustomIntFromBool(Bool=$User.bisHeadHidden, True INT=1, False INT=0)) {isSetParameters=true}
  ]
  rendererClasses [&NiagaraSpriteRendererProperties]
  renderers [{rendererClass=&NiagaraSpriteRendererProperties}]
 }}
summary {
 systemName=NS_F_MED_Dark_Elf_Eyes_Enchanting
 emitters [{emitterName=eyes_light_Mercury, rendererClasses [&NiagaraSpriteRendererProperties]}]
 userVariables [
  User.bIsFrontEnd {type=NiagaraBool}
  User.bIsFrontEndPreview {type=NiagaraBool}
  User.bisHeadHidden {type=NiagaraBool}
  User.Is_it_in_FrontEnd {type=NiagaraBool}
  User.SkeletalMesh {
   type=NiagaraDataInterfaceSkeletalMesh
   defaultValue {
    struct=NiagaraExt_VariableValue_DataInterface
    value {dataInterface=&self:NiagaraDataInterfaceSkeletalMesh_0}
   }
  }
 ]
}
