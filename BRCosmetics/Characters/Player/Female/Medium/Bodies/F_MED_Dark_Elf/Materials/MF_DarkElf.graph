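# Material function for the Dark Elf cosmetic body that adds animated panning glow and vein pulse effects to incoming material attributes.
# Triplanar noise projected from pre-skinned position and normal blends two panning layers, shaped by invertible Fresnel and tinted by a color or the incoming emissive.
# A third panning layer drives vein brightness.
# SkinFX mask channels limit the result, where base color defaults to black.
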MaterialFunction /BRCosmetics/Characters/Player/Female/Medium/Bodies/F_MED_Dark_Elf/Materials/MF_DarkElf
graph {nodes [
  n1 = GetMaterialAttributes(None=FunctionInput() {inputType=FunctionInput_MaterialAttributes}) {attributeGetTypes [#1, #2]}
  n2 = Multiply(A=ScalarParameter() {defaultValue=1, group=PanningEmissive, parameterName=Emissive Tiling}, B=VertexInterpolator(VS=PreSkinnedPosition()).PS)
  n3 = Divide(A=@n2, B=ScalarParameter() {defaultValue=40, group=PanningEmissive, parameterName=Triplanner_ConstantScaleDivisor})
  n4 = Time()
  n5 = Frac(None=Multiply(A=@n4, B=VectorParameter() {group=PanningEmissive, parameterName=TriPlanner Speed2, defaultValue=<rgba 0 0.2 0.2 1>}.RGB))
  n6 = TextureObjectParameter() {
   group=PanningEmissive
   parameterName=EmissiveGradient2
   samplerType=SAMPLERTYPE_Masks
   texture=/Game/Global/Textures/Default/Masks_and_Gradients/T_VertGradient_MASK
  }
  n7 = TextureSample(UVs=ComponentMask(None=Add(A=@n3, B=@n5)) {b=true, g=true}, Tex=@n6) {
   samplerSource=SSM_Wrap_WorldGroupSettings
   samplerType=SAMPLERTYPE_Masks
   texture=/Game/Weapons/WeaponSkins/Textures/T_Swirly_Noise_highContrast
  }
  n8 = Frac(None=Multiply(A=@n4, B=VectorParameter() {group=PanningEmissive, parameterName=TriPlanner Speed, defaultValue=<rgba 0 0.2 0.2 1>}.RGB))
  n9 = TextureObjectParameter() {
   group=PanningEmissive
   parameterName=EmissiveGradient
   samplerType=SAMPLERTYPE_Masks
   texture=&T_VertGradient_MASK
  }
  n10 = TextureSample(UVs=ComponentMask(None=Add(A=@n3, B=@n8)) {b=true, g=true}, Tex=@n9) {
   samplerSource=SSM_Wrap_WorldGroupSettings
   samplerType=SAMPLERTYPE_Masks
   texture=&T_Swirly_Noise_highContrast
  }
  n11 = LinearInterpolate(A=@n7.RGB, B=@n10.RGB)
  n12 = TextureSample(UVs=ComponentMask(None=Add(A=@n3, B=@n5)) {b=true, r=true}, Tex=@n6) {
   samplerSource=SSM_Wrap_WorldGroupSettings
   samplerType=SAMPLERTYPE_Masks
   texture=&T_Swirly_Noise_highContrast
  }
  n13 = TextureSample(UVs=ComponentMask(None=Add(A=@n3, B=@n8)) {b=true, r=true}, Tex=@n9) {
   samplerSource=SSM_Wrap_WorldGroupSettings
   samplerType=SAMPLERTYPE_Masks
   texture=&T_Swirly_Noise_highContrast
  }
  n14 = LinearInterpolate(A=@n12.RGB, B=@n13.RGB)
  n15 = VertexInterpolator(VS=PreSkinnedNormal())
  n16 = Abs(None=ComponentMask(None=@n15.PS) {g=true})
  n17 = ScalarParameter() {defaultValue=1, group=PanningEmissive, parameterName=TriPlanner_AxisBlendStrength}
  n18 = Saturate(None=ComponentMask(None=MaterialFunctionCall(In=@n16, Contrast=@n17) {materialFunction=&CheapContrast}.Result) {g=true})
  n19 = TextureSample(UVs=ComponentMask(None=Add(A=@n3, B=@n5)) {g=true, r=true}, Tex=@n6) {
   samplerSource=SSM_Wrap_WorldGroupSettings
   samplerType=SAMPLERTYPE_Masks
   texture=&T_Swirly_Noise_highContrast
  }
  n20 = TextureSample(UVs=ComponentMask(None=Add(A=@n3, B=@n8)) {g=true, r=true}, Tex=@n9) {
   samplerSource=SSM_Wrap_WorldGroupSettings
   samplerType=SAMPLERTYPE_Masks
   texture=&T_Swirly_Noise_highContrast
  }
  n21 = Abs(None=ComponentMask(None=@n15.PS) {b=true})
  n22 = Saturate(None=ComponentMask(None=MaterialFunctionCall(In=@n21, Contrast=@n17) {materialFunction=&CheapContrast}.Result) {b=true})
  n23 = LinearInterpolate(
   A=LinearInterpolate(A=LinearInterpolate(A=@n11, B=@n14, Alpha=@n18), B=LinearInterpolate(A=@n19.RGB, B=@n20.RGB), Alpha=@n22)
   B=@n14
   Alpha=ScalarParameter() {group=PanningEmissive, parameterName=Triplanner_Z axis Only}
  )
  n24 = Saturate(None=Fresnel(ExponentIn=ScalarParameter() {defaultValue=4, group=PanningEmissive, parameterName=PanningEmissive_Fresnel Exponent}) {baseReflectFraction=0})
  n25 = Multiply(
   A=LinearInterpolate(
    A=ScalarParameter() {defaultValue=0.1, group=PanningEmissive, parameterName=Emissive_BrightnessMIn}
    B=ScalarParameter() {defaultValue=2, group=PanningEmissive, parameterName=Emissive_BrightnessMax}
    Alpha=MaterialFunctionCall(Float3=@n23) {materialFunction=&BreakOutFloat3Components}.R
   )
   B=LinearInterpolate(A=OneMinus(None=@n24), B=@n24, Alpha=ScalarParameter() {group=PanningEmissive, parameterName=Panning Emissive_Invert Fresnel})
  )
  n26 = LinearInterpolate(
   A=VectorParameter() {group=PanningEmissive, parameterName=EmissiveColor, defaultValue=<rgba 1 1 1 1>}.RGB
   B=@n1.EmissiveColor
   Alpha=ScalarParameter() {group=PanningEmissive, parameterName='Emissive' Texture Amount}
  )
  n27 = ScalarParameter() {group=PanningEmissive, parameterName=Emissive_BrightnessMIn3}
  n28 = ScalarParameter() {defaultValue=2, group=PanningEmissive, parameterName=Emissive_BrightnessMax3}
  n29 = Divide(A=@n2, B=ScalarParameter() {defaultValue=100, group=PanningEmissive, parameterName=Triplanner_ConstantScaleDivisor3})
  n30 = Frac(None=Multiply(A=@n4, B=VectorParameter() {group=PanningEmissive, parameterName=TriPlanner Speed3, defaultValue=<rgba 0 0.2 0.2 1>}.RGB))
  n31 = TextureObjectParameter() {
   group=PanningEmissive
   parameterName=EmissiveGradient3
   samplerType=SAMPLERTYPE_Masks
   texture=&T_VertGradient_MASK
  }
  n32 = TextureSample(UVs=ComponentMask(None=Add(A=@n29, B=@n30)) {b=true, g=true}, Tex=@n31) {
   samplerSource=SSM_Wrap_WorldGroupSettings
   samplerType=SAMPLERTYPE_Masks
   texture=&T_Swirly_Noise_highContrast
  }
  n33 = TextureSample(UVs=ComponentMask(None=Add(A=@n29, B=@n30)) {b=true, r=true}, Tex=@n31) {
   samplerSource=SSM_Wrap_WorldGroupSettings
   samplerType=SAMPLERTYPE_Masks
   texture=&T_Swirly_Noise_highContrast
  }
  n34 = TextureSample(UVs=ComponentMask(None=Add(A=@n29, B=@n30)) {g=true, r=true}, Tex=@n31) {
   samplerSource=SSM_Wrap_WorldGroupSettings
   samplerType=SAMPLERTYPE_Masks
   texture=&T_Swirly_Noise_highContrast
  }
  n35 = MaterialFunctionCall(Float3=LinearInterpolate(A=LinearInterpolate(A=@n32.RGB, B=@n33.RGB, Alpha=@n18), B=@n34.RGB, Alpha=@n22)) {materialFunction=&BreakOutFloat3Components}
  n36 = TextureSampleParameter2D() {parameterName=SkinFX_Mask, texture=/BRCosmetics/Characters/Player/Female/Medium/Bodies/F_MED_Dark_Elf/Textures/T_Dark_Elf_Body_FX}
  n37 = ChannelMaskParameter(None=@n36.RGB) {maskChannel=Green, parameterName=VeinPulse Channel, defaultValue=<rgba 0 1 0 0>}
  n38 = Add(A=Add(A=@n1.EmissiveColor, B=Multiply(A=@n25, B=@n26)), B=LinearInterpolate(B=LinearInterpolate(A=@n27, B=@n28, Alpha=@n35.R), Alpha=@n37.R))
  n39 = LinearInterpolate(
   A=Constant()
   B=@n1.BaseColor
   Alpha=ScalarParameter() {group=PanningEmissive, parameterName="Panning Emissive_Use Diffuse Map as Base Color (Otherwise Black)"}
  )
  n40 = ChannelMaskParameter(None=@n36.RGB) {parameterName=PannginEmissive Channel}
  FunctionOutput(None=BlendMaterialAttributes(
    A=@n1.MaterialAttributes
    B=SetMaterialAttributes!(MaterialAttributes=@n1.MaterialAttributes, EmissiveColor=@n38, BaseColor=@n39) {attributeSetTypes [#2, #1]}
    Alpha=Add(A=@n40.R, B=@n37.R)
   ))
 ]}
