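# Master character material for the female medium Dark Elf body, used on skeletal meshes, with shading model chosen by expression and a hero skin subsurface profile.
# Shared PreFX and PostFX functions bracket a material attributes chain that runs through CharacterSkinFX.
# A static switch optionally inserts MF_DarkElf for custom panning emissive.
# The Mercury skin's panning function follows, with a selectable SkinFX_Mask channel (ThinFilm_Channel) as alpha.
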Material /BRCosmetics/Characters/Player/Female/Medium/Bodies/F_MED_Dark_Elf/Materials/F_MED_DarkElf_Master
graph {
 properties {
  bUseMaterialAttributes=true
  bUsedWithMeshDeformer=true
  bUsedWithMorphTargets=true
  bUsedWithSkeletalMesh=true
  pixelDepthOffsetMode=PDOM_Legacy
  refractionMethod=RM_IndexOfRefraction
  shadingModel=MSM_FromMaterialExpression
  subsurfaceProfile=&SS_HeroSkin_02
  usedShadingModels=Default Lit | Subsurface | Clear Coat | Subsurface Profile | Cloth | Eye
 }
 nodes [
  n1 = MaterialFunctionCall() {materialFunction=&BaseFN_CharacterFunction_PreFX}
  n2 = TextureSample(Tex=TextureObjectParameter() {group=Skin FX, parameterName=SkinFX_Mask})
  n3 = MaterialFunctionCall(
   PreskinnedPosition=@n1.Pre skinned Local Position
   Gmap=@n1.Pre skinned Local Position
   MaterialAttributes=@n1.Pre skinned Local Position
   AO=@n1.Pre skinned Local Position
  ) {materialFunction=&CharacterSkinFX}
  n4 = MaterialFunctionCall!(In=@n3.Result) {materialFunction=&/MF_DarkElf}
  MaterialFunctionCall(
   AO=@n1.AO
   Distance Blend=@n1.AO
   Fresnel=@n1.AO
   Gmap=@n1.AO
   Pre Skinned Local Position=@n1.AO
   Material=MaterialFunctionCall!(
    Alpha=StaticComponentMaskParameter(None=@n2.RGB) {defaultR=true, group=Skin FX, parameterName=ThinFilm_Channel}
    Material Attributes=StaticSwitchParameter(True=@n4.Result, False=@n3.Result) {group=PanningEmissive, parameterName=Use Custom Panning Emissive}
   ) {materialFunction=/BRCosmetics/Characters/Player/Female/Medium/Bodies/F_MED_Dark_Elf/Skins/Mercury/Materials/MF_DarkElfPanning_Mercury}.Result
  ) {materialFunction=&BaseFN_CharacterFunction_PostFX}
 ]
}
