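# Physics asset for the Dark Elf female body's back loincloth, simulating the two cloth strips against the lower body.
# Left and right dynamic bone chains hang from master bones linked freely to the pelvis, using overlapping capsules with limited swing and twist that tightens toward the tips.
# Capsules on pelvis, thighs, calves and feet, plus spheres on a Floor_COL bone, provide collision shapes, including a large sphere approximating the ground.
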PhysicsAsset /BRCosmetics/Characters/Player/Female/Medium/Bodies/F_MED_Dark_Elf/Meshes/F_MED_Dark_Elf_ContentROT_Back_Physics
bodies {
 Floor_COL [
  {shapeType=Sphere, radius=187.1686, shapeName=floorCOL_sphere, center=<-3.4923 -2.7141 -174.8455>}
  {shapeType=Sphere, radius=23.7085, shapeName=floorCOL_sphere, center=<0 13.3771 7.8946>}
 ]
 calf_l [{shapeType=Capsule, length=42.2852, radius=7.5484, center=<-28.3031 0.6827 -0.0409>, rotation=<rot -0.2033 84.5346 -88.4397>}]
 calf_r [{shapeType=Capsule, length=42.2852, radius=7.5484, center=<28.2849 0.7039 0.7221>, rotation=<rot -0.2033 84.5346 -88.4397>}]
 dyn_bk_loincloth_l_a_00 [{shapeType=Capsule, length=3.5, radius=9.625, shapeName=dyn_bk_loincloth_l_a_01_capsule, center=<6 -6 0>, rotation=<rot 0 0.0001 5>}]
 dyn_bk_loincloth_l_a_01 [{shapeType=Capsule, length=3.875, radius=9.25, shapeName=dyn_bk_loincloth_l_a_01_capsule, center=<6 -6 0>, rotation=<rot 0 0.0001 5>}]
 dyn_bk_loincloth_l_a_02 [{
   shapeType=Capsule
   length=9.375
   radius=8.875
   shapeName=dyn_bk_loincloth_l_a_02_capsule
   center=<6 -4.937 -0.093>
   rotation=<rot 0 0.0001 5>
  }]
 dyn_bk_loincloth_l_a_03 [{shapeType=Capsule, length=4.625, radius=8.75, shapeName=dyn_bk_loincloth_l_a_03_capsule, center=<6 -5.2479 0>}]
 dyn_bk_loincloth_l_a_04 [{shapeType=Capsule, length=7.625, radius=8.25, shapeName=dyn_bk_loincloth_l_a_04_capsule, center=<6 -3.0774 1>}]
 dyn_bk_loincloth_l_a_05 [{shapeType=Capsule, length=6.875, radius=7.75, shapeName=dyn_bk_loincloth_l_a_05_capsule, center=<6 -2.1307 1>}]
 dyn_bk_loincloth_l_a_MSTR [{shapeType=Capsule, length=0.505, radius=0.505, shapeName=dyn_bk_loincloth_l_a_MSTR_capsule, rotation=<rot 0 -90 0>}]
 dyn_bk_loincloth_r_a_00 [{shapeType=Capsule, length=3.5, radius=9.625, shapeName=dyn_bk_loincloth_r_a_01_capsule, center=<6 -6 0>, rotation=<rot 0 0.0001 -5>}]
 dyn_bk_loincloth_r_a_01 [
  {shapeType=Capsule, length=3.875, radius=9.25, shapeName=dyn_bk_loincloth_r_a_01_capsule, center=<6 -6 0>, rotation=<rot 0 0.0001 -5>}
  {
   shapeType=Capsule
   length=18.5
   radius=3.25
   shapeName=dyn_bk_loincloth_r_a_01_capsule
   center=<4.5943 5.0287 -7.3522>
   rotation=<rot 85 0 5>
  }
 ]
 dyn_bk_loincloth_r_a_02 [{shapeType=Capsule, length=9.375, radius=8.875, shapeName=dyn_bk_loincloth_r_a_02_capsule, center=<6 -4.937 0.093>, rotation=<rot 0 0 -5>}]
 dyn_bk_loincloth_r_a_03 [{shapeType=Capsule, length=4.625, radius=8.75, shapeName=dyn_bk_loincloth_r_a_03_capsule, center=<6 -5.2479 0>}]
 dyn_bk_loincloth_r_a_04 [{shapeType=Capsule, length=7.625, radius=8.25, shapeName=dyn_bk_loincloth_r_a_04_capsule, center=<6 -3.0774 -1>}]
 dyn_bk_loincloth_r_a_05 [{shapeType=Capsule, length=6.125, radius=8.25, shapeName=dyn_bk_loincloth_r_a_05_capsule, center=<6 -1.9307 -1>}]
 dyn_bk_loincloth_r_a_MSTR [{shapeType=Sphere, radius=0.505, shapeName=dyn_bk_loincloth_r_a_MSTR_sphere}]
 foot_l [{
   shapeType=Capsule
   length=9.7993
   radius=6.8871
   shapeName=foot_l_capsule
   center=<-4.8441 2.8693 0.5179>
   rotation=<rot 2.1482 4.5918 -86.2747>
  }]
 foot_r [{
   shapeType=Capsule
   length=8.2993
   radius=8.3871
   shapeName=foot_r_capsule
   center=<4.8439 -2.8693 -0.5179>
   rotation=<rot 2.1482 4.5919 -86.2747>
  }]
 pelvis [
  {shapeType=Capsule, length=11.9625, radius=9.6801, center=<-3.9614 -1.2812 0>}
  {shapeType=Capsule, length=14.29, radius=7.5952, shapeName=pelvis_capsule, center=<-6.8127 -3.7131 0>}
 ]
 thigh_l [{shapeType=Capsule, length=28.205, radius=7.7783, center=<-21.7405 -0.2981 0>, rotation=<rot 90 0.0001 0.0002>}]
 thigh_r [{shapeType=Capsule, length=27.83, radius=8.1533, center=<21.7405 0.6343 0>, rotation=<rot 90 0.0001 0.0002>}]
}
constraints [
 {bone1Name=thigh_l, bone2Name=pelvis, swing1LimitDegrees=45, swing2LimitDegrees=45, twistLimitDegrees=45}
 {bone1Name=calf_l, bone2Name=thigh_l, swing1LimitDegrees=45, swing2LimitDegrees=45, twistLimitDegrees=45}
 {bone1Name=thigh_r, bone2Name=pelvis, swing1LimitDegrees=45, swing2LimitDegrees=45, twistLimitDegrees=45}
 {bone1Name=calf_r, bone2Name=thigh_r, swing1LimitDegrees=45, swing2LimitDegrees=45, twistLimitDegrees=45}
 {
  bone1Name=dyn_bk_loincloth_r_a_MSTR
  bone2Name=pelvis
  swing1LimitDegrees=45
  swing1Motion=Free
  swing2LimitDegrees=45
  swing2Motion=Free
  twistLimitDegrees=45
  twistMotion=Free
 }
 {bone1Name=dyn_bk_loincloth_r_a_00, bone2Name=dyn_bk_loincloth_r_a_MSTR, swing1LimitDegrees=45, swing2LimitDegrees=45, twistLimitDegrees=15}
 {bone1Name=dyn_bk_loincloth_r_a_05, bone2Name=dyn_bk_loincloth_r_a_04, swing1LimitDegrees=45, swing2LimitDegrees=45, twistLimitDegrees=2}
 {bone1Name=dyn_bk_loincloth_r_a_04, bone2Name=dyn_bk_loincloth_r_a_03, swing1LimitDegrees=45, swing2LimitDegrees=45, twistLimitDegrees=2}
 {bone1Name=dyn_bk_loincloth_r_a_03, bone2Name=dyn_bk_loincloth_r_a_02, swing1LimitDegrees=45, swing2LimitDegrees=45, twistLimitDegrees=2}
 {bone1Name=dyn_bk_loincloth_r_a_02, bone2Name=dyn_bk_loincloth_r_a_01, swing1LimitDegrees=45, swing2LimitDegrees=45, twistLimitDegrees=15}
 {bone1Name=dyn_bk_loincloth_r_a_01, bone2Name=dyn_bk_loincloth_r_a_00, swing1LimitDegrees=45, swing2LimitDegrees=45, twistLimitDegrees=15}
 {bone1Name=dyn_bk_loincloth_l_a_05, bone2Name=dyn_bk_loincloth_l_a_04, swing1LimitDegrees=45, swing2LimitDegrees=45, twistLimitDegrees=2}
 {
  bone1Name=dyn_bk_loincloth_l_a_MSTR
  bone2Name=pelvis
  swing1LimitDegrees=45
  swing1Motion=Free
  swing2LimitDegrees=45
  swing2Motion=Free
  twistLimitDegrees=45
  twistMotion=Free
 }
 {bone1Name=dyn_bk_loincloth_l_a_00, bone2Name=dyn_bk_loincloth_l_a_MSTR, swing1LimitDegrees=45, swing2LimitDegrees=45, twistLimitDegrees=15}
 {bone1Name=dyn_bk_loincloth_l_a_01, bone2Name=dyn_bk_loincloth_l_a_00, swing1LimitDegrees=45, swing2LimitDegrees=45, twistLimitDegrees=15}
 {bone1Name=dyn_bk_loincloth_l_a_02, bone2Name=dyn_bk_loincloth_l_a_01, swing1LimitDegrees=45, swing2LimitDegrees=45, twistLimitDegrees=15}
 {bone1Name=dyn_bk_loincloth_l_a_03, bone2Name=dyn_bk_loincloth_l_a_02, swing1LimitDegrees=55, swing2LimitDegrees=55, twistLimitDegrees=2}
 {bone1Name=dyn_bk_loincloth_l_a_04, bone2Name=dyn_bk_loincloth_l_a_03, swing1LimitDegrees=45, swing2LimitDegrees=45, twistLimitDegrees=2}
 {
  bone1Name=foot_l
  bone2Name=calf_l
  swing1LimitDegrees=45
  swing1Motion=Free
  swing2LimitDegrees=45
  swing2Motion=Free
  twistLimitDegrees=45
  twistMotion=Free
 }
 {
  bone1Name=foot_r
  bone2Name=calf_r
  swing1LimitDegrees=45
  swing1Motion=Free
  swing2LimitDegrees=45
  swing2Motion=Free
  twistLimitDegrees=45
  twistMotion=Free
 }
]
