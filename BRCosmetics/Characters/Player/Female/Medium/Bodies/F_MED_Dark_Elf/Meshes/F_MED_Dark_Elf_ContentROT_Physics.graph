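# Physics asset for the female medium Dark Elf body, setting up front loincloth physics on the player cosmetic mesh.
# Left and right five-segment capsule chains hang from MSTR bodies with free angular joints to the pelvis, using limits with tighter twist along each chain.
# A free-angular constraint ties the two chains at their third segments.
# Pelvis, thigh and calf capsules form the lower body, and an unconstrained oversized Floor_COL sphere approximates the ground.
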PhysicsAsset /BRCosmetics/Characters/Player/Female/Medium/Bodies/F_MED_Dark_Elf/Meshes/F_MED_Dark_Elf_ContentROT_Physics
bodies {
 Floor_COL [{shapeType=Sphere, radius=187.1686, shapeName=floorCOL_sphere, center=<-3.4923 -26.2818 -180.6054>}]
 calf_l [{shapeType=Capsule, length=39.875, radius=7.625, center=<-26.9522 -4 0.7116>, rotation=<rot -90 180 179.1261>}]
 calf_r [{shapeType=Capsule, length=39.875, radius=7.625, center=<26.9522 4 0.7116>, rotation=<rot -90 180 179.1261>}]
 dyn_fr_loincloth_l_01 [{shapeType=Capsule, length=8, radius=4, shapeName=dyn_fr_loincloth_l_03_capsule, center=<5 -2 2>, rotation=<rot 90 0 0>}]
 dyn_fr_loincloth_l_02 [{shapeType=Capsule, length=8, radius=4, shapeName=dyn_fr_loincloth_l_03_capsule, center=<5 -2 2>, rotation=<rot 85 179.9998 -180>}]
 dyn_fr_loincloth_l_03 [{shapeType=Capsule, length=8, radius=4, shapeName=dyn_fr_loincloth_l_03_capsule, center=<5 -2 2>, rotation=<rot 90 0 0>}]
 dyn_fr_loincloth_l_04 [{shapeType=Capsule, length=7.625, radius=4.375, shapeName=dyn_fr_loincloth_l_03_capsule, center=<5 -2 2>, rotation=<rot 85 179.9998 -180>}]
 dyn_fr_loincloth_l_05 [{shapeType=Capsule, length=7.625, radius=4.375, shapeName=dyn_fr_loincloth_l_03_capsule, center=<5 -2 2>, rotation=<rot 90 0 0>}]
 dyn_fr_loincloth_l_MSTR [{shapeType=Capsule, length=0.505, radius=0.505, shapeName=dyn_fr_loincloth_l_MSTR_capsule, rotation=<rot 0 -90 0>}]
 dyn_fr_loincloth_r_01 [{shapeType=Capsule, length=8, radius=4, shapeName=dyn_fr_loincloth_r_03_capsule, center=<5 -2 -2>, rotation=<rot 90 0 0>}]
 dyn_fr_loincloth_r_02 [{shapeType=Capsule, length=8, radius=4, shapeName=dyn_fr_loincloth_r_03_capsule, center=<5 -2 -2>, rotation=<rot 90 0 0>}]
 dyn_fr_loincloth_r_03 [{shapeType=Capsule, length=8, radius=4, shapeName=dyn_fr_loincloth_r_03_capsule, center=<5 -2 -2>, rotation=<rot 90 0 0>}]
 dyn_fr_loincloth_r_04 [{shapeType=Capsule, length=7.625, radius=4.375, shapeName=dyn_fr_loincloth_r_03_capsule, center=<5 -2 -2>, rotation=<rot 90 0 0>}]
 dyn_fr_loincloth_r_05 [{shapeType=Capsule, length=7.625, radius=4.375, shapeName=dyn_fr_loincloth_r_03_capsule, center=<5 -2 -2>, rotation=<rot 90 0 0>}]
 dyn_fr_loincloth_r_MSTR [{shapeType=Capsule, length=0.505, radius=0.505, shapeName=dyn_fr_loincloth_r_MSTR_capsule, rotation=<rot 0 -90 0>}]
 pelvis [{shapeType=Capsule, length=41.885, radius=20.365, center=<-18.26 -16.9727 0>, rotation=<rot -90 180 178.2371>}]
 thigh_l [
  {shapeType=Capsule, length=16.125, radius=7.2, center=<-17.1818 -1.1337 0>, rotation=<rot 90 0 0>}
  {shapeType=Capsule, length=15, radius=6.25, center=<-34.6731 -1.5914 -0.0345>, rotation=<rot -89.2595 -179.9997 -170.1271>}
 ]
 thigh_r [
  {shapeType=Capsule, length=16.125, radius=7.2, center=<17.1818 1.1337 0>, rotation=<rot 90 0 0>}
  {shapeType=Capsule, length=15, radius=6.25, center=<34.6731 1.5914 -0.0345>, rotation=<rot -89.2595 -179.9997 -170.1271>}
 ]
}
constraints [
 {bone1Name=thigh_l, bone2Name=pelvis, swing1LimitDegrees=45, swing2LimitDegrees=45, twistLimitDegrees=45}
 {bone1Name=calf_l, bone2Name=thigh_l, swing1LimitDegrees=45, swing2LimitDegrees=45, twistLimitDegrees=45}
 {bone1Name=thigh_r, bone2Name=pelvis, swing1LimitDegrees=45, swing2LimitDegrees=45, twistLimitDegrees=45}
 {bone1Name=calf_r, bone2Name=thigh_r, swing1LimitDegrees=45, swing2LimitDegrees=45, twistLimitDegrees=45}
 {
  bone1Name=dyn_fr_loincloth_l_MSTR
  bone2Name=pelvis
  swing1LimitDegrees=45
  swing1Motion=Free
  swing2LimitDegrees=45
  swing2Motion=Free
  twistLimitDegrees=45
  twistMotion=Free
 }
 {bone1Name=dyn_fr_loincloth_l_01, bone2Name=dyn_fr_loincloth_l_MSTR, swing1LimitDegrees=45, swing2LimitDegrees=45, twistLimitDegrees=15}
 {
  bone1Name=dyn_fr_loincloth_r_MSTR
  bone2Name=pelvis
  swing1LimitDegrees=45
  swing1Motion=Free
  swing2LimitDegrees=45
  swing2Motion=Free
  twistLimitDegrees=45
  twistMotion=Free
 }
 {bone1Name=dyn_fr_loincloth_r_01, bone2Name=dyn_fr_loincloth_r_MSTR, swing1LimitDegrees=45, swing2LimitDegrees=45, twistLimitDegrees=15}
 {bone1Name=dyn_fr_loincloth_r_02, bone2Name=dyn_fr_loincloth_r_01, swing1LimitDegrees=45, swing2LimitDegrees=45, twistLimitDegrees=15}
 {bone1Name=dyn_fr_loincloth_l_02, bone2Name=dyn_fr_loincloth_l_01, swing1LimitDegrees=45, swing2LimitDegrees=45, twistLimitDegrees=15}
 {bone1Name=dyn_fr_loincloth_r_03, bone2Name=dyn_fr_loincloth_r_02, swing1LimitDegrees=45, swing2LimitDegrees=45, twistLimitDegrees=15}
 {bone1Name=dyn_fr_loincloth_l_03, bone2Name=dyn_fr_loincloth_l_02, swing1LimitDegrees=45, swing2LimitDegrees=45, twistLimitDegrees=15}
 {bone1Name=dyn_fr_loincloth_l_04, bone2Name=dyn_fr_loincloth_l_03, swing1LimitDegrees=45, swing2LimitDegrees=45, twistLimitDegrees=15}
 {bone1Name=dyn_fr_loincloth_r_04, bone2Name=dyn_fr_loincloth_r_03, swing1LimitDegrees=45, swing2LimitDegrees=45, twistLimitDegrees=15}
 {bone1Name=dyn_fr_loincloth_l_05, bone2Name=dyn_fr_loincloth_l_04, swing1LimitDegrees=45, swing2LimitDegrees=45, twistLimitDegrees=15}
 {bone1Name=dyn_fr_loincloth_r_05, bone2Name=dyn_fr_loincloth_r_04, swing1LimitDegrees=45, swing2LimitDegrees=45, twistLimitDegrees=15}
 {
  bone1Name=dyn_fr_loincloth_r_03
  bone2Name=dyn_fr_loincloth_l_03
  swing1LimitDegrees=45
  swing1Motion=Free
  swing2LimitDegrees=45
  swing2Motion=Free
  twistLimitDegrees=45
  twistMotion=Free
 }
]
