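# Physics asset for the FrontEnd_Back mesh of the female medium Dark Elf body, set up for dynamic loincloth bones.
# Four capsule-based chains (front and back, left and right) hang from small MSTR bodies linked to the pelvis, with swing and twist limits between segments.
# Pelvis, thigh, calf and foot capsules form the lower body with free angular joints.
# An unconstrained Floor_COL box adds a separate collision volume.
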PhysicsAsset /BRCosmetics/Characters/Player/Female/Medium/Bodies/F_MED_Dark_Elf/Meshes/F_MED_Dark_Elf_FrontEnd_Back_PhysicsAsset
bodies {
 Floor_COL [{shapeType=Box, extentX=40, extentY=23.5986, extentZ=83, shapeName=Floor_COL_box, center=<2.2975 11.0858 0>, rotation=<rot 0 -15 0>}]
 calf_l [
  {shapeType=Capsule, length=8.832, radius=6.23, shapeName=calf_l_capsule, center=<-8.1779 -3.074 -0.6334>, rotation=<rot 0 78.8361 -83.0945>}
  {shapeType=Capsule, length=23.5, radius=6.5, shapeName=calf_l_capsule, center=<-28.7333 -2.1914 0>, rotation=<rot -90 -161.0296 156.0294>}
 ]
 calf_r [
  {
   shapeType=Capsule
   length=8.6071
   radius=6.0885
   shapeName=calf_r_capsule
   center=<8.2952 3.0895 0.3231>
   rotation=<rot -1.4777 96.6581 -84.4754>
  }
  {shapeType=Capsule, length=23.25, radius=6.75, shapeName=calf_r_capsule, center=<28.3502 1.4334 0>, rotation=<rot 90 -165.9638 -155.9634>}
 ]
 dyn_bk_loincloth_l_a_00 [{shapeType=Capsule, length=1.25, radius=9.375, shapeName=dyn_bk_loincloth_l_a_01_capsule, center=<6 -6 0>, rotation=<rot 0 0.0001 5>}]
 dyn_bk_loincloth_l_a_01 [{
   shapeType=Capsule
   length=6.375
   radius=6.75
   shapeName=dyn_bk_loincloth_l_a_01_capsule
   center=<6 -4.6985 -0.1139>
   rotation=<rot 0 0.0001 5>
  }]
 dyn_bk_loincloth_l_a_02 [{
   shapeType=Capsule
   length=7.125
   radius=8.625
   shapeName=dyn_bk_loincloth_l_a_02_capsule
   center=<6 -6.5257 0.046>
   rotation=<rot 0 0.0001 5>
  }]
 dyn_bk_loincloth_l_a_03 [{shapeType=Capsule, length=2.375, radius=8.5, shapeName=dyn_bk_loincloth_l_a_03_capsule, center=<6 -5.9018 0>}]
 dyn_bk_loincloth_l_a_04 [{shapeType=Capsule, length=3.875, radius=8, shapeName=dyn_bk_loincloth_l_a_04_capsule, center=<6 -4.8037 1>}]
 dyn_bk_loincloth_l_a_05 [{shapeType=Capsule, length=4.625, radius=7.5, shapeName=dyn_bk_loincloth_l_a_05_capsule, center=<6 -5.5446 1>}]
 dyn_bk_loincloth_l_a_MSTR [{shapeType=Capsule, length=0.505, radius=0.505, shapeName=dyn_bk_loincloth_l_a_MSTR_capsule, rotation=<rot 0 -90 0>}]
 dyn_bk_loincloth_r_a_00 [{shapeType=Capsule, length=1.25, radius=9.375, shapeName=dyn_bk_loincloth_r_a_01_capsule, center=<6 -6 0>, rotation=<rot 0 0.0001 -5>}]
 dyn_bk_loincloth_r_a_01 [
  {
   shapeType=Capsule
   length=4.875
   radius=6.6513
   shapeName=dyn_bk_loincloth_r_a_01_capsule
   center=<6 -6.6657 -0.0582>
   rotation=<rot 0 0.0001 -5>
  }
  {shapeType=Capsule, length=8.5, radius=3, shapeName=dyn_bk_loincloth_r_a_01_capsule, center=<7.5701 1.3377 -7.6125>, rotation=<rot 85 0 5>}
 ]
 dyn_bk_loincloth_r_a_02 [{shapeType=Capsule, length=7.125, radius=8.625, shapeName=dyn_bk_loincloth_r_a_02_capsule, center=<6 -8.1287 -0.1862>, rotation=<rot 0 0 -5>}]
 dyn_bk_loincloth_r_a_03 [{shapeType=Capsule, length=2.375, radius=8.5, shapeName=dyn_bk_loincloth_r_a_03_capsule, center=<6 -6.609 0>}]
 dyn_bk_loincloth_r_a_04 [{shapeType=Capsule, length=3.875, radius=8, shapeName=dyn_bk_loincloth_r_a_04_capsule, center=<6 -4.109 -1>}]
 dyn_bk_loincloth_r_a_05 [{shapeType=Capsule, length=3.875, radius=8, shapeName=dyn_bk_loincloth_r_a_05_capsule, center=<6 -3.5 -1>}]
 dyn_bk_loincloth_r_a_MSTR [{shapeType=Sphere, radius=0.505, shapeName=dyn_bk_loincloth_r_a_MSTR_sphere}]
 dyn_fr_loincloth_l_01 [{shapeType=Capsule, length=8, radius=4, shapeName=dyn_fr_loincloth_l_03_capsule, center=<5 -2 2>, rotation=<rot 90 0 0>}]
 dyn_fr_loincloth_l_02 [{shapeType=Capsule, length=8, radius=4, shapeName=dyn_fr_loincloth_l_03_capsule, center=<5 -2 2>, rotation=<rot 85 179.9998 -180>}]
 dyn_fr_loincloth_l_03 [{shapeType=Capsule, length=8, radius=4, shapeName=dyn_fr_loincloth_l_03_capsule, center=<5 -2 2>, rotation=<rot 90 0 0>}]
 dyn_fr_loincloth_l_04 [{shapeType=Capsule, length=7.625, radius=4.375, shapeName=dyn_fr_loincloth_l_03_capsule, center=<5 -2 2>, rotation=<rot 85 179.9998 -180>}]
 dyn_fr_loincloth_l_05 [{shapeType=Capsule, length=7.625, radius=4.375, shapeName=dyn_fr_loincloth_l_03_capsule, center=<5 -2 2>, rotation=<rot 90 0 0>}]
 dyn_fr_loincloth_l_MSTR [{shapeType=Capsule, length=0.505, radius=0.505, shapeName=dyn_fr_loincloth_l_MSTR_capsule, rotation=<rot 0 -90 0>}]
 dyn_fr_loincloth_r_01 [{shapeType=Capsule, length=8, radius=4, shapeName=dyn_fr_loincloth_r_03_capsule, center=<5 -2 -2>, rotation=<rot 90 0 0>}]
 dyn_fr_loincloth_r_02 [{shapeType=Capsule, length=8, radius=4, shapeName=dyn_fr_loincloth_r_03_capsule, center=<5 -2 -2>, rotation=<rot 90 0 0>}]
 dyn_fr_loincloth_r_03 [{shapeType=Capsule, length=8, radius=4, shapeName=dyn_fr_loincloth_r_03_capsule, center=<5 -2 -2>, rotation=<rot 90 0 0>}]
 dyn_fr_loincloth_r_04 [{shapeType=Capsule, length=7.625, radius=4.375, shapeName=dyn_fr_loincloth_r_03_capsule, center=<5 -2 -2>, rotation=<rot 90 0 0>}]
 dyn_fr_loincloth_r_05 [{shapeType=Capsule, length=7.625, radius=4.375, shapeName=dyn_fr_loincloth_r_03_capsule, center=<5 -2 -2>, rotation=<rot 90 0 0>}]
 dyn_fr_loincloth_r_MSTR [{shapeType=Capsule, length=0.505, radius=0.505, shapeName=dyn_fr_loincloth_r_MSTR_capsule, rotation=<rot 0 -90 0>}]
 foot_l [{
   shapeType=Capsule
   length=8.4243
   radius=5.6371
   shapeName=foot_l_capsule
   center=<-4.8175 5.6179 0.3488>
   rotation=<rot 2.4645 -0.4022 -86.4759>
  }]
 foot_r [{
   shapeType=Capsule
   length=8.4243
   radius=5.6371
   shapeName=foot_r_capsule
   center=<4.8338 -3.9183 -0.4534>
   rotation=<rot 2.4645 -0.4022 -86.4759>
  }]
 pelvis [{shapeType=Capsule, length=10.0463, radius=9.7553, center=<0 -2.4417 0>}]
 thigh_l [{
   shapeType=Capsule
   length=27.5805
   radius=7.3458
   shapeName=thigh_r_capsule
   center=<-24.0412 -0.4481 0.3698>
   rotation=<rot 0.3509 -88.182 -88.0093>
  }]
 thigh_r [{
   shapeType=Capsule
   length=27.0805
   radius=7.8458
   shapeName=thigh_r_capsule
   center=<24.3812 0.9373 -0.2956>
   rotation=<rot 0.3509 -88.182 -88.0093>
  }]
}
constraints [
 {
  bone1Name=thigh_l
  bone2Name=pelvis
  swing1LimitDegrees=45
  swing1Motion=Free
  swing2LimitDegrees=45
  swing2Motion=Free
  twistLimitDegrees=45
  twistMotion=Free
 }
 {
  bone1Name=thigh_r
  bone2Name=pelvis
  swing1LimitDegrees=45
  swing1Motion=Free
  swing2LimitDegrees=45
  swing2Motion=Free
  twistLimitDegrees=45
  twistMotion=Free
 }
 {
  bone1Name=calf_r
  bone2Name=thigh_r
  swing1LimitDegrees=45
  swing1Motion=Free
  swing2LimitDegrees=45
  swing2Motion=Free
  twistLimitDegrees=45
  twistMotion=Free
 }
 {
  bone1Name=calf_l
  bone2Name=thigh_l
  swing1LimitDegrees=45
  swing1Motion=Free
  swing2LimitDegrees=45
  swing2Motion=Free
  twistLimitDegrees=45
  twistMotion=Free
 }
 {
  bone1Name=foot_r
  bone2Name=calf_r
  swing1LimitDegrees=45
  swing1Motion=Free
  swing2LimitDegrees=45
  swing2Motion=Free
  twistLimitDegrees=45
  twistMotion=Free
 }
 {
  bone1Name=foot_l
  bone2Name=calf_l
  swing1LimitDegrees=45
  swing1Motion=Free
  swing2LimitDegrees=45
  swing2Motion=Free
  twistLimitDegrees=45
  twistMotion=Free
 }
 {
  bone1Name=dyn_bk_loincloth_l_a_MSTR
  bone2Name=pelvis
  swing1LimitDegrees=45
  swing1Motion=Free
  swing2LimitDegrees=45
  swing2Motion=Free
  twistLimitDegrees=45
  twistMotion=Free
 }
 {bone1Name=dyn_bk_loincloth_l_a_00, bone2Name=dyn_bk_loincloth_l_a_MSTR, swing1LimitDegrees=30, swing2LimitDegrees=30, twistLimitDegrees=5}
 {bone1Name=dyn_bk_loincloth_l_a_01, bone2Name=dyn_bk_loincloth_l_a_00, swing1LimitDegrees=30, swing2LimitDegrees=30, twistLimitDegrees=5}
 {bone1Name=dyn_bk_loincloth_l_a_02, bone2Name=dyn_bk_loincloth_l_a_01, swing1LimitDegrees=30, swing2LimitDegrees=30, twistLimitDegrees=5}
 {bone1Name=dyn_bk_loincloth_l_a_03, bone2Name=dyn_bk_loincloth_l_a_02, swing1LimitDegrees=30, swing2LimitDegrees=30, twistLimitDegrees=5}
 {bone1Name=dyn_bk_loincloth_l_a_04, bone2Name=dyn_bk_loincloth_l_a_03, swing1LimitDegrees=30, swing2LimitDegrees=30, twistLimitDegrees=5}
 {bone1Name=dyn_bk_loincloth_l_a_05, bone2Name=dyn_bk_loincloth_l_a_04, swing1LimitDegrees=30, swing2LimitDegrees=30, twistLimitDegrees=5}
 {
  bone1Name=dyn_bk_loincloth_r_a_MSTR
  bone2Name=pelvis
  swing1LimitDegrees=45
  swing1Motion=Free
  swing2LimitDegrees=45
  swing2Motion=Free
  twistLimitDegrees=45
  twistMotion=Free
 }
 {bone1Name=dyn_bk_loincloth_r_a_00, bone2Name=dyn_bk_loincloth_r_a_MSTR, swing1LimitDegrees=30, swing2LimitDegrees=30, twistLimitDegrees=0}
 {bone1Name=dyn_bk_loincloth_r_a_01, bone2Name=dyn_bk_loincloth_r_a_00, swing1LimitDegrees=30, swing2LimitDegrees=30, twistLimitDegrees=0}
 {bone1Name=dyn_bk_loincloth_r_a_02, bone2Name=dyn_bk_loincloth_r_a_01, swing1LimitDegrees=30, swing2LimitDegrees=30, twistLimitDegrees=0}
 {bone1Name=dyn_bk_loincloth_r_a_03, bone2Name=dyn_bk_loincloth_r_a_02, swing1LimitDegrees=30, swing2LimitDegrees=30, twistLimitDegrees=0}
 {bone1Name=dyn_bk_loincloth_r_a_04, bone2Name=dyn_bk_loincloth_r_a_03, swing1LimitDegrees=30, swing2LimitDegrees=30, twistLimitDegrees=0}
 {bone1Name=dyn_bk_loincloth_r_a_05, bone2Name=dyn_bk_loincloth_r_a_04, swing1LimitDegrees=30, swing2LimitDegrees=30, twistLimitDegrees=0}
 {
  bone1Name=dyn_fr_loincloth_l_MSTR
  bone2Name=pelvis
  swing1LimitDegrees=45
  swing1Motion=Free
  swing2LimitDegrees=45
  swing2Motion=Free
  twistLimitDegrees=45
  twistMotion=Free
 }
 {bone1Name=dyn_fr_loincloth_l_01, bone2Name=dyn_fr_loincloth_l_MSTR, swing1LimitDegrees=30, swing2LimitDegrees=30, twistLimitDegrees=5}
 {bone1Name=dyn_fr_loincloth_l_02, bone2Name=dyn_fr_loincloth_l_01, swing1LimitDegrees=30, swing2LimitDegrees=30, twistLimitDegrees=5}
 {bone1Name=dyn_fr_loincloth_l_03, bone2Name=dyn_fr_loincloth_l_02, swing1LimitDegrees=30, swing2LimitDegrees=30, twistLimitDegrees=5}
 {bone1Name=dyn_fr_loincloth_l_04, bone2Name=dyn_fr_loincloth_l_03, swing1LimitDegrees=30, swing2LimitDegrees=30, twistLimitDegrees=5}
 {bone1Name=dyn_fr_loincloth_l_05, bone2Name=dyn_fr_loincloth_l_04, swing1LimitDegrees=30, swing2LimitDegrees=30, twistLimitDegrees=5}
 {
  bone1Name=dyn_fr_loincloth_r_MSTR
  bone2Name=pelvis
  swing1LimitDegrees=45
  swing1Motion=Free
  swing2LimitDegrees=45
  swing2Motion=Free
  twistLimitDegrees=45
  twistMotion=Free
 }
 {bone1Name=dyn_fr_loincloth_r_01, bone2Name=dyn_fr_loincloth_r_MSTR, swing1LimitDegrees=30, swing2LimitDegrees=30, twistLimitDegrees=5}
 {bone1Name=dyn_fr_loincloth_r_02, bone2Name=dyn_fr_loincloth_r_01, swing1LimitDegrees=30, swing2LimitDegrees=30, twistLimitDegrees=5}
 {bone1Name=dyn_fr_loincloth_r_03, bone2Name=dyn_fr_loincloth_r_02, swing1LimitDegrees=30, swing2LimitDegrees=30, twistLimitDegrees=5}
 {bone1Name=dyn_fr_loincloth_r_04, bone2Name=dyn_fr_loincloth_r_03, swing1LimitDegrees=30, swing2LimitDegrees=30, twistLimitDegrees=5}
 {bone1Name=dyn_fr_loincloth_r_05, bone2Name=dyn_fr_loincloth_r_04, swing1LimitDegrees=30, swing2LimitDegrees=30, twistLimitDegrees=5}
]
