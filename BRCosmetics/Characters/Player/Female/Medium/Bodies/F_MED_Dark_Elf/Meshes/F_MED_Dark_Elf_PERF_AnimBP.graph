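# Animation Blueprint for the female medium Dark Elf body that adds deformation and secondary loincloth motion to the pose copied from its parent mesh.
# A deform Control Rig feeds weighted front-loincloth constraints, then a second Control Rig.
# Blended additive bone offsets pose the back cloth for crouching, sliding, moving and jumping before RigidBody simulation.
# AnimDynamics, LookAt and other constraint chains appear unconnected in the visible graph.
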AnimBlueprint /BRCosmetics/Characters/Player/Female/Medium/Bodies/F_MED_Dark_Elf/Meshes/F_MED_Dark_Elf_PERF_AnimBP
target_skeleton=&/F_MED_Dark_Elf_Skeleton
graphs [
 AnimGraph {nodes [
   n1 = Anim Dynamics - Bone: dyn_fr_loincloth_l_02(Alpha=1, ExternalForce="0,0,0") {
    angularDampingOverride=0.999
    angularSpringConstant=75
    bAngularSpring=true
    bGravityOverrideInSimSpace=true
    bOverrideAngularDamping=true
    simulationSpace=RootRelative
    alphaBoolBlend {blendOption=Linear}
    boundBone {boneName=dyn_fr_loincloth_l_02}
    componentAppliedLinearAccClamp=<3500 3500 3500>
    componentLinearAccScale=<0.8 0.8 0.8>
    componentLinearVelScale=<0.8 0.8 0.8>
    physicsBodyDefinitions [{
      sphereCollisionRadius=0
      boundBone {boneName=dyn_fr_loincloth_l_02}
      boxExtents=<25 15 15>
      constraintSetup {angularLimitsMax=<1 0 0>, angularLimitsMin=<-1 -35 -35>, angularTarget=<1 0 0>}
      localJointOffset=<2 0 0>
     }]
    relativeSpaceBone {boneName=pelvis}
   }
   n2 = Anim Dynamics - Bone: dyn_fr_loincloth_l_03(ComponentPose=@n1, Alpha=1, ExternalForce="0,0,0") {
    angularDampingOverride=0.999
    angularSpringConstant=75
    bAngularSpring=true
    bGravityOverrideInSimSpace=true
    bOverrideAngularDamping=true
    simulationSpace=RootRelative
    alphaBoolBlend {blendOption=Linear}
    boundBone {boneName=dyn_fr_loincloth_l_03}
    componentAppliedLinearAccClamp=<3500 3500 3500>
    componentLinearAccScale=<0.8 0.8 0.8>
    componentLinearVelScale=<0.8 0.8 0.8>
    physicsBodyDefinitions [{
      sphereCollisionRadius=0
      boundBone {boneName=dyn_fr_loincloth_l_03}
      boxExtents=<25 15 15>
      constraintSetup {angularLimitsMax=<1 0 0>, angularLimitsMin=<-1 -35 -35>, angularTarget=<1 0 0>}
      localJointOffset=<2 0 0>
     }]
    relativeSpaceBone {boneName=pelvis}
   }
   n3 = Anim Dynamics - Bone: dyn_fr_loincloth_l_04(ComponentPose=@n2, Alpha=1, ExternalForce="0,0,0") {
    angularDampingOverride=0.999
    angularSpringConstant=75
    bAngularSpring=true
    bGravityOverrideInSimSpace=true
    bOverrideAngularDamping=true
    simulationSpace=RootRelative
    alphaBoolBlend {blendOption=Linear}
    boundBone {boneName=dyn_fr_loincloth_l_04}
    componentAppliedLinearAccClamp=<3500 3500 3500>
    componentLinearAccScale=<0.8 0.8 0.8>
    componentLinearVelScale=<0.8 0.8 0.8>
    physicsBodyDefinitions [{
      sphereCollisionRadius=0
      boundBone {boneName=dyn_fr_loincloth_l_04}
      boxExtents=<25 15 15>
      constraintSetup {angularLimitsMax=<1 0 10>, angularLimitsMin=<-1 -35 -35>, angularTarget=<1 0 0>}
      localJointOffset=<2 0 0>
     }]
    relativeSpaceBone {boneName=pelvis}
   }
   n4 = Anim Dynamics - Bone: dyn_fr_loincloth_l_05(ComponentPose=@n3, Alpha=1, ExternalForce="0,0,0") {
    angularDampingOverride=0.999
    angularSpringConstant=75
    bAngularSpring=true
    bGravityOverrideInSimSpace=true
    bOverrideAngularDamping=true
    simulationSpace=RootRelative
    alphaBoolBlend {blendOption=Linear}
    boundBone {boneName=dyn_fr_loincloth_l_05}
    componentAppliedLinearAccClamp=<3500 3500 3500>
    componentLinearAccScale=<0.8 0.8 0.8>
    componentLinearVelScale=<0.8 0.8 0.8>
    physicsBodyDefinitions [{
      sphereCollisionRadius=0
      boundBone {boneName=dyn_fr_loincloth_l_05}
      boxExtents=<25 15 15>
      constraintSetup {angularLimitsMax=<1 0 10>, angularLimitsMin=<-1 -35 -35>, angularTarget=<1 0 0>}
      localJointOffset=<2 0 0>
     }]
    relativeSpaceBone {boneName=pelvis}
   }
   n5 = Anim Dynamics - Bone: dyn_fr_loincloth_r_02(ComponentPose=@n4, Alpha=1, ExternalForce="0,0,0") {
    angularDampingOverride=0.999
    angularSpringConstant=75
    bAngularSpring=true
    bGravityOverrideInSimSpace=true
    bOverrideAngularDamping=true
    simulationSpace=RootRelative
    alphaBoolBlend {blendOption=Linear}
    boundBone {boneName=dyn_fr_loincloth_r_02}
    componentAppliedLinearAccClamp=<3500 3500 3500>
    componentLinearAccScale=<0.8 0.8 0.8>
    componentLinearVelScale=<0.8 0.8 0.8>
    physicsBodyDefinitions [{
      sphereCollisionRadius=0
      boundBone {boneName=dyn_fr_loincloth_r_02}
      boxExtents=<25 15 15>
      constraintSetup {angularLimitsMax=<1 35 0>, angularLimitsMin=<-1 0 -35>, angularTarget=<1 0 0>}
      localJointOffset=<2 0 0>
     }]
    relativeSpaceBone {boneName=pelvis}
   }
   n6 = Anim Dynamics - Bone: dyn_fr_loincloth_r_03(ComponentPose=@n5, Alpha=1, ExternalForce="0,0,0") {
    angularDampingOverride=0.999
    angularSpringConstant=75
    bAngularSpring=true
    bGravityOverrideInSimSpace=true
    bOverrideAngularDamping=true
    simulationSpace=RootRelative
    alphaBoolBlend {blendOption=Linear}
    boundBone {boneName=dyn_fr_loincloth_r_03}
    componentAppliedLinearAccClamp=<3500 3500 3500>
    componentLinearAccScale=<0.8 0.8 0.8>
    componentLinearVelScale=<0.8 0.8 0.8>
    physicsBodyDefinitions [{
      sphereCollisionRadius=0
      boundBone {boneName=dyn_fr_loincloth_r_03}
      boxExtents=<25 15 15>
      constraintSetup {angularLimitsMax=<1 35 0>, angularLimitsMin=<-1 0 -35>, angularTarget=<1 0 0>}
      localJointOffset=<2 0 0>
     }]
    relativeSpaceBone {boneName=pelvis}
   }
   n7 = Anim Dynamics - Bone: dyn_fr_loincloth_r_04(ComponentPose=@n6, Alpha=1, ExternalForce="0,0,0") {
    angularDampingOverride=0.999
    angularSpringConstant=75
    bAngularSpring=true
    bGravityOverrideInSimSpace=true
    bOverrideAngularDamping=true
    simulationSpace=RootRelative
    alphaBoolBlend {blendOption=Linear}
    boundBone {boneName=dyn_fr_loincloth_r_04}
    componentAppliedLinearAccClamp=<3500 3500 3500>
    componentLinearAccScale=<0.8 0.8 0.8>
    componentLinearVelScale=<0.8 0.8 0.8>
    physicsBodyDefinitions [{
      sphereCollisionRadius=0
      boundBone {boneName=dyn_fr_loincloth_r_04}
      boxExtents=<25 15 15>
      constraintSetup {angularLimitsMax=<1 35 10>, angularLimitsMin=<-1 0 -35>, angularTarget=<1 0 0>}
      localJointOffset=<2 0 0>
     }]
    relativeSpaceBone {boneName=pelvis}
   }
   Anim Dynamics - Bone: dyn_fr_loincloth_r_05(ComponentPose=@n7, Alpha=1, ExternalForce="0,0,0") {
    angularDampingOverride=0.999
    angularSpringConstant=75
    bAngularSpring=true
    bGravityOverrideInSimSpace=true
    bOverrideAngularDamping=true
    simulationSpace=RootRelative
    alphaBoolBlend {blendOption=Linear}
    boundBone {boneName=dyn_fr_loincloth_r_05}
    componentAppliedLinearAccClamp=<3500 3500 3500>
    componentLinearAccScale=<0.8 0.8 0.8>
    componentLinearVelScale=<0.8 0.8 0.8>
    physicsBodyDefinitions [{
      sphereCollisionRadius=0
      boundBone {boneName=dyn_fr_loincloth_r_05}
      boxExtents=<25 15 15>
      constraintSetup {angularLimitsMax=<1 35 10>, angularLimitsMin=<-1 0 -35>, angularTarget=<1 0 0>}
      localJointOffset=<2 0 0>
     }]
    relativeSpaceBone {boneName=pelvis}
   }
   n8 = Anim Dynamics - Bone: dyn_bk_loincloth_r_a_01(Alpha=1, ExternalForce="0,0,0") {
    angularDampingOverride=0.999
    angularSpringConstant=75
    bAngularSpring=true
    bGravityOverrideInSimSpace=true
    bOverrideAngularDamping=true
    bUseSphericalLimits=true
    simulationSpace=RootRelative
    alphaBoolBlend {blendOption=Linear}
    boundBone {boneName=dyn_bk_loincloth_r_a_01}
    componentAppliedLinearAccClamp=<3500 3500 3500>
    componentLinearAccScale=<1.2 1.2 1.2>
    componentLinearVelScale=<1.2 1.2 1.2>
    physicsBodyDefinitions [{
      collisionType=CustomSphere
      sphereCollisionRadius=4
      boundBone {boneName=dyn_bk_loincloth_r_a_01}
      boxExtents=<25 15 15>
      constraintSetup {angularLimitsMax=<5 35 0>, angularLimitsMin=<-5 -35 -35>, angularTarget=<1 0 0>}
      localJointOffset=<3 0 0>
     }]
    relativeSpaceBone {boneName=pelvis}
    sphericalLimits [
     {limitRadius=12, drivingBone {boneName=foot_r}, sphereLocalOffset=<0 -5 0>}
     {limitRadius=10, drivingBone {boneName=calf_twist_02_r}, sphereLocalOffset=<0 0 0>}
     {limitRadius=10, drivingBone {boneName=calf_twist_01_r}, sphereLocalOffset=<0 0 0>}
     {limitRadius=12, drivingBone {boneName=foot_l}, sphereLocalOffset=<0 5 0>}
     {limitRadius=10, drivingBone {boneName=calf_twist_01_l}, sphereLocalOffset=<0 0 0>}
     {limitRadius=10, drivingBone {boneName=calf_twist_02_l}, sphereLocalOffset=<0 0 0>}
    ]
   }
   n9 = Anim Dynamics - Bone: dyn_bk_loincloth_r_a_02(ComponentPose=@n8, Alpha=1, ExternalForce="0,0,0") {
    angularDampingOverride=0.999
    angularSpringConstant=75
    bAngularSpring=true
    bGravityOverrideInSimSpace=true
    bOverrideAngularDamping=true
    bUseSphericalLimits=true
    simulationSpace=RootRelative
    alphaBoolBlend {blendOption=Linear}
    boundBone {boneName=dyn_bk_loincloth_r_a_02}
    componentAppliedLinearAccClamp=<3500 3500 3500>
    componentLinearAccScale=<1.2 1.2 1.2>
    componentLinearVelScale=<1.2 1.2 1.2>
    physicsBodyDefinitions [{
      collisionType=CustomSphere
      sphereCollisionRadius=4
      boundBone {boneName=dyn_bk_loincloth_r_a_02}
      boxExtents=<25 15 15>
      constraintSetup {angularLimitsMax=<5 35 0>, angularLimitsMin=<-5 -35 -35>, angularTarget=<1 0 0>}
      localJointOffset=<3 0 0>
     }]
    relativeSpaceBone {boneName=pelvis}
    sphericalLimits [
     {limitRadius=12, drivingBone {boneName=foot_r}, sphereLocalOffset=<0 -5 0>}
     {limitRadius=10, drivingBone {boneName=calf_twist_02_r}, sphereLocalOffset=<0 0 0>}
     {limitRadius=10, drivingBone {boneName=calf_twist_01_r}, sphereLocalOffset=<0 0 0>}
     {limitRadius=12, drivingBone {boneName=foot_l}, sphereLocalOffset=<0 5 0>}
     {limitRadius=10, drivingBone {boneName=calf_twist_01_l}, sphereLocalOffset=<0 0 0>}
     {limitRadius=10, drivingBone {boneName=calf_twist_02_l}, sphereLocalOffset=<0 0 0>}
    ]
   }
   n10 = Anim Dynamics - Bone: dyn_bk_loincloth_r_a_03(ComponentPose=@n9, Alpha=1, ExternalForce="0,0,0") {
    angularDampingOverride=0.999
    angularSpringConstant=75
    bAngularSpring=true
    bGravityOverrideInSimSpace=true
    bOverrideAngularDamping=true
    bUseSphericalLimits=true
    simulationSpace=RootRelative
    alphaBoolBlend {blendOption=Linear}
    boundBone {boneName=dyn_bk_loincloth_r_a_03}
    componentAppliedLinearAccClamp=<3500 3500 3500>
    componentLinearAccScale=<1.2 1.2 1.2>
    componentLinearVelScale=<1.2 1.2 1.2>
    physicsBodyDefinitions [{
      collisionType=CustomSphere
      sphereCollisionRadius=4
      boundBone {boneName=dyn_bk_loincloth_r_a_03}
      boxExtents=<25 15 15>
      constraintSetup {angularLimitsMax=<5 35 10>, angularLimitsMin=<-5 -35 -35>, angularTarget=<1 0 0>}
      localJointOffset=<3 0 0>
     }]
    relativeSpaceBone {boneName=pelvis}
    sphericalLimits [
     {limitRadius=12, drivingBone {boneName=foot_r}, sphereLocalOffset=<0 -5 0>}
     {limitRadius=10, drivingBone {boneName=calf_twist_02_r}, sphereLocalOffset=<0 0 0>}
     {limitRadius=10, drivingBone {boneName=calf_twist_01_r}, sphereLocalOffset=<0 0 0>}
     {limitRadius=12, drivingBone {boneName=foot_l}, sphereLocalOffset=<0 5 0>}
     {limitRadius=10, drivingBone {boneName=calf_twist_01_l}, sphereLocalOffset=<0 0 0>}
     {limitRadius=10, drivingBone {boneName=calf_twist_02_l}, sphereLocalOffset=<0 0 0>}
    ]
   }
   n11 = Anim Dynamics - Bone: dyn_bk_loincloth_r_a_04(ComponentPose=@n10, Alpha=1, ExternalForce="0,0,0") {
    angularDampingOverride=0.999
    angularSpringConstant=75
    bAngularSpring=true
    bGravityOverrideInSimSpace=true
    bOverrideAngularDamping=true
    bUseSphericalLimits=true
    simulationSpace=RootRelative
    alphaBoolBlend {blendOption=Linear}
    boundBone {boneName=dyn_bk_loincloth_r_a_04}
    componentAppliedLinearAccClamp=<3500 3500 3500>
    componentLinearAccScale=<1.2 1.2 1.2>
    componentLinearVelScale=<1.2 1.2 1.2>
    physicsBodyDefinitions [{
      collisionType=CustomSphere
      sphereCollisionRadius=4
      boundBone {boneName=dyn_bk_loincloth_r_a_04}
      boxExtents=<25 15 15>
      constraintSetup {angularLimitsMax=<5 35 10>, angularLimitsMin=<-5 -35 -35>, angularTarget=<1 0 0>}
      localJointOffset=<3 0 0>
     }]
    relativeSpaceBone {boneName=pelvis}
    sphericalLimits [
     {limitRadius=12, drivingBone {boneName=foot_r}, sphereLocalOffset=<0 -5 0>}
     {limitRadius=10, drivingBone {boneName=calf_twist_02_r}, sphereLocalOffset=<0 0 0>}
     {limitRadius=10, drivingBone {boneName=calf_twist_01_r}, sphereLocalOffset=<0 0 0>}
     {limitRadius=12, drivingBone {boneName=foot_l}, sphereLocalOffset=<0 5 0>}
     {limitRadius=10, drivingBone {boneName=calf_twist_01_l}, sphereLocalOffset=<0 0 0>}
     {limitRadius=10, drivingBone {boneName=calf_twist_02_l}, sphereLocalOffset=<0 0 0>}
    ]
   }
   n12 = Anim Dynamics - Bone: dyn_bk_loincloth_r_a_05(ComponentPose=@n11, Alpha=1, ExternalForce="0,0,0") {
    angularDampingOverride=0.999
    angularSpringConstant=75
    bAngularSpring=true
    bGravityOverrideInSimSpace=true
    bOverrideAngularDamping=true
    bUseSphericalLimits=true
    simulationSpace=RootRelative
    alphaBoolBlend {blendOption=Linear}
    boundBone {boneName=dyn_bk_loincloth_r_a_05}
    componentAppliedLinearAccClamp=<3500 3500 3500>
    componentLinearAccScale=<1.2 1.2 1.2>
    componentLinearVelScale=<1.2 1.2 1.2>
    physicsBodyDefinitions [{
      collisionType=CustomSphere
      sphereCollisionRadius=4
      boundBone {boneName=dyn_bk_loincloth_r_a_05}
      boxExtents=<25 15 15>
      constraintSetup {angularLimitsMax=<5 35 10>, angularLimitsMin=<-5 -35 -35>, angularTarget=<1 0 0>}
      localJointOffset=<3 0 0>
     }]
    relativeSpaceBone {boneName=pelvis}
    sphericalLimits [
     {limitRadius=12, drivingBone {boneName=foot_r}, sphereLocalOffset=<0 -5 0>}
     {limitRadius=10, drivingBone {boneName=calf_twist_02_r}, sphereLocalOffset=<0 0 0>}
     {limitRadius=10, drivingBone {boneName=calf_twist_01_r}, sphereLocalOffset=<0 0 0>}
     {limitRadius=12, drivingBone {boneName=foot_l}, sphereLocalOffset=<0 5 0>}
     {limitRadius=10, drivingBone {boneName=calf_twist_01_l}, sphereLocalOffset=<0 0 0>}
     {limitRadius=10, drivingBone {boneName=calf_twist_02_l}, sphereLocalOffset=<0 0 0>}
    ]
   }
   n13 = Anim Dynamics - Bone: dyn_bk_loincloth_l_a_01(ComponentPose=@n12, Alpha=1, ExternalForce="0,0,0") {
    angularDampingOverride=0.999
    angularSpringConstant=75
    bAngularSpring=true
    bGravityOverrideInSimSpace=true
    bOverrideAngularDamping=true
    bUseSphericalLimits=true
    simulationSpace=RootRelative
    alphaBoolBlend {blendOption=Linear}
    boundBone {boneName=dyn_bk_loincloth_l_a_01}
    componentAppliedLinearAccClamp=<3500 3500 3500>
    componentLinearAccScale=<1.2 1.2 1.2>
    componentLinearVelScale=<1.2 1.2 1.2>
    physicsBodyDefinitions [{
      collisionType=CustomSphere
      sphereCollisionRadius=4
      boundBone {boneName=dyn_bk_loincloth_l_a_01}
      boxExtents=<25 15 15>
      constraintSetup {angularLimitsMax=<5 35 0>, angularLimitsMin=<-5 -35 -35>, angularTarget=<1 0 0>}
      localJointOffset=<3 0 0>
     }]
    relativeSpaceBone {boneName=pelvis}
    sphericalLimits [
     {limitRadius=12, drivingBone {boneName=foot_r}, sphereLocalOffset=<0 -5 0>}
     {limitRadius=10, drivingBone {boneName=calf_twist_02_r}, sphereLocalOffset=<0 0 0>}
     {limitRadius=10, drivingBone {boneName=calf_twist_01_r}, sphereLocalOffset=<0 0 0>}
     {limitRadius=12, drivingBone {boneName=foot_l}, sphereLocalOffset=<0 5 0>}
     {limitRadius=10, drivingBone {boneName=calf_twist_01_l}, sphereLocalOffset=<0 0 0>}
     {limitRadius=10, drivingBone {boneName=calf_twist_02_l}, sphereLocalOffset=<0 0 0>}
    ]
   }
   n14 = Anim Dynamics - Bone: dyn_bk_loincloth_l_a_02(ComponentPose=@n13, Alpha=1, ExternalForce="0,0,0") {
    angularDampingOverride=0.999
    angularSpringConstant=75
    bAngularSpring=true
    bGravityOverrideInSimSpace=true
    bOverrideAngularDamping=true
    bUseSphericalLimits=true
    simulationSpace=RootRelative
    alphaBoolBlend {blendOption=Linear}
    boundBone {boneName=dyn_bk_loincloth_l_a_02}
    componentAppliedLinearAccClamp=<3500 3500 3500>
    componentLinearAccScale=<1.2 1.2 1.2>
    componentLinearVelScale=<1.2 1.2 1.2>
    physicsBodyDefinitions [{
      collisionType=CustomSphere
      sphereCollisionRadius=4
      boundBone {boneName=dyn_bk_loincloth_l_a_02}
      boxExtents=<25 15 15>
      constraintSetup {angularLimitsMax=<5 35 0>, angularLimitsMin=<-5 -35 -35>, angularTarget=<1 0 0>}
      localJointOffset=<3 0 0>
     }]
    relativeSpaceBone {boneName=pelvis}
    sphericalLimits [
     {limitRadius=12, drivingBone {boneName=foot_r}, sphereLocalOffset=<0 -5 0>}
     {limitRadius=10, drivingBone {boneName=calf_twist_02_r}, sphereLocalOffset=<0 0 0>}
     {limitRadius=10, drivingBone {boneName=calf_twist_01_r}, sphereLocalOffset=<0 0 0>}
     {limitRadius=12, drivingBone {boneName=foot_l}, sphereLocalOffset=<0 5 0>}
     {limitRadius=10, drivingBone {boneName=calf_twist_01_l}, sphereLocalOffset=<0 0 0>}
     {limitRadius=10, drivingBone {boneName=calf_twist_02_l}, sphereLocalOffset=<0 0 0>}
    ]
   }
   n15 = Anim Dynamics - Bone: dyn_bk_loincloth_l_a_03(ComponentPose=@n14, Alpha=1, ExternalForce="0,0,0") {
    angularDampingOverride=0.999
    angularSpringConstant=75
    bAngularSpring=true
    bGravityOverrideInSimSpace=true
    bOverrideAngularDamping=true
    bUseSphericalLimits=true
    simulationSpace=RootRelative
    alphaBoolBlend {blendOption=Linear}
    boundBone {boneName=dyn_bk_loincloth_l_a_03}
    componentAppliedLinearAccClamp=<3500 3500 3500>
    componentLinearAccScale=<1.2 1.2 1.2>
    componentLinearVelScale=<1.2 1.2 1.2>
    physicsBodyDefinitions [{
      collisionType=CustomSphere
      sphereCollisionRadius=4
      boundBone {boneName=dyn_bk_loincloth_l_a_03}
      boxExtents=<25 15 15>
      constraintSetup {angularLimitsMax=<5 35 10>, angularLimitsMin=<-5 -35 -35>, angularTarget=<1 0 0>}
      localJointOffset=<3 0 0>
     }]
    relativeSpaceBone {boneName=pelvis}
    sphericalLimits [
     {limitRadius=12, drivingBone {boneName=foot_r}, sphereLocalOffset=<0 -5 0>}
     {limitRadius=10, drivingBone {boneName=calf_twist_02_r}, sphereLocalOffset=<0 0 0>}
     {limitRadius=10, drivingBone {boneName=calf_twist_01_r}, sphereLocalOffset=<0 0 0>}
     {limitRadius=12, drivingBone {boneName=foot_l}, sphereLocalOffset=<0 5 0>}
     {limitRadius=10, drivingBone {boneName=calf_twist_01_l}, sphereLocalOffset=<0 0 0>}
     {limitRadius=10, drivingBone {boneName=calf_twist_02_l}, sphereLocalOffset=<0 0 0>}
    ]
   }
   n16 = Anim Dynamics - Bone: dyn_bk_loincloth_l_a_04(ComponentPose=@n15, Alpha=1, ExternalForce="0,0,0") {
    angularDampingOverride=0.999
    angularSpringConstant=75
    bAngularSpring=true
    bGravityOverrideInSimSpace=true
    bOverrideAngularDamping=true
    bUseSphericalLimits=true
    simulationSpace=RootRelative
    alphaBoolBlend {blendOption=Linear}
    boundBone {boneName=dyn_bk_loincloth_l_a_04}
    componentAppliedLinearAccClamp=<3500 3500 3500>
    componentLinearAccScale=<1.2 1.2 1.2>
    componentLinearVelScale=<1.2 1.2 1.2>
    physicsBodyDefinitions [{
      collisionType=CustomSphere
      sphereCollisionRadius=4
      boundBone {boneName=dyn_bk_loincloth_l_a_04}
      boxExtents=<25 15 15>
      constraintSetup {angularLimitsMax=<5 35 10>, angularLimitsMin=<-5 -35 -35>, angularTarget=<1 0 0>}
      localJointOffset=<3 0 0>
     }]
    relativeSpaceBone {boneName=pelvis}
    sphericalLimits [
     {limitRadius=12, drivingBone {boneName=foot_r}, sphereLocalOffset=<0 -5 0>}
     {limitRadius=10, drivingBone {boneName=calf_twist_02_r}, sphereLocalOffset=<0 0 0>}
     {limitRadius=10, drivingBone {boneName=calf_twist_01_r}, sphereLocalOffset=<0 0 0>}
     {limitRadius=12, drivingBone {boneName=foot_l}, sphereLocalOffset=<0 5 0>}
     {limitRadius=10, drivingBone {boneName=calf_twist_01_l}, sphereLocalOffset=<0 0 0>}
     {limitRadius=10, drivingBone {boneName=calf_twist_02_l}, sphereLocalOffset=<0 0 0>}
    ]
   }
   Anim Dynamics - Bone: dyn_bk_loincloth_l_a_05(ComponentPose=@n16, Alpha=1, ExternalForce="0,0,0") {
    angularDampingOverride=0.999
    angularSpringConstant=75
    bAngularSpring=true
    bGravityOverrideInSimSpace=true
    bOverrideAngularDamping=true
    bUseSphericalLimits=true
    simulationSpace=RootRelative
    alphaBoolBlend {blendOption=Linear}
    boundBone {boneName=dyn_bk_loincloth_l_a_05}
    componentAppliedLinearAccClamp=<3500 3500 3500>
    componentLinearAccScale=<1.2 1.2 1.2>
    componentLinearVelScale=<1.2 1.2 1.2>
    physicsBodyDefinitions [{
      collisionType=CustomSphere
      sphereCollisionRadius=4
      boundBone {boneName=dyn_bk_loincloth_l_a_05}
      boxExtents=<25 15 15>
      constraintSetup {angularLimitsMax=<5 35 10>, angularLimitsMin=<-5 -35 -35>, angularTarget=<1 0 0>}
      localJointOffset=<3 0 0>
     }]
    relativeSpaceBone {boneName=pelvis}
    sphericalLimits [
     {limitRadius=12, drivingBone {boneName=foot_r}, sphereLocalOffset=<0 -5 0>}
     {limitRadius=10, drivingBone {boneName=calf_twist_02_r}, sphereLocalOffset=<0 0 0>}
     {limitRadius=10, drivingBone {boneName=calf_twist_01_r}, sphereLocalOffset=<0 0 0>}
     {limitRadius=12, drivingBone {boneName=foot_l}, sphereLocalOffset=<0 5 0>}
     {limitRadius=10, drivingBone {boneName=calf_twist_01_l}, sphereLocalOffset=<0 0 0>}
     {limitRadius=10, drivingBone {boneName=calf_twist_02_l}, sphereLocalOffset=<0 0 0>}
    ]
   }
   Component To Local()
   n17 = "Constraint - dyn_bk_loincloth_l_b_MSTR (PPPP)"!(
    ComponentPose
    Alpha=1
    ConstraintWeights_0=0.3817
    ConstraintWeights_1:float=0.3357
    ConstraintWeights_2:float=0.0027
    ConstraintWeights_3:float=0.2799
   ) -> (Pose) {
    alphaBoolBlend {blendOption=Linear}
    boneToModify {boneName=dyn_bk_loincloth_l_b_MSTR}
    constraintSetup [
     {targetBone {boneName=pelvis}}
     {targetBone {boneName=thigh_twist_01_l}}
     {targetBone {boneName=thigh_l}}
     {targetBone {boneName=deform_glute_l}}
    ]
    constraintWeights [1,1,1,1]
   }
   n18 = "Constraint - dyn_bk_loincloth_l_a_MSTR (PPPPPPPP)"!(
    ComponentPose=@n17
    Alpha=1
    ConstraintWeights_0=0
    ConstraintWeights_1:float=0.5101
    ConstraintWeights_2:float=0.1392
    ConstraintWeights_3:float=0.0077
    ConstraintWeights_4:float=0.0026
    ConstraintWeights_5:float=0.0002
    ConstraintWeights_6:float=0.326
    ConstraintWeights_7:float=0.0143
   ) -> (Pose) {
    alphaBoolBlend {blendOption=Linear}
    boneToModify {boneName=dyn_bk_loincloth_l_a_MSTR}
    constraintSetup [
     {targetBone {boneName=spine_02}}
     {targetBone {boneName=pelvis}}
     {targetBone {boneName=thigh_twist_01_l}}
     {targetBone {boneName=thigh_twist_01_r}}
     {targetBone {boneName=thigh_l}}
     {targetBone {boneName=thigh_r}}
     {targetBone {boneName=deform_glute_l}}
     {targetBone {boneName=deform_glute_r}}
    ]
    constraintWeights [1,1,1,1,1,1,1,1]
   }
   n19 = "Constraint - dyn_bk_loincloth_r_b_MSTR (PPPP)"!(
    ComponentPose=@n18
    Alpha=1
    ConstraintWeights_0=0.3817
    ConstraintWeights_1:float=0.3357
    ConstraintWeights_2:float=0.0027
    ConstraintWeights_3:float=0.2799
   ) -> (Pose) {
    alphaBoolBlend {blendOption=Linear}
    boneToModify {boneName=dyn_bk_loincloth_r_b_MSTR}
    constraintSetup [
     {targetBone {boneName=pelvis}}
     {targetBone {boneName=thigh_twist_01_r}}
     {targetBone {boneName=thigh_r}}
     {targetBone {boneName=deform_glute_r}}
    ]
    constraintWeights [1,1,1,1]
   }
   "Constraint - dyn_bk_loincloth_r_a_MSTR (PPPPPPPP)"!(
    ComponentPose=@n19
    Alpha=1
    ConstraintWeights_0=0
    ConstraintWeights_1:float=0.5101
    ConstraintWeights_2:float=0.0077
    ConstraintWeights_3:float=0.1392
    ConstraintWeights_4:float=0.0002
    ConstraintWeights_5:float=0.0026
    ConstraintWeights_6:float=0.0143
    ConstraintWeights_7:float=0.326
   ) -> (Pose) {
    alphaBoolBlend {blendOption=Linear}
    boneToModify {boneName=dyn_bk_loincloth_r_a_MSTR}
    constraintSetup [
     {targetBone {boneName=spine_02}}
     {targetBone {boneName=pelvis}}
     {targetBone {boneName=thigh_twist_01_l}}
     {targetBone {boneName=thigh_twist_01_r}}
     {targetBone {boneName=thigh_l}}
     {targetBone {boneName=thigh_r}}
     {targetBone {boneName=deform_glute_l}}
     {targetBone {boneName=deform_glute_r}}
    ]
    constraintWeights [1,1,1,1,1,1,1,1]
   }
   Local To Component()
   n20 = Look At - Bone: dyn_bk_loincloth_r_a_00(Alpha=1) {
    alphaBoolBlend {blendOption=Linear}
    boneToModify {boneName=dyn_bk_loincloth_r_a_00}
    lookAtLocation=<0 10 0>
    lookAtTarget {boneReference {boneName=foot_r}}
    lookAt_Axis {axis=<1 0>}
   }
   Look At - Bone: dyn_bk_loincloth_l_a_00(ComponentPose=@n20, Alpha=1) {
    alphaBoolBlend {blendOption=Linear}
    boneToModify {boneName=dyn_bk_loincloth_l_a_00}
    lookAtLocation=<0 -10>
    lookAtTarget {boneReference {boneName=foot_l}}
    lookAt_Axis {axis=<1 0>}
   }
   "Transform (Modify) Bone - Bone: dyn_bk_loincloth_l_b_MSTR"!(ComponentPose, bAlphaBoolEnabled:bool=True, Translation="0,0,0", Rotation="0,-25,0", Scale="1,1,1") -> (Pose) {
    alphaInputType=Bool
    rotationMode=BMM_Additive
    rotationSpace=BCS_BoneSpace
    alphaBoolBlend {blendInTime=0.1, blendOption=Linear, blendOutTime=0.1}
    boneToModify {boneName=dyn_bk_loincloth_l_b_MSTR}
    rotation {yaw=-25}
   }
   n21 = "Transform (Modify) Bone - Bone: dyn_bk_loincloth_r_a_03"!(ComponentPose, bAlphaBoolEnabled:bool=True, Translation="0,0,0", Rotation="0,0,0", Scale="1,1,1") -> (Pose) {
    alphaInputType=Bool
    rotationMode=BMM_Additive
    rotationSpace=BCS_BoneSpace
    scaleMode=BMM_Additive
    scaleSpace=BCS_BoneSpace
    translationMode=BMM_Additive
    translationSpace=BCS_BoneSpace
    alphaBoolBlend {blendInTime=0.1, blendOption=Linear, blendOutTime=0.5}
    boneToModify {boneName=dyn_bk_loincloth_r_a_03}
   }
   "Transform (Modify) Bone - Bone: dyn_bk_loincloth_l_a_03"!(ComponentPose=@n21, bAlphaBoolEnabled:bool=True, Translation="0,0,0", Rotation="20,0,0", Scale="1,1,1") -> (Pose) {
    alphaInputType=Bool
    rotationMode=BMM_Additive
    rotationSpace=BCS_BoneSpace
    scaleMode=BMM_Additive
    scaleSpace=BCS_BoneSpace
    translationMode=BMM_Additive
    translationSpace=BCS_BoneSpace
    alphaBoolBlend {blendInTime=0.1, blendOption=Linear, blendOutTime=0.5}
    boneToModify {boneName=dyn_bk_loincloth_l_a_03}
    rotation {pitch=20}
   }
   "Transform (Modify) Bone - Bone: dyn_bk_loincloth_r_b_MSTR"!(ComponentPose, bAlphaBoolEnabled:bool=True, Translation="0,0,0", Rotation="0,-25,0", Scale="1,1,1") -> (Pose) {
    alphaInputType=Bool
    rotationMode=BMM_Additive
    rotationSpace=BCS_BoneSpace
    alphaBoolBlend {blendInTime=0.1, blendOption=Linear, blendOutTime=0.1}
    boneToModify {boneName=dyn_bk_loincloth_r_b_MSTR}
    rotation {yaw=-25}
   }
   RigidBody!(ComponentPose, Alpha, OverrideWorldGravity:Vector="0,0,-980", ExternalForce) -> (Pose) {
    bClampLinearTranslationLimitToRefPose=true
    bOverrideWorldGravity=true
    overridePhysicsAsset=&/F_MED_Dark_Elf_ContentROT_Back_Physics
    simulationSpace=BaseBoneSpace
    alphaBoolBlend {blendOption=Linear}
    baseBoneRef {boneName=pelvis}
    componentAppliedLinearAccClamp=<3000 3000 3000>
    componentLinearAccScale=<0.3 0.3 0.5>
    componentLinearVelScale=<0.3 0.3 0.3>
   }
   n22 = Control Rig(Source=Copy Pose From Mesh() {bUseAttachedParent=true}) {
    alphaBoolBlend {blendOption=Linear}
    controlRigAssetReference {blueprintRigClass=&/F_MED_Dark_Elf_DeformRig_Master_ControlRig.F_MED_Dark_Elf_DeformRig_Master_ControlRig_C}
   }
   n23 = "Constraint - dyn_fr_loincloth_l_MSTR (PPPPPP)"!(
    ComponentPose=Local To Component(LocalPose=@n22)
    Alpha=1
    ConstraintWeights_0=0.2964
    ConstraintWeights_1:float=0.3879
    ConstraintWeights_2:float=0.0182
    ConstraintWeights_3:float=0.0515
    ConstraintWeights_4:float=0.2456
    ConstraintWeights_5:float=0.0003
   ) -> (Pose) {
    alphaBoolBlend {blendOption=Linear}
    boneToModify {boneName=dyn_fr_loincloth_l_MSTR}
    constraintSetup [
     {targetBone {boneName=pelvis}}
     {targetBone {boneName=thigh_twist_01_l}}
     {targetBone {boneName=thigh_twist_01_r}}
     {targetBone {boneName=thigh_l}}
     {targetBone {boneName=deform_groin_l}}
     {targetBone {boneName=deform_groin_r}}
    ]
    constraintWeights [1,1,1,1,1,1]
   }
   n24 = "Constraint - dyn_fr_loincloth_r_MSTR (PPPPPP)"!(
    ComponentPose=@n23
    Alpha=1
    ConstraintWeights_0=0.2964
    ConstraintWeights_1:float=0.0182
    ConstraintWeights_2:float=0.3879
    ConstraintWeights_3:float=0.0515
    ConstraintWeights_4:float=0.0003
    ConstraintWeights_5:float=0.2456
   ) -> (Pose) {
    alphaBoolBlend {blendOption=Linear}
    boneToModify {boneName=dyn_fr_loincloth_r_MSTR}
    constraintSetup [
     {targetBone {boneName=pelvis}}
     {targetBone {boneName=thigh_twist_01_l}}
     {targetBone {boneName=thigh_twist_01_r}}
     {targetBone {boneName=thigh_r}}
     {targetBone {boneName=deform_groin_l}}
     {targetBone {boneName=deform_groin_r}}
    ]
    constraintWeights [1,1,1,1,1,1]
   }
   n25 = Control Rig(Source=Component To Local(ComponentPose=@n24)) {
    alphaBoolBlend {blendOption=Linear}
    controlRigAssetReference {blueprintRigClass=&/F_MED_Dark_Elf_ContentROT_CtrlRig.F_MED_Dark_Elf_ContentROT_CtrlRig_C}
   }
   n26 = Get still_crouch() -> (still_crouch:bool)
   n27 = "Transform (Modify) Bone - Bone: dyn_bk_loincloth_r_a_MSTR"!(ComponentPose=Local To Component(LocalPose=@n25), bAlphaBoolEnabled:bool=@n26, Translation="0,0,0", Rotation="2,-15,0", Scale="1,1,1") -> (Pose) {
    alphaInputType=Bool
    rotationMode=BMM_Additive
    rotationSpace=BCS_BoneSpace
    scaleMode=BMM_Additive
    scaleSpace=BCS_BoneSpace
    translationMode=BMM_Additive
    translationSpace=BCS_BoneSpace
    alphaBoolBlend {blendInTime=0.1, blendOption=Linear, blendOutTime=0.5}
    boneToModify {boneName=dyn_bk_loincloth_r_a_MSTR}
    rotation {pitch=2, yaw=-15}
   }
   n28 = "Transform (Modify) Bone - Bone: dyn_bk_loincloth_l_a_MSTR"!(ComponentPose=@n27, bAlphaBoolEnabled:bool=@n26, Translation="0,0,0", Rotation="0,-5,15", Scale="1,1,1") -> (Pose) {
    alphaInputType=Bool
    rotationMode=BMM_Additive
    rotationSpace=BCS_BoneSpace
    scaleMode=BMM_Additive
    scaleSpace=BCS_BoneSpace
    translationMode=BMM_Additive
    translationSpace=BCS_BoneSpace
    alphaBoolBlend {blendInTime=0.1, blendOption=Linear, blendOutTime=0.5}
    boneToModify {boneName=dyn_bk_loincloth_l_a_MSTR}
    rotation {roll=15, yaw=-5}
   }
   n29 = "Transform (Modify) Bone - Bone: dyn_bk_loincloth_r_a_01"!(ComponentPose=@n28, bAlphaBoolEnabled:bool=@n26, Translation="0,0,0", Rotation="0,-25,15", Scale="1,1,1") -> (Pose) {
    alphaInputType=Bool
    rotationMode=BMM_Additive
    rotationSpace=BCS_BoneSpace
    scaleMode=BMM_Additive
    scaleSpace=BCS_BoneSpace
    translationMode=BMM_Additive
    translationSpace=BCS_BoneSpace
    alphaBoolBlend {blendInTime=0.1, blendOption=Linear, blendOutTime=0.5}
    boneToModify {boneName=dyn_bk_loincloth_r_a_01}
    rotation {roll=15, yaw=-25}
   }
   n30 = "Transform (Modify) Bone - Bone: dyn_bk_loincloth_l_a_01"!(ComponentPose=@n29, bAlphaBoolEnabled:bool=@n26, Translation="0,0,0", Rotation="0,-10,15", Scale="1,1,1") -> (Pose) {
    alphaInputType=Bool
    rotationMode=BMM_Additive
    rotationSpace=BCS_BoneSpace
    scaleMode=BMM_Additive
    scaleSpace=BCS_BoneSpace
    translationMode=BMM_Additive
    translationSpace=BCS_BoneSpace
    alphaBoolBlend {blendInTime=0.1, blendOption=Linear, blendOutTime=0.5}
    boneToModify {boneName=dyn_bk_loincloth_l_a_01}
    rotation {roll=15, yaw=-10}
   }
   n31 = "Transform (Modify) Bone - Bone: dyn_bk_loincloth_r_a_02"!(ComponentPose=@n30, bAlphaBoolEnabled:bool=@n26, Translation="0,0,0", Rotation="0,-20,0", Scale="1,1,1") -> (Pose) {
    alphaInputType=Bool
    rotationMode=BMM_Additive
    rotationSpace=BCS_BoneSpace
    scaleMode=BMM_Additive
    scaleSpace=BCS_BoneSpace
    translationMode=BMM_Additive
    translationSpace=BCS_BoneSpace
    alphaBoolBlend {blendInTime=0.1, blendOption=Linear, blendOutTime=0.5}
    boneToModify {boneName=dyn_bk_loincloth_r_a_02}
    rotation {yaw=-20}
   }
   n32 = "Transform (Modify) Bone - Bone: dyn_bk_loincloth_l_a_02"!(ComponentPose=@n31, bAlphaBoolEnabled:bool=@n26, Translation="0,0,0", Rotation="0,-35,0", Scale="1,1,1") -> (Pose) {
    alphaInputType=Bool
    rotationMode=BMM_Additive
    rotationSpace=BCS_BoneSpace
    scaleMode=BMM_Additive
    scaleSpace=BCS_BoneSpace
    translationMode=BMM_Additive
    translationSpace=BCS_BoneSpace
    alphaBoolBlend {blendInTime=0.1, blendOption=Linear, blendOutTime=0.5}
    boneToModify {boneName=dyn_bk_loincloth_l_a_02}
    rotation {yaw=-35}
   }
   n33 = "Transform (Modify) Bone - Bone: dyn_bk_loincloth_r_a_03"(ComponentPose=@n32, bAlphaBoolEnabled=@n26, Translation="0,0,0", Rotation="0,5,0", Scale="1,1,1") {
    alphaInputType=Bool
    rotationMode=BMM_Additive
    rotationSpace=BCS_BoneSpace
    scaleMode=BMM_Additive
    scaleSpace=BCS_BoneSpace
    translationMode=BMM_Additive
    translationSpace=BCS_BoneSpace
    alphaBoolBlend {blendInTime=0.1, blendOption=Linear, blendOutTime=0.5}
    boneToModify {boneName=dyn_bk_loincloth_r_a_03}
    rotation {yaw=5}
   }
   n34 = "Transform (Modify) Bone - Bone: dyn_bk_loincloth_l_a_03"(ComponentPose=@n33, bAlphaBoolEnabled=@n26, Translation="0,0,0", Rotation="0,-15,0", Scale="1,1,1") {
    alphaInputType=Bool
    rotationMode=BMM_Additive
    rotationSpace=BCS_BoneSpace
    scaleMode=BMM_Additive
    scaleSpace=BCS_BoneSpace
    translationMode=BMM_Additive
    translationSpace=BCS_BoneSpace
    alphaBoolBlend {blendInTime=0.1, blendOption=Linear, blendOutTime=0.5}
    boneToModify {boneName=dyn_bk_loincloth_l_a_03}
    rotation {yaw=-15}
   }
   n35 = Get bIsSliding()
   n36 = "Transform (Modify) Bone - Bone: dyn_bk_loincloth_r_a_MSTR"(ComponentPose=@n34, bAlphaBoolEnabled=@n35, Translation="4,-2,0", Rotation="0,-45,0", Scale="1,1,1") {
    alphaInputType=Bool
    rotationMode=BMM_Additive
    rotationSpace=BCS_BoneSpace
    translationMode=BMM_Additive
    translationSpace=BCS_BoneSpace
    alphaBoolBlend {blendInTime=0.5, blendOption=Linear, blendOutTime=1}
    boneToModify {boneName=dyn_bk_loincloth_r_a_MSTR}
    rotation {yaw=-45}
    translation=<4 -2>
   }
   n37 = "Transform (Modify) Bone - Bone: dyn_bk_loincloth_l_a_MSTR"(ComponentPose=@n36, bAlphaBoolEnabled=@n35, Translation="4,-2,0", Rotation="0,-45,0", Scale="1,1,1") {
    alphaInputType=Bool
    rotationMode=BMM_Additive
    rotationSpace=BCS_BoneSpace
    translationMode=BMM_Additive
    translationSpace=BCS_BoneSpace
    alphaBoolBlend {blendInTime=0.5, blendOption=Linear, blendOutTime=1}
    boneToModify {boneName=dyn_bk_loincloth_l_a_MSTR}
    rotation {yaw=-45}
    translation=<4 -2>
   }
   n38 = "Transform (Modify) Bone - Bone: Floor_COL"!(ComponentPose=@n37, bAlphaBoolEnabled:bool=@n35, Translation="4,0,-100", Rotation="0,0,0", Scale="1,1,1") -> (Pose) {
    alphaInputType=Bool
    rotationMode=BMM_Additive
    rotationSpace=BCS_BoneSpace
    translationMode=BMM_Additive
    translationSpace=BCS_BoneSpace
    alphaBoolBlend {blendInTime=0.5, blendOption=Linear, blendOutTime=1}
    boneToModify {boneName=Floor_COL}
    translation {x=4, z=-100}
   }
   n39 = Get bIsMoving2D()
   n40 = "Transform (Modify) Bone - Bone: dyn_bk_loincloth_r_a_MSTR"(ComponentPose=@n38, bAlphaBoolEnabled=@n39, Translation="0,0,0", Rotation="0,-35,0", Scale="1,1,1") {
    alphaInputType=Bool
    rotationMode=BMM_Additive
    rotationSpace=BCS_BoneSpace
    alphaBoolBlend {blendInTime=0.5, blendOption=Linear, blendOutTime=0.5}
    boneToModify {boneName=dyn_bk_loincloth_r_a_MSTR}
    rotation {yaw=-35}
   }
   n41 = "Transform (Modify) Bone - Bone: dyn_bk_loincloth_l_a_MSTR"(ComponentPose=@n40, bAlphaBoolEnabled=@n39, Translation="0,0,0", Rotation="0,-35,0", Scale="1,1,1") {
    alphaInputType=Bool
    rotationMode=BMM_Additive
    rotationSpace=BCS_BoneSpace
    alphaBoolBlend {blendInTime=0.5, blendOption=Linear, blendOutTime=0.5}
    boneToModify {boneName=dyn_bk_loincloth_l_a_MSTR}
    rotation {yaw=-35}
   }
   n42 = Get bIsMoving2D()
   n43 = "Transform (Modify) Bone - Bone: dyn_bk_loincloth_r_a_01"(ComponentPose=@n41, bAlphaBoolEnabled=@n42, Translation="0,0,0", Rotation="0,15,0", Scale="1,1,1") {
    alphaInputType=Bool
    rotationMode=BMM_Additive
    rotationSpace=BCS_BoneSpace
    alphaBoolBlend {blendInTime=0.5, blendOption=Linear, blendOutTime=0.5}
    boneToModify {boneName=dyn_bk_loincloth_r_a_01}
    rotation {yaw=15}
   }
   n44 = "Transform (Modify) Bone - Bone: dyn_bk_loincloth_l_a_01"(ComponentPose=@n43, bAlphaBoolEnabled=@n42, Translation="0,0,0", Rotation="0,15,0", Scale="1,1,1") {
    alphaInputType=Bool
    rotationMode=BMM_Additive
    rotationSpace=BCS_BoneSpace
    alphaBoolBlend {blendInTime=0.5, blendOption=Linear, blendOutTime=0.5}
    boneToModify {boneName=dyn_bk_loincloth_l_a_01}
    rotation {yaw=15}
   }
   n45 = Get bIsMoving2D()
   n46 = "Transform (Modify) Bone - Bone: dyn_bk_loincloth_r_a_02"(ComponentPose=@n44, bAlphaBoolEnabled=@n45, Translation="0,0,0", Rotation="0,-10,0", Scale="1,1,1") {
    alphaInputType=Bool
    rotationMode=BMM_Additive
    rotationSpace=BCS_BoneSpace
    alphaBoolBlend {blendInTime=0.5, blendOption=Linear, blendOutTime=0.5}
    boneToModify {boneName=dyn_bk_loincloth_r_a_02}
    rotation {yaw=-10}
   }
   n47 = "Transform (Modify) Bone - Bone: dyn_bk_loincloth_l_a_02"(ComponentPose=@n46, bAlphaBoolEnabled=@n45, Translation="0,0,0", Rotation="0,-10,0", Scale="1,1,1") {
    alphaInputType=Bool
    rotationMode=BMM_Additive
    rotationSpace=BCS_BoneSpace
    alphaBoolBlend {blendInTime=0.5, blendOption=Linear, blendOutTime=0.5}
    boneToModify {boneName=dyn_bk_loincloth_l_a_02}
    rotation {yaw=-10}
   }
   n48 = "Transform (Modify) Bone - Bone: dyn_bk_loincloth_r_a_03"(ComponentPose=@n47, bAlphaBoolEnabled=@n45, Translation="0,0,0", Rotation="0,-15,0", Scale="1,1,1") {
    alphaInputType=Bool
    rotationMode=BMM_Additive
    rotationSpace=BCS_BoneSpace
    alphaBoolBlend {blendInTime=0.5, blendOption=Linear, blendOutTime=0.5}
    boneToModify {boneName=dyn_bk_loincloth_r_a_03}
    rotation {yaw=-15}
   }
   n49 = "Transform (Modify) Bone - Bone: dyn_bk_loincloth_l_a_03"(ComponentPose=@n48, bAlphaBoolEnabled=@n45, Translation="0,0,0", Rotation="0,-15,0", Scale="1,1,1") {
    alphaInputType=Bool
    rotationMode=BMM_Additive
    rotationSpace=BCS_BoneSpace
    alphaBoolBlend {blendInTime=0.5, blendOption=Linear, blendOutTime=0.5}
    boneToModify {boneName=dyn_bk_loincloth_l_a_03}
    rotation {yaw=-15}
   }
   n50 = Get jumping() -> (jumping:double)
   n51 = "Transform (Modify) Bone - Bone: dyn_bk_loincloth_r_a_00"(ComponentPose=@n49, Alpha=@n50, Translation="0,0,0", Rotation="0,-15,0", Scale="1,1,1") {
    rotationMode=BMM_Additive
    rotationSpace=BCS_BoneSpace
    alphaBoolBlend {blendOption=Linear}
    alphaScaleBiasClamp {bInterpResult=true, interpSpeedIncreasing=5}
    boneToModify {boneName=dyn_bk_loincloth_r_a_00}
    rotation {yaw=-15}
   }
   n52 = "Transform (Modify) Bone - Bone: dyn_bk_loincloth_l_a_00"(ComponentPose=@n51, Alpha=@n50, Translation="0,0,0", Rotation="0,-15,0", Scale="1,1,1") {
    rotationMode=BMM_Additive
    rotationSpace=BCS_BoneSpace
    alphaBoolBlend {blendOption=Linear}
    alphaScaleBiasClamp {bInterpResult=true, interpSpeedIncreasing=5}
    boneToModify {boneName=dyn_bk_loincloth_l_a_00}
    rotation {yaw=-15}
   }
   n53 = RigidBody(ComponentPose=@n52, OverrideWorldGravity="0,0,-980") {
    bOverrideWorldGravity=true
    overridePhysicsAsset=&/F_MED_Dark_Elf_ContentROT_Physics
    simulationSpace=BaseBoneSpace
    alphaBoolBlend {blendOption=Linear}
    baseBoneRef {boneName=root}
    componentAppliedLinearAccClamp=<1100 1100 1100>
    componentLinearAccScale=<0.4 0.4 0.4>
    componentLinearVelScale=<0.4 0.4 0.4>
    overrideWorldGravity {z=-900}
   }
   n54 = RigidBody(ComponentPose=@n53, OverrideWorldGravity=Get Gravity Override B() -> (Gravity Override B:Vector)) {
    bClampLinearTranslationLimitToRefPose=true
    bOverrideWorldGravity=true
    overridePhysicsAsset=&F_MED_Dark_Elf_ContentROT_Back_Physics
    simulationSpace=BaseBoneSpace
    alphaBoolBlend {blendOption=Linear}
    baseBoneRef {boneName=root}
    componentAppliedLinearAccClamp=<3000 3000 3000>
    componentLinearAccScale=<0.3 0.3 0.5>
    componentLinearVelScale=<0.3 0.3 0.5>
    simSpaceSettings {
     maxAngularAcceleration=1000
     maxAngularVelocity=3
     maxLinearAcceleration=700
     maxLinearVelocity=500
     velocityScaleZ=0
     worldAlpha=1
     externalLinearDragV=<0.2 0.2>
    }
   }
   n55 = Control Rig(Source=Copy Pose From Mesh() {bUseAttachedParent=true}) {
    alphaBoolBlend {blendOption=Linear}
    controlRigAssetReference {blueprintRigClass=&F_MED_Dark_Elf_DeformRig_Master_ControlRig_C}
   }
   n56 = RigidBody(ComponentPose=Local To Component(LocalPose=@n55), OverrideWorldGravity="0,0,-980") {
    bOverrideWorldGravity=true
    overridePhysicsAsset=&/F_MED_Dark_Elf_FrontEnd_Back_PhysicsAsset
    simulationSpace=BaseBoneSpace
    alphaBoolBlend {blendOption=Linear}
    baseBoneRef {boneName=root}
    componentAppliedLinearAccClamp=<1100 1100 1100>
    componentLinearAccScale=<0.4 0.4 0.4>
    componentLinearVelScale=<0.4 0.4 0.4>
    overrideWorldGravity {z=-900}
   }
   n57 = Blend Poses by bool(BlendPose_0=Component To Local(ComponentPose=@n56), BlendPose_1=Component To Local(ComponentPose=@n54), bActiveValue=Get bIsInFrontEnd()) {blendType=Linear}
   Output Pose(Result=Component To Local(ComponentPose=Local To Component(LocalPose=@n57))) {layerGroup=None}
  ]}
 EventGraph {nodes [
   Event BlueprintInitializeAnimation()
   > Set Skel Mesh(Skel Mesh=GetPartSkeletalMeshComponent())
   n1 = Event OnBeginFalling()
   Set jumping(execute=[@n1.then, @n2.then], jumping:double=1) -> (Output_Get:double)
   n2 = Event OnBeginJumping()
   Event OnEndPlayingEmote() {is=title K2Node_Event}
   > Delay(Duration=0.1)
   > ResetDynamics(InTeleportType=TeleportPhysics)
   > ResetDynamics(InTeleportType=ResetPhysics)
   n3 = Event OnEndFalling()
   Set jumping(execute=[@n3.then, @n4.then], jumping=0)
   n4 = Event OnEndJumping()
   Event OnBeginPlayingEmote() {is=title K2Node_Event}
   > Delay(Duration=0.1)
   > ResetDynamics(InTeleportType=TeleportPhysics)
   > ResetDynamics(InTeleportType=ResetPhysics)
   Event BlueprintUpdateAnimation()
   > n5 = Sequence!(execute) -> (then_0, then_1, then_2:Exec, then_3:Exec, then_4:Exec, then_5:Exec)
   > Set crouch_alpha(crouch_alpha:double="To Float (Boolean)"(InBool=Get bIsCrouching())) -> (Output_Get:double)
   n6 = GetOwnerSkeletalMeshComponent()
   n7 = BreakRotator(InRot=BreakTransform(InTransform=GetDeltaTransformFromRefPose(self=@n6, BoneName=thigh_r, BaseName=pelvis)).Rotation)
   > Set thigh_r_rot_yaw_alpha(thigh_r_rot_yaw_alpha:double=MapRangeClamped(Value=@n7.Yaw, InRangeA=5, InRangeB=-60)) -> (Output_Get:double)
   n8 = GetOwnerSkeletalMeshComponent()
   n9 = BreakRotator(InRot=BreakTransform(InTransform=GetDeltaTransformFromRefPose(self=@n8, BoneName=thigh_l, BaseName=pelvis)).Rotation)
   > Set thigh_l_rot_yaw_alpha(thigh_l_rot_yaw_alpha:double=MapRangeClamped(Value=@n9.Yaw, InRangeA=-20, InRangeB=-60)) -> (Output_Get:double)
   n10 = GetOwnerSkeletalMeshComponent()
   n11 = BreakRotator(InRot=BreakTransform(InTransform=GetDeltaTransformFromRefPose(self=@n10, BoneName=spine_01, BaseName=pelvis)).Rotation)
   > Set spine_01_rot_alpha(spine_01_rot_alpha:double=MapRangeClamped(Value=@n11.Yaw, InRangeA=0, InRangeB=-8)) -> (Output_Get:double)
   > Set thigh_l_rot_crouch_alpha(thigh_l_rot_crouch_alpha:double=MapRangeClamped(Value=@n9.Yaw, InRangeA=-50, InRangeB=-109)) -> (Output_Get:double)
   > Set spine_01_crouch_rot_alpha(spine_01_crouch_rot_alpha:double=MapRangeClamped(Value=@n11.Yaw, InRangeA=13, InRangeB=8)) -> (Output_Get:double)
   > Set spine_01_crouch_rot_backwards_alpha(spine_01_crouch_rot_backwards_alpha:double=MapRangeClamped(Value=@n11.Yaw, InRangeA=11, InRangeB=15)) -> (Output_Get:double)
   > Branch(Condition=NOT Boolean(A=Get bIsCrouching()))
   > Set thigh_r_rot_pitch_alpha(thigh_r_rot_pitch_alpha:double=MapRangeClamped(Value=@n7, InRangeA=0, InRangeB=42)) -> (Output_Get:double)
   > Set thigh_l_rot_pitch_alpha(thigh_l_rot_pitch_alpha:double=MapRangeClamped(Value=@n9, InRangeA=6, InRangeB=42)) -> (Output_Get:double)
   > Set spine_01_crouch_rot_alpha(spine_01_crouch_rot_alpha=0)
   n12 = GetOwnerSkeletalMeshComponent()
   n13 = BreakRotator(InRot=BreakTransform(InTransform=GetDeltaTransformFromRefPose(self=@n12, BoneName=pelvis, BaseName=root)).Rotation)
   > Set pelvis_rot_alpha(pelvis_rot_alpha:double=MapRangeClamped(Value=@n13.Yaw, InRangeA=-2, InRangeB=-15)) -> (Output_Get:double)
   > Set pelvis_rot_forward_alpha(pelvis_rot_forward_alpha:double=MapRangeClamped(Value=@n13.Yaw, InRangeA=0, InRangeB=20)) -> (Output_Get:double)
   n14 = Gravity Override State Function DEPRECATED(
    execute=@n5.then_1
    CharacterPartAnimInstance=Self-Reference()
    GameTimeInSeconds=GetGameTimeInSeconds()
    GravityOverrideParams=Get Gravity Override Params()
   )
   > Set Gravity Override B(Gravity Override B:Vector=@n14) -> (Output_Get:Vector)
   Branch(execute=@n5.then_2, Condition=Get bIsInFrontEnd())
   n15 = BreakTransform(InTransform=GetDeltaTransformFromRefPose(self=GetOwnerSkeletalMeshComponent(), BoneName=pelvis, BaseName=root))
   > Set spine_01_idle_alpha(spine_01_idle_alpha:double=MapRangeClamped(Value=BreakRotator(InRot=@n15.Rotation).Pitch, InRangeA=0, InRangeB=7)) -> (Output_Get:double)
   n16 = BreakTransform(InTransform=GetDeltaTransformFromRefPose(self=GetOwnerSkeletalMeshComponent(), BoneName=thigh_r, BaseName=pelvis))
   Set rban_dis_thigh_r_const(
    execute=@n5.then_3
    rban_dis_thigh_r_const:bool="float < float"!(A:double=BreakRotator(InRot=@n16.Rotation).Yaw, B:double=-50) -> (ReturnValue:bool) {is=title K2Node_CallFunction}
   ) -> (Output_Get:bool)
   n17 = BreakTransform(InTransform=GetDeltaTransformFromRefPose(self=GetOwnerSkeletalMeshComponent(), BoneName=thigh_l, BaseName=pelvis))
   > Set rban_dis_thigh_r_const(rban_dis_thigh_r_const="float > float"(A=BreakRotator(InRot=@n17.Rotation).Yaw, B=0) {is=title K2Node_CallFunction})
   n18 = GetOwnerSkeletalMeshComponent()
   > Set congaEmote(congaEmote:bool=Montage_IsActive!(
     self:Anim Instance Object Reference=GetAnimInstance(self=@n18)
     Montage:Anim Montage Object Reference=/BRCosmetics/Animation/Game/MainPlayer/Emotes/Conga/Emote_CongaLine_CMF_M
    ) -> (ReturnValue:bool)) -> (Output_Get:bool)
   n19 = PlayerMovement(execute=@n5.then_4, AnimInstance=Self-Reference(), XDotProductCoefficient=1, YDotProductCoefficient=1, ZDotProductCoefficient=1)
   > Set FwdMovement(FwdMovement:double=MapRangeClamped(Value=@n19.YMovement, InRangeA=0, InRangeB=1)) -> (Output_Get:double)
   n20 = Get bIsMoving2D()
   n21 = Branch(execute=@n5.then_5, Condition=AND Boolean(A=Get bIsCrouching(), B=NOT Boolean(A=@n20)))
   > Set still_crouch(still_crouch:bool=true) -> (Output_Get:bool)
   Set still_crouch(execute=@n21.else, still_crouch=false)
   MapRangeClamped(Value=@n7.Yaw, InRangeA=-50, InRangeB=-103, OutRangeB=0)
   Montage_IsActive(Montage=&Emote_CongaLine_CMF_M)
   OR Boolean(A=false, B=false)
   Set skydiving(skydiving:bool=OR Boolean!(A=Get bIsSkydiveFloating(), B=Get bIsParachuteOpened(), C:bool=Get bIsSkydiveDiveMode()) -> (ReturnValue)) -> (Output_Get:bool)
  ]}
]
variables [
 congaEmote {type=boolean}
 crouch_alpha {type=number}
 FwdMovement {type=number}
 Gravity Override B {type=Vector}
 Gravity Override Params {type=GravityOverrideParamsStruct, default {skydive_Parachuting {gravityOverride=<0 0 -980>, windAmplitude=<200 -200 400>}}}
 jumping {type=number}
 pelvis_rot_alpha {type=number}
 pelvis_rot_forward_alpha {type=number}
 rban_dis_thigh_l_const {type=boolean}
 rban_dis_thigh_r_const {type=boolean}
 Skel Mesh {type=SkeletalMeshComponent}
 skydiving {type=boolean}
 spine_01_crouch_rot_alpha {type=number}
 spine_01_crouch_rot_backwards_alpha {type=number}
 spine_01_idle_alpha {type=number}
 spine_01_rot_alpha {type=number}
 still_crouch {type=boolean}
 thigh_l_rot_crouch_alpha {type=number}
 thigh_l_rot_pitch_alpha {type=number}
 thigh_l_rot_yaw_alpha {type=number}
 thigh_r_rot_crouch_alpha {type=number}
 thigh_r_rot_pitch_alpha {type=number}
 thigh_r_rot_yaw_alpha {type=number}
]
