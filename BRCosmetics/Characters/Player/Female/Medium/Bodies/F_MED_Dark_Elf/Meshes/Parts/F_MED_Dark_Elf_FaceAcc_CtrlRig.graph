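# Control Rig for the female Dark Elf face accessory that orients hair driver bones from head and upper-spine motion.
# It repositions the left aim bone, then aims four strand and back-hair drivers at spine_05-parented targets, propagating to children.
# Strand drivers also orient toward the head, the right at half weight, and the left gets two rotation adjustments afterward.
# The right strand targets `hairStrand_r_aim2`, which is absent from the hierarchy.
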ControlRigBlueprint /BRCosmetics/Characters/Player/Female/Medium/Bodies/F_MED_Dark_Elf/Meshes/Parts/F_MED_Dark_Elf_FaceAcc_CtrlRig
graphs [
 RigVMFunctionLibrary {links [], nodes []}
 RigVMModel {nodes [
   RigUnit_BeginExecution = &RigUnit_BeginExecution()
   &RigUnit_SetTransform(
    Item=(Type=Bone,Name="hairStrand_l_aim")
    Space=LocalSpace
    Value=(Rotation=(X=0,Y=0,Z=0,W=1),Translation=(X=-3.5,Y=-5.4,Z=0),Scale3D=(X=1,Y=1,Z=1))
    Weight=1
    bInitial=true
    ExecutePin=@RigUnit_BeginExecution.ExecutePin
   )
   n1 = &RigUnit_GetTransform(Item=(Type=Bone,Name="head"), bInitial=false)
   > /Script/ControlRig.RigUnit_AimBone(
    Bone=driver_hairStrand_01_l
    DebugSettings=(Scale=10,WorldOffset=(Rotation=(X=0,Y=0,Z=0,W=1),Translation=(X=0,Y=0,Z=0),Scale3D=(X=1,Y=1,Z=1)))
    Primary=(Weight=1,Axis=(X=1,Y=0,Z=0),Target=(X=1,Y=0,Z=0),Kind=Location,Space="None")
    Secondary=(Weight=1,Axis=(X=0,Y=0,Z=1),Target=(X=1,Y=0,Z=0),Kind=Location)
    Weight=1
    bPropagateToChildren=True
    Primary.Target=&RigUnit_GetTransform(Item=(Type=Bone,Name="hairStrand_l_aim"), bInitial=false).Transform.Translation
    Secondary.Target=@n1.Transform.Translation
   ) {is=struct RigVMUnitNode}
   > &RigUnit_AimBone(
    Bone=driver_hairStrand_b_01_r
    DebugSettings=(Scale=10,WorldOffset=(Rotation=(X=0,Y=0,Z=0,W=1),Translation=(X=0,Y=0,Z=0),Scale3D=(X=1,Y=1,Z=1)))
    Primary=(Weight=1,Axis=(X=1,Y=0,Z=0),Target=(X=1,Y=0,Z=0),Kind=Location,Space="None")
    Secondary=(Weight=1,Axis=(X=0,Y=0,Z=1),Target=(X=1,Y=0,Z=0),Kind=Location)
    Weight=0.5
    bPropagateToChildren=True
    Primary.Target=&RigUnit_GetBoneTransform(Bone=hairStrand_r_aim2).Transform.Translation
    Secondary.Target=@n1.Transform.Translation
   )
   > &RigUnit_AimBone(
    Bone=driver_hairBack_01_l
    DebugSettings=(Scale=10,WorldOffset=(Rotation=(X=0,Y=0,Z=0,W=1),Translation=(X=0,Y=0,Z=0),Scale3D=(X=1,Y=1,Z=1)))
    Primary=(Weight=1,Axis=(X=-1,Y=0,Z=0),Target=(X=1,Y=0,Z=0),Kind=Location,Space="None")
    Secondary=(Weight=0,Axis=(X=0,Y=0,Z=1),Target=(X=1,Y=0,Z=0),Kind=Location)
    Weight=1
    bPropagateToChildren=True
    Secondary.Target=@n1.Transform.Translation
    Primary.Target=&RigUnit_GetTransform(Item=(Type=Bone,Name="hairStrand_l_aim"), bInitial=false).Transform.Translation
   )
   > &RigUnit_AimBone(
    Bone=driver_hairBack_01_r
    DebugSettings=(Scale=10,WorldOffset=(Rotation=(X=0,Y=0,Z=0,W=1),Translation=(X=0,Y=0,Z=0),Scale3D=(X=1,Y=1,Z=1)))
    Primary=(Weight=1,Axis=(X=1,Y=0,Z=0),Target=(X=1,Y=0,Z=0),Kind=Location,Space="None")
    Secondary=(Weight=0,Axis=(X=0,Y=0,Z=1),Target=(X=1,Y=0,Z=0),Kind=Location)
    Weight=1
    bPropagateToChildren=True
    Secondary.Target=@n1.Transform.Translation
    Primary.Target=&RigUnit_GetTransform(Item=(Type=Bone,Name="hairStrand_r_aim"), bInitial=false).Transform.Translation
   )
   > &RigUnit_ModifyTransforms(
    ItemToModify=((Item=(Type=Bone,Name="driver_hairStrand_01_l"),Transform=(Rotation=(X=0,Y=0,Z=0,W=1),Translation=(X=0,Y=0,Z=0),Scale3D=(X=1,Y=1,Z=1))))
    Weight=1
    ItemToModify.0.Transform.Rotation=&RigVMFunction_MathQuaternionFromEuler(Euler=(X=180,Y=0,Z=0)).Result
   )
   > &RigUnit_ModifyTransforms(
    ItemToModify=((Item=(Type=Bone,Name="driver_hairStrand_01_l"),Transform=(Rotation=(X=0,Y=0,Z=0,W=1),Translation=(X=0,Y=0,Z=0),Scale3D=(X=1,Y=1,Z=1))))
    Weight=1
    ItemToModify.0.Transform.Rotation=&RigVMFunction_MathQuaternionFromEuler(Euler=(X=0,Y=0,Z=5)).Result
   )
  ]}
]
hierarchy [
 root {loc=<0 0 0>, rot=<0 0 0 1>}
 attach {parent=root, loc=<0 0 0>, rot=<0 0 0 1>}
 pelvis {parent=root, loc=<0 -0.5301 98.0571>, rot=<0.0127 0.707 -0.0127 -0.707>}
 spine_01 {parent=pelvis, loc=<6.1208 -0.9328 0>, rot=<0 0 0.034 -0.9994>}
 spine_02 {parent=spine_01, loc=<6.5996 -0.2095 0>, rot=<0 0 0.0238 -0.9997>}
 spine_03 {parent=spine_02, loc=<6.4914 -0.1237 0>, rot=<0 0 0.0149 -0.9999>}
 spine_04 {parent=spine_03, loc=<6.9598 0.0794 0>, rot=<0 0 -0.0132 -0.9999>}
 spine_05 {parent=spine_04, loc=<6.9987 0 0>, rot=<0 0 -0.0015 -1>}
 clavicle_l {parent=spine_05, loc=<5.4831 -0.1053 -3.2946>, rot=<0.0465 -0.7289 -0.009 -0.6829>}
 upperarm_l {parent=clavicle_l, loc=<9.8658 -1.2136 -0.4105>, rot=<0.0208 -0.486 -0.0242 -0.8734>}
 lowerarm_l {parent=upperarm_l, loc=<26.465 -0.8345 0.259>, rot=<0.0036 0.0161 -0.2298 -0.9731>}
 hand_l {parent=lowerarm_l, loc=<22.7355 -0.3203 -0.6105>, rot=<-0.6784 -0.0223 0.0735 -0.7307>}
 index_metacarpal_l {parent=hand_l, loc=<1.6535 0.0183 -2.96>, rot=<0.0158 -0.0124 -0.0173 -0.9996>}
 index_01_l {parent=index_metacarpal_l, loc=<8.3008 0.0357 -0.0524>, rot=<-0.0044 0.0485 0.0924 -0.9945>}
 index_02_l {parent=index_01_l, loc=<3.3587 -0.3004 0.0452>, rot=<-0.002 0.0027 0.3225 -0.9466>}
 index_03_l {parent=index_02_l, loc=<2.5578 -0.123 -0.0252>, rot=<-0.002 0.0067 0.2558 -0.9667>}
 middle_metacarpal_l {parent=hand_l, loc=<2.1856 -0.1725 -0.5984>, rot=<0.006 0.0168 -0.0329 -0.9993>}
 middle_01_l {parent=middle_metacarpal_l, loc=<7.3275 -0.2447 0.0129>, rot=<-0.0208 0.0042 0.0858 -0.9961>}
 middle_02_l {parent=middle_01_l, loc=<4.0123 0.1401 0.1175>, rot=<-0.0218 0.0258 0.3512 -0.9357>}
 middle_03_l {parent=middle_02_l, loc=<2.6077 0.0534 0.1373>, rot=<0.0082 0.0188 0.1708 -0.9851>}
 pinky_metacarpal_l {parent=hand_l, loc=<1.7616 -0.9753 3.3814>, rot=<-0.0397 0.0814 -0.069 -0.9935>}
 pinky_01_l {parent=pinky_metacarpal_l, loc=<7.1709 -0.4831 -0.1094>, rot=<-0.0654 -0.0035 0.2242 -0.9723>}
 pinky_02_l {parent=pinky_01_l, loc=<2.8274 0.0893 -0.1043>, rot=<0.0037 0.0011 0.2854 -0.9584>}
 pinky_03_l {parent=pinky_02_l, loc=<2.2046 0.0307 -0.0078>, rot=<-0.0002 0.0098 0.1445 -0.9895>}
 ring_metacarpal_l {parent=hand_l, loc=<2.2074 -0.3625 1.3638>, rot=<-0.0284 0.0722 -0.0458 -0.9959>}
 ring_01_l {parent=ring_metacarpal_l, loc=<7.0295 -0.6197 -0.1446>, rot=<-0.0489 -0.0252 0.1892 -0.9804>}
 ring_02_l {parent=ring_01_l, loc=<3.6174 0.4647 -0.1027>, rot=<0 0 0.2806 -0.9598>}
 ring_03_l {parent=ring_02_l, loc=<2.4152 0.114 0.0199>, rot=<0 0 0.1189 -0.9929>}
 thumb_01_l {parent=hand_l, loc=<1.3118 -0.3191 -2.459>, rot=<0.6546 -0.4042 0.0664 -0.6354>}
 thumb_02_l {parent=thumb_01_l, loc=<5.9336 -0.0448 0.0401>, rot=<-0.0037 0.0213 0.298 -0.9543>}
 thumb_03_l {parent=thumb_02_l, loc=<2.2223 0.3682 -0.0907>, rot=<0.0118 -0.013 0.0066 -0.9998>}
 weapon_l {parent=hand_l, loc=<-0.1974 -3.4111 0.3811>, rot=<0 0 0 -1>}
 lowerarm_twist_01_l {parent=lowerarm_l, loc=<18.316 0 0>, rot=<0 0 0 -1>}
 lowerarm_twist_02_l {parent=lowerarm_l, loc=<10.6707 0 0>, rot=<0 0 0 -1>}
 upperarm_twist_01_l {parent=upperarm_l, loc=<7.2135 -0.1134 0>, rot=<0 0 0 1>}
 upperarm_twist_02_l {parent=upperarm_l, loc=<19.7998 -0.1134 0>, rot=<0 0 0 1>}
 clavicle_r {parent=spine_05, loc=<5.4831 -0.1053 3.2946>, rot=<0.7289 0.0465 -0.6829 0.009>}
 upperarm_r {parent=clavicle_r, loc=<-9.8658 1.2136 0.4104>, rot=<0.0208 -0.486 -0.0242 -0.8734>}
 lowerarm_r {parent=upperarm_r, loc=<-26.4651 0.8345 -0.259>, rot=<0.0036 0.0161 -0.2298 -0.9731>}
 hand_r {parent=lowerarm_r, loc=<-22.7356 0.3203 0.6106>, rot=<-0.6784 -0.0223 0.0735 -0.7307>}
 index_metacarpal_r {parent=hand_r, loc=<-1.6534 -0.0181 2.9601>, rot=<0.0158 -0.0124 -0.0173 -0.9996>}
 index_01_r {parent=index_metacarpal_r, loc=<-8.3008 -0.0357 0.0524>, rot=<-0.0044 0.0485 0.0924 -0.9945>}
 index_02_r {parent=index_01_r, loc=<-3.3589 0.3003 -0.0453>, rot=<-0.002 0.0027 0.3225 -0.9466>}
 index_03_r {parent=index_02_r, loc=<-2.5578 0.123 0.0252>, rot=<-0.002 0.0067 0.2558 -0.9667>}
 middle_metacarpal_r {parent=hand_r, loc=<-2.1856 0.1726 0.5984>, rot=<0.006 0.0168 -0.0329 -0.9993>}
 middle_01_r {parent=middle_metacarpal_r, loc=<-7.3273 0.2447 -0.0129>, rot=<-0.0208 0.0042 0.0858 -0.9961>}
 middle_02_r {parent=middle_01_r, loc=<-4.0124 -0.14 -0.1175>, rot=<-0.0218 0.0258 0.3512 -0.9357>}
 middle_03_r {parent=middle_02_r, loc=<-2.6077 -0.0533 -0.1373>, rot=<0.0082 0.0188 0.1708 -0.9851>}
 pinky_metacarpal_r {parent=hand_r, loc=<-1.7617 0.9753 -3.3814>, rot=<-0.0397 0.0814 -0.069 -0.9935>}
 pinky_01_r {parent=pinky_metacarpal_r, loc=<-7.1707 0.4832 0.1094>, rot=<-0.0654 -0.0035 0.2242 -0.9723>}
 pinky_02_r {parent=pinky_01_r, loc=<-2.8274 -0.0892 0.1043>, rot=<0.0037 0.0011 0.2854 -0.9584>}
 pinky_03_r {parent=pinky_02_r, loc=<-2.2045 -0.031 0.0078>, rot=<-0.0002 0.0098 0.1445 -0.9895>}
 ring_metacarpal_r {parent=hand_r, loc=<-2.2073 0.3627 -1.3638>, rot=<-0.0284 0.0722 -0.0458 -0.9959>}
 ring_01_r {parent=ring_metacarpal_r, loc=<-7.0297 0.6197 0.1446>, rot=<-0.0489 -0.0252 0.1892 -0.9804>}
 ring_02_r {parent=ring_01_r, loc=<-3.6171 -0.4648 0.1027>, rot=<0 0 0.2806 -0.9598>}
 ring_03_r {parent=ring_02_r, loc=<-2.4153 -0.1141 -0.0199>, rot=<0 0 0.1189 -0.9929>}
 thumb_01_r {parent=hand_r, loc=<-1.3119 0.3192 2.459>, rot=<0.6546 -0.4042 0.0664 -0.6354>}
 thumb_02_r {parent=thumb_01_r, loc=<-5.9335 0.0447 -0.0401>, rot=<-0.0037 0.0213 0.298 -0.9543>}
 thumb_03_r {parent=thumb_02_r, loc=<-2.2223 -0.3682 0.0907>, rot=<0.0118 -0.013 0.0066 -0.9998>}
 weapon_r {parent=hand_r, loc=<0.1975 3.4112 -0.3811>, rot=<0 0 0 -1>}
 lowerarm_twist_01_r {parent=lowerarm_r, loc=<-18.3161 0 0>, rot=<0 0 0 1>}
 lowerarm_twist_02_r {parent=lowerarm_r, loc=<-10.6707 0 0>, rot=<0 0 0 1>}
 upperarm_twist_01_r {parent=upperarm_r, loc=<-7.2135 0.1134 0>, rot=<0 0 0 1>}
 upperarm_twist_02_r {parent=upperarm_r, loc=<-19.7998 0.1134 0>, rot=<0 0 0 1>}
 neck_01 {parent=spine_05, loc=<11.3138 0.1529 0>, rot=<0 0 -0.2003 -0.9797>}
 neck_02 {parent=neck_01, loc=<4.5627 0 0>, rot=<0 0 0.0504 -0.9987>}
 head {parent=neck_02, loc=<5.3914 0 0>, rot=<0 0 0.115 -0.9934>}
 driver_hairBack_01_l {parent=head, loc=<7.2635 -4.1748 -3.25>, rot=<0.9701 0.2425 0 0>}
 hairBack_01_l {parent=driver_hairBack_01_l, loc=<0 0 0>, rot=<0 -0.0465 0 -0.9989>}
 hairBack_02_l {parent=hairBack_01_l, loc=<-4.9083 -0.0002 0>, rot=<0.0188 -0.0259 -0.0894 -0.9955>}
 hairBack_03_l {parent=hairBack_02_l, loc=<-7.6342 0.0001 0>, rot=<0.0004 -0.0011 0.0333 -0.9994>}
 hairBack_04_l {parent=hairBack_03_l, loc=<-7.0437 0 0>, rot=<-0.0168 -0.0006 0.037 -0.9992>}
 hairBack_05_l {parent=hairBack_04_l, loc=<-6.182 -1.3269 0>, rot=<0.0139 0.0099 -0.2336 -0.9722>}
 hairBack_06_l {parent=hairBack_05_l, loc=<-6.6612 0 -0.0001>, rot=<0 0 0 1>}
 driver_hairBack_01_r {parent=head, loc=<7.2631 -4.1748 3.25>, rot=<0.2425 -0.9701 0 0>}
 hairBack_01_r {parent=driver_hairBack_01_r, loc=<0 0 0>, rot=<0 0 0 -1>}
 hairBack_02_r {parent=hairBack_01_r, loc=<4.9078 0 0>, rot=<0.0206 -0.0466 -0.089 -0.9947>}
 hairBack_03_r {parent=hairBack_02_r, loc=<7.6343 0 0>, rot=<0.0004 -0.0011 0.0333 -0.9994>}
 hairBack_04_r {parent=hairBack_03_r, loc=<7.0439 0 0>, rot=<-0.0168 -0.0006 0.037 -0.9992>}
 hairBack_05_r {parent=hairBack_04_r, loc=<6.1822 1.3271 0>, rot=<0.0139 0.0099 -0.2336 -0.9722>}
 hairBack_06_r {parent=hairBack_05_r, loc=<6.6604 0 0>, rot=<0 0 0 1>}
 driver_hairStrand_01_l {parent=head, loc=<-3.639 2.0766 -8.554>, rot=<-0.0997 -0.9844 0.0158 0.1443>}
 hairStrand_01_l {parent=driver_hairStrand_01_l, loc=<0 0 0>, rot=<0 0 0 -1>}
 hairStrand_02_l {parent=hairStrand_01_l, loc=<8.1609 0 0>, rot=<-0.0354 -0.1001 -0.1283 -0.986>}
 hairStrand_03_l {parent=hairStrand_02_l, loc=<7.1816 0 0>, rot=<0.0519 0.0905 -0.0513 -0.9932>}
 hairStrand_04_l {parent=hairStrand_03_l, loc=<10.3152 0 0>, rot=<0.0557 0.1859 0.2722 -0.9425>}
 hairStrand_05_l {parent=hairStrand_04_l, loc=<14.563 0 0>, rot=<0 0 0 1>}
 driver_hairStrand_b_01_r {parent=head, loc=<4.2644 7.0731 6.4145>, rot=<0.1975 -0.9694 -0.0279 0.1431>}
 hairStrand_b_01_r {parent=driver_hairStrand_b_01_r, loc=<0 0 0>, rot=<0 0 0 1>}
 hairStrand_b_02_r {parent=hairStrand_b_01_r, loc=<3.3132 0 0>, rot=<0.011 -0.0162 0.2002 -0.9796>}
 hairStrand_b_03_r {parent=hairStrand_b_02_r, loc=<4.4461 0 0>, rot=<0 0 0 -1>}
 hairStrand_a_01_r {parent=hairStrand_b_01_r, loc=<4.98 -2.5979 0.7743>, rot=<-0.035 0.1732 -0.19 -0.9658>}
 hairStrand_a_02_r {parent=hairStrand_a_01_r, loc=<9.4166 0 0>, rot=<-0.0281 -0.1367 -0.2043 -0.9689>}
 hairStrand_a_03_r {parent=hairStrand_a_02_r, loc=<8.4361 0 0>, rot=<-0.0351 -0.1042 0.0847 -0.9903>}
 hairStrand_a_04_r {parent=hairStrand_a_03_r, loc=<9.3417 0 0>, rot=<0 0 0 1>}
 hairStrand_r_aim {parent=spine_05, loc=<6.8246 5.0987 5.8159>, rot=<0.0427 -0.9986 -0.0012 -0.0313>}
 hairStrand_l_aim {parent=spine_05, loc=<5.8475 6.9011 -5.293>, rot=<-0.0643 -0.9873 0.021 0.1437>}
 thigh_l {parent=pelvis, loc=<-1.0855 -0.2109 -8.7053>, rot=<0.0009 0.0355 0.0155 -0.9993>}
 calf_l {parent=thigh_l, loc=<-42.5284 -0.0019 -0.7297>, rot=<0.0006 -0.0161 -0.0391 -0.9991>}
 calf_twist_01_l {parent=calf_l, loc=<-35.8483 -1.9197 -0.5794>, rot=<0.0237 0.0095 -0.0035 -0.9997>}
 calf_twist_02_l {parent=calf_l, loc=<-19.1546 -1.7943 -0.2638>, rot=<0.0237 0.0095 -0.0035 -0.9997>}
 foot_l {parent=calf_l, loc=<-46.7401 -2.5014 -0.592>, rot=<0.0427 -0.0238 0.0353 -0.9982>}
 ball_l {parent=foot_l, loc=<-3.5658 10.8066 -0.1544>, rot=<0.0001 -0.0001 -0.7063 -0.7079>}
 thigh_twist_01_l {parent=thigh_l, loc=<-27.952 0 0>, rot=<0.0013 0 0 -1>}
 thigh_r {parent=pelvis, loc=<-1.0853 -0.2109 8.7053>, rot=<0.0355 -0.0009 0.9993 0.0155>}
 calf_r {parent=thigh_r, loc=<42.5286 0.0019 0.7297>, rot=<0.0006 -0.0161 -0.0391 -0.9991>}
 calf_twist_01_r {parent=calf_r, loc=<35.8485 1.9197 0.5793>, rot=<0.0237 0.0095 -0.0035 -0.9997>}
 calf_twist_02_r {parent=calf_r, loc=<19.1546 1.7943 0.2638>, rot=<0.0237 0.0095 -0.0035 -0.9997>}
 foot_r {parent=calf_r, loc=<46.7403 2.5014 0.5919>, rot=<0.0427 -0.0238 0.0353 -0.9982>}
 ball_r {parent=foot_r, loc=<3.5658 -10.8066 0.1544>, rot=<0.0001 -0.0001 -0.7063 -0.7079>}
 thigh_twist_01_r {parent=thigh_r, loc=<27.952 0 0>, rot=<0.0013 0 0 -1>}
 ik_foot_root {parent=root, loc=<0 0 0>, rot=<0 0 0 1>}
 ik_foot_l {parent=ik_foot_root, loc=<14.8262 -7.3545 8.296>, rot=<-0.0357 -0.7045 -0.0284 0.7082>}
 ik_foot_r {parent=ik_foot_root, loc=<-14.8262 -7.3546 8.2958>, rot=<-0.7082 -0.0284 0.7045 -0.0357>}
 ik_hand_root {parent=root, loc=<0 0 0>, rot=<0 0 0 1>}
 ik_hand_gun {parent=ik_hand_root, loc=<-35.2227 3.1613 94.1496>, rot=<-0.5451 -0.2436 -0.5214 0.6097>}
 ik_hand_l {parent=ik_hand_gun, loc=<23.7864 63.4948 19.1095>, rot=<-0.3377 -0.9013 -0.2713 0>}
 ik_hand_r {parent=ik_hand_gun, loc=<0 0 0>, rot=<0 0 0 -1>}
]
