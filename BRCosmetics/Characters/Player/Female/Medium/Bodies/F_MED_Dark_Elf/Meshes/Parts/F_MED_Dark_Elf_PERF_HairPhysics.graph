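# Physics asset for the hair of the medium female Dark Elf body (PERF mesh part), modeling two back-hair chains and two side strands.
# Capsules on the upper spine, neck and head form the body volume, joined by free constraints.
# Tiny driver capsules hang from the head on free constraints and root each chain of tightly limited hair capsules.
# The right strand's chain crosses naming sets, from a "b" driver to "a" bones.
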PhysicsAsset /BRCosmetics/Characters/Player/Female/Medium/Bodies/F_MED_Dark_Elf/Meshes/Parts/F_MED_Dark_Elf_PERF_HairPhysics
bodies {
 driver_hairBack_01_l [{shapeType=Capsule, length=0.505, radius=0.505, rotation=<rot 0 -90 0>}]
 driver_hairBack_01_r [{shapeType=Capsule, length=0.505, radius=0.505, rotation=<rot 0 -90 0>}]
 driver_hairStrand_01_l [{shapeType=Capsule, length=0.505, radius=0.505, rotation=<rot 0 -90 0>}]
 driver_hairStrand_b_01_r [{shapeType=Capsule, length=0.505, radius=0.505, rotation=<rot 0 -90 0>}]
 hairBack_01_l [{shapeType=Capsule, length=3, radius=2.5, rotation=<rot 90 0 0>}]
 hairBack_01_r [{shapeType=Capsule, length=3, radius=2.5, rotation=<rot 90 0 0>}]
 hairBack_02_l [{shapeType=Capsule, length=3, radius=2.5, rotation=<rot 90 0 0>}]
 hairBack_02_r [{shapeType=Capsule, length=3, radius=2.5, rotation=<rot 90 0 0>}]
 hairBack_03_l [{shapeType=Capsule, length=3, radius=2.5, rotation=<rot 90 0 0>}]
 hairBack_03_r [{shapeType=Capsule, length=3, radius=2.5, rotation=<rot 90 0 0>}]
 hairBack_04_l [{shapeType=Capsule, length=3, radius=2.5, rotation=<rot 90 0 0>}]
 hairBack_04_r [{shapeType=Capsule, length=3, radius=2.5, rotation=<rot 90 0 0>}]
 hairBack_05_l [{shapeType=Capsule, length=3, radius=2.5, rotation=<rot 90 0 0>}]
 hairBack_05_r [{shapeType=Capsule, length=3, radius=2.5, rotation=<rot 90 0 0>}]
 hairBack_06_l [{shapeType=Capsule, length=3, radius=2.5, rotation=<rot 90 0 0>}]
 hairBack_06_r [{shapeType=Capsule, length=3, radius=2.5, rotation=<rot 90 0 0>}]
 hairStrand_02_l [{shapeType=Capsule, length=5, radius=3, rotation=<rot 90 0 0>}]
 hairStrand_03_l [{shapeType=Capsule, length=4, radius=3, rotation=<rot 90 0 0>}]
 hairStrand_04_l [{shapeType=Capsule, length=4, radius=3, rotation=<rot 90 0 0>}]
 hairStrand_05_l [{shapeType=Capsule, length=4, radius=3, rotation=<rot 90 0 0>}]
 hairStrand_a_02_r [{shapeType=Capsule, length=3.75, radius=2.25, rotation=<rot 90 0 0>}]
 hairStrand_a_03_r [{shapeType=Capsule, length=4, radius=3, rotation=<rot 90 0 0>}]
 hairStrand_a_04_r [{shapeType=Capsule, length=4, radius=3, rotation=<rot 90 0 0>}]
 head [{shapeType=Capsule, length=9.5, radius=5.5, rotation=<rot 90 180 158.7211>}]
 neck_01 [{shapeType=Capsule, length=8.5, radius=5.005, rotation=<rot 0 -90 0>}]
 neck_02 [{shapeType=Capsule, length=12.5, radius=5.255, rotation=<rot 0 -90 0>}]
 spine_03 [{shapeType=Capsule, length=21.005, radius=11.505, rotation=<rot 0 -90 0>}]
 spine_04 [{shapeType=Capsule, length=20.005, radius=10.005, center=<0 1.1835 0>, rotation=<rot 0 -90 0>}]
 spine_05 [{shapeType=Capsule, length=25.25, radius=6.755, center=<3.9977 -0.8674 0>, rotation=<rot 0 -90 0>}]
}
constraints [
 {
  bone1Name=head
  bone2Name=neck_02
  swing1LimitDegrees=45
  swing1Motion=Free
  swing2LimitDegrees=45
  swing2Motion=Free
  twistLimitDegrees=45
  twistMotion=Free
 }
 {
  bone1Name=neck_02
  bone2Name=neck_01
  swing1LimitDegrees=45
  swing1Motion=Free
  swing2LimitDegrees=45
  swing2Motion=Free
  twistLimitDegrees=45
  twistMotion=Free
 }
 {
  bone1Name=neck_01
  bone2Name=spine_05
  swing1LimitDegrees=45
  swing1Motion=Free
  swing2LimitDegrees=45
  swing2Motion=Free
  twistLimitDegrees=45
  twistMotion=Free
 }
 {
  bone1Name=driver_hairBack_01_r
  bone2Name=head
  swing1LimitDegrees=45
  swing1Motion=Free
  swing2LimitDegrees=45
  swing2Motion=Free
  twistLimitDegrees=45
  twistMotion=Free
 }
 {
  bone1Name=driver_hairBack_01_l
  bone2Name=head
  swing1LimitDegrees=45
  swing1Motion=Free
  swing2LimitDegrees=45
  swing2Motion=Free
  twistLimitDegrees=45
  twistMotion=Free
 }
 {
  bone1Name=driver_hairStrand_01_l
  bone2Name=head
  swing1LimitDegrees=45
  swing1Motion=Free
  swing2LimitDegrees=45
  swing2Motion=Free
  twistLimitDegrees=45
  twistMotion=Free
 }
 {
  bone1Name=driver_hairStrand_b_01_r
  bone2Name=head
  swing1LimitDegrees=45
  swing1Motion=Free
  swing2LimitDegrees=45
  swing2Motion=Free
  twistLimitDegrees=45
  twistMotion=Free
 }
 {bone1Name=hairBack_01_r, bone2Name=driver_hairBack_01_r, swing1LimitDegrees=15, swing2LimitDegrees=15, twistLimitDegrees=5}
 {bone1Name=hairBack_02_r, bone2Name=hairBack_01_r, swing1LimitDegrees=15, swing2LimitDegrees=15, twistLimitDegrees=5}
 {bone1Name=hairBack_03_r, bone2Name=hairBack_02_r, swing1LimitDegrees=15, swing2LimitDegrees=15, twistLimitDegrees=5}
 {bone1Name=hairBack_04_r, bone2Name=hairBack_03_r, swing1LimitDegrees=15, swing2LimitDegrees=15, twistLimitDegrees=5}
 {bone1Name=hairBack_05_r, bone2Name=hairBack_04_r, swing1LimitDegrees=15, swing2LimitDegrees=15, twistLimitDegrees=5}
 {bone1Name=hairBack_06_r, bone2Name=hairBack_05_r, swing1LimitDegrees=15, swing2LimitDegrees=15, twistLimitDegrees=5}
 {bone1Name=hairBack_01_l, bone2Name=driver_hairBack_01_l, swing1LimitDegrees=15, swing2LimitDegrees=15, twistLimitDegrees=5}
 {bone1Name=hairBack_02_l, bone2Name=hairBack_01_l, swing1LimitDegrees=15, swing2LimitDegrees=15, twistLimitDegrees=5}
 {bone1Name=hairBack_03_l, bone2Name=hairBack_02_l, swing1LimitDegrees=15, swing2LimitDegrees=15, twistLimitDegrees=5}
 {bone1Name=hairBack_04_l, bone2Name=hairBack_03_l, swing1LimitDegrees=15, swing2LimitDegrees=15, twistLimitDegrees=5}
 {bone1Name=hairBack_05_l, bone2Name=hairBack_04_l, swing1LimitDegrees=15, swing2LimitDegrees=15, twistLimitDegrees=5}
 {bone1Name=hairBack_06_l, bone2Name=hairBack_05_l, swing1LimitDegrees=15, swing2LimitDegrees=15, twistLimitDegrees=5}
 {bone1Name=hairStrand_a_02_r, bone2Name=driver_hairStrand_b_01_r, swing1LimitDegrees=15, swing2LimitDegrees=15, twistLimitDegrees=5}
 {bone1Name=hairStrand_02_l, bone2Name=driver_hairStrand_01_l, swing1LimitDegrees=15, swing2LimitDegrees=15, twistLimitDegrees=5}
 {bone1Name=hairStrand_a_03_r, bone2Name=hairStrand_a_02_r, swing1LimitDegrees=15, swing2LimitDegrees=15, twistLimitDegrees=5}
 {bone1Name=hairStrand_a_04_r, bone2Name=hairStrand_a_03_r, swing1LimitDegrees=15, swing2LimitDegrees=15, twistLimitDegrees=5}
 {bone1Name=hairStrand_03_l, bone2Name=hairStrand_02_l, swing1LimitDegrees=15, swing2LimitDegrees=15, twistLimitDegrees=5}
 {bone1Name=hairStrand_04_l, bone2Name=hairStrand_03_l, swing1LimitDegrees=15, swing2LimitDegrees=15, twistLimitDegrees=5}
 {bone1Name=hairStrand_05_l, bone2Name=hairStrand_04_l, swing1LimitDegrees=15, swing2LimitDegrees=15, twistLimitDegrees=5}
 {
  bone1Name=spine_05
  bone2Name=spine_04
  swing1LimitDegrees=45
  swing1Motion=Free
  swing2LimitDegrees=45
  swing2Motion=Free
  twistLimitDegrees=45
  twistMotion=Free
 }
 {
  bone1Name=spine_04
  bone2Name=spine_03
  swing1LimitDegrees=45
  swing1Motion=Free
  swing2LimitDegrees=45
  swing2Motion=Free
  twistLimitDegrees=45
  twistMotion=Free
 }
]
