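# Anim Blueprint for the Dark Elf female character's hair, apparently a performance variant, adding secondary motion over the body pose.
# The output chain layers copied mesh poses at the upper arms, applies weighted constraints and skydiving rotation offsets, then runs a Control Rig.
# AnimDynamics on front strands and RigidBody with variable gravity follow, finished by constraints softening dynamic bones toward spine_05.
# Other visible backpack, head and clavicle correction branches sit outside that output chain.
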AnimBlueprint /BRCosmetics/Characters/Player/Female/Medium/Bodies/F_MED_Dark_Elf/Meshes/Parts/F_MED_Dark_Elf_Hair_PERF_AnimBP
target_skeleton=&/F_MED_Dark_Elf_FaceAcc_Skeleton
graphs [
 AnimGraph {nodes [
   Anim Dynamics - Bone: hairStrand_a_01_r(Alpha=1, ExternalForce="0,0,0") {
    angularDampingOverride=0.98
    angularSpringConstant=75
    bAngularSpring=true
    bGravityOverrideInSimSpace=true
    bOverrideAngularDamping=true
    simulationSpace=RootRelative
    alphaBoolBlend {blendOption=Linear}
    boundBone {boneName=hairStrand_a_01_r}
    componentAppliedLinearAccClamp=<3500 3500 3500>
    componentLinearAccScale=<1 1 1>
    componentLinearVelScale=<1 1 1>
    physicsBodyDefinitions [{
      sphereCollisionRadius=0
      boundBone {boneName=hairStrand_a_01_r}
      boxExtents=<20 20 15>
      constraintSetup {angularLimitsMax=<1 15 30>, angularLimitsMin=<-1 -15 0>, angularTarget=<1 0 0>}
      localJointOffset=<2 0 0>
     }]
    relativeSpaceBone {boneName=root}
   }
   n1 = RigidBody!(ComponentPose=Local To Component(LocalPose=Use cached pose 'pre-backpack'()), bAlphaBoolEnabled:bool, ExternalForce) -> (Pose) {
    alphaInputType=Bool
    bClampLinearTranslationLimitToRefPose=true
    bOverrideWorldGravity=true
    overridePhysicsAsset=&/F_MED_Dark_Elf_Backbling_PERF_HairPhysics
    simulationSpace=BaseBoneSpace
    alphaBoolBlend {blendOption=Linear}
    baseBoneRef {boneName=spine_05}
    componentAppliedLinearAccClamp=<3000 3000 300>
    componentLinearAccScale=<0.4 0.4 0.05>
    componentLinearVelScale=<0.15 0.15 0.075>
    overrideWorldGravity {z=-980}
   }
   n2 = "Constraint - hairBack_03_r (P)"(ComponentPose=@n1, Alpha=1, ConstraintWeights_0=1) {
    alphaBoolBlend {blendOption=Linear}
    boneToModify {boneName=hairBack_03_r}
    constraintSetup [{targetBone {boneName=spine_04}}]
    constraintWeights [1]
   }
   n3 = "Constraint - hairBack_03_l (P)"(ComponentPose=@n2, Alpha=1, ConstraintWeights_0=1) {
    alphaBoolBlend {blendOption=Linear}
    boneToModify {boneName=hairBack_03_l}
    constraintSetup [{targetBone {boneName=spine_04}}]
    constraintWeights [1]
   }
   n4 = "Constraint - hairBack_05_r (P)"(ComponentPose=@n3, Alpha=1, ConstraintWeights_0=1) {
    alphaBoolBlend {blendOption=Linear}
    boneToModify {boneName=hairBack_05_r}
    constraintSetup [{targetBone {boneName=spine_04}}]
    constraintWeights [1]
   }
   n5 = "Constraint - hairBack_05_l (P)"(ComponentPose=@n4, Alpha=1, ConstraintWeights_0=1) {
    alphaBoolBlend {blendOption=Linear}
    boneToModify {boneName=hairBack_05_l}
    constraintSetup [{targetBone {boneName=spine_04}}]
    constraintWeights [1]
   }
   n6 = "Transform (Modify) Bone - Bone: hairBack_02_l"(ComponentPose=@n5, Alpha=1, Translation="0,0,0", Rotation="0,0,0", Scale="0.8,1,1") {
    rotationMode=BMM_Additive
    rotationSpace=BCS_BoneSpace
    scaleMode=BMM_Additive
    scaleSpace=BCS_BoneSpace
    translationMode=BMM_Additive
    translationSpace=BCS_BoneSpace
    alphaBoolBlend {blendOption=Linear}
    boneToModify {boneName=hairBack_02_l}
    scale {x=0.8}
   }
   n7 = "Transform (Modify) Bone - Bone: hairBack_02_r"(ComponentPose=@n6, Alpha=1, Translation="0,0,0", Rotation="0,0,0", Scale="0.8,1,1") {
    rotationMode=BMM_Additive
    rotationSpace=BCS_BoneSpace
    scaleMode=BMM_Additive
    scaleSpace=BCS_BoneSpace
    translationMode=BMM_Additive
    translationSpace=BCS_BoneSpace
    alphaBoolBlend {blendOption=Linear}
    boneToModify {boneName=hairBack_02_r}
    scale {x=0.8}
   }
   n8 = "Transform (Modify) Bone - Bone: hairBack_05_l"(ComponentPose=@n7, Alpha=1, Translation="0,0,0", Rotation="0,0,0", Scale="1,1,1") {
    rotationMode=BMM_Additive
    rotationSpace=BCS_BoneSpace
    scaleMode=BMM_Additive
    scaleSpace=BCS_BoneSpace
    translationMode=BMM_Additive
    translationSpace=BCS_BoneSpace
    alphaBoolBlend {blendOption=Linear}
    boneToModify {boneName=hairBack_05_l}
   }
   n9 = "Transform (Modify) Bone - Bone: hairBack_05_r"(ComponentPose=@n8, Alpha=1, Translation="0,0,0", Rotation="0,0,0", Scale="1,1,1") {
    rotationMode=BMM_Additive
    rotationSpace=BCS_BoneSpace
    scaleMode=BMM_Additive
    scaleSpace=BCS_BoneSpace
    translationMode=BMM_Additive
    translationSpace=BCS_BoneSpace
    alphaBoolBlend {blendOption=Linear}
    boneToModify {boneName=hairBack_05_r}
   }
   n10 = "Transform (Modify) Bone - Bone: hairBack_04_l"(ComponentPose=@n9, Alpha=1, Translation="0,0,0", Rotation="0,-3,0", Scale="1,1,1") {
    rotationMode=BMM_Additive
    rotationSpace=BCS_BoneSpace
    scaleMode=BMM_Additive
    scaleSpace=BCS_BoneSpace
    translationMode=BMM_Additive
    translationSpace=BCS_BoneSpace
    alphaBoolBlend {blendOption=Linear}
    boneToModify {boneName=hairBack_04_l}
    rotation {yaw=-3}
   }
   n11 = "Transform (Modify) Bone - Bone: hairBack_04_r"(ComponentPose=@n10, Alpha=1, Translation="0,0,0", Rotation="0,-3,0", Scale="1,1,1") {
    rotationMode=BMM_Additive
    rotationSpace=BCS_BoneSpace
    scaleMode=BMM_Additive
    scaleSpace=BCS_BoneSpace
    translationMode=BMM_Additive
    translationSpace=BCS_BoneSpace
    alphaBoolBlend {blendOption=Linear}
    boneToModify {boneName=hairBack_04_r}
    rotation {yaw=-3}
   }
   n12 = Anim Dynamics - Bone: hairBack_04_l(ComponentPose=@n11, Alpha=1, ExternalForce="0,0,0") {
    angularDampingOverride=0.9999
    angularSpringConstant=100
    bGravityOverrideInSimSpace=true
    bOverrideAngularDamping=true
    simulationSpace=BoneRelative
    alphaBoolBlend {blendOption=Linear}
    boundBone {boneName=hairBack_04_l}
    componentAppliedLinearAccClamp=<3500 3500 3500>
    componentLinearAccScale=<0.6 0.6 0.8>
    componentLinearVelScale=<0.6 0.6 0.8>
    physicsBodyDefinitions [{
      sphereCollisionRadius=0
      boundBone {boneName=hairBack_04_l}
      boxExtents=<20 20 10>
      constraintSetup {angularLimitsMax=<1 15 -3>, angularLimitsMin=<-1 -15 -8>, angularTarget=<-1 0 0>}
      localJointOffset=<-1 0 0>
     }]
    relativeSpaceBone {boneName=root}
   }
   n13 = Anim Dynamics - Bone: hairBack_05_l(ComponentPose=@n12, Alpha=1, ExternalForce="0,0,0") {
    angularDampingOverride=0.9999
    angularSpringConstant=100
    bGravityOverrideInSimSpace=true
    bOverrideAngularDamping=true
    simulationSpace=BoneRelative
    alphaBoolBlend {blendOption=Linear}
    boundBone {boneName=hairBack_05_l}
    componentAppliedLinearAccClamp=<3500 3500 3500>
    componentLinearAccScale=<0.6 0.6 0.8>
    componentLinearVelScale=<0.6 0.6 0.8>
    physicsBodyDefinitions [{
      sphereCollisionRadius=0
      boundBone {boneName=hairBack_05_l}
      boxExtents=<20 20 10>
      constraintSetup {angularLimitsMax=<1 15 -3>, angularLimitsMin=<-1 -15 -8>, angularTarget=<-1 0 0>}
      localJointOffset=<-1 0 0>
     }]
    relativeSpaceBone {boneName=root}
   }
   n14 = Anim Dynamics - Bone: hairBack_06_l(ComponentPose=@n13, Alpha=1, ExternalForce="0,0,0") {
    angularDampingOverride=0.9999
    angularSpringConstant=100
    bGravityOverrideInSimSpace=true
    bOverrideAngularDamping=true
    simulationSpace=BoneRelative
    alphaBoolBlend {blendOption=Linear}
    boundBone {boneName=hairBack_06_l}
    componentAppliedLinearAccClamp=<3500 3500 3500>
    componentLinearAccScale=<0.6 0.6 0.8>
    componentLinearVelScale=<0.6 0.6 0.8>
    physicsBodyDefinitions [{
      sphereCollisionRadius=0
      boundBone {boneName=hairBack_06_l}
      boxExtents=<20 20 10>
      constraintSetup {angularLimitsMax=<1 15 -3>, angularLimitsMin=<-1 -15 -8>, angularTarget=<-1 0 0>}
      localJointOffset=<-1 0 0>
     }]
    relativeSpaceBone {boneName=root}
   }
   n15 = Anim Dynamics - Bone: hairBack_04_r(ComponentPose=@n14, Alpha=1, ExternalForce="0,0,0") {
    angularDampingOverride=0.9999
    angularSpringConstant=100
    bGravityOverrideInSimSpace=true
    bOverrideAngularDamping=true
    simulationSpace=BoneRelative
    alphaBoolBlend {blendOption=Linear}
    boundBone {boneName=hairBack_04_r}
    componentAppliedLinearAccClamp=<3500 3500 3500>
    componentLinearAccScale=<0.6 0.6 0.8>
    componentLinearVelScale=<0.6 0.6 0.8>
    physicsBodyDefinitions [{
      sphereCollisionRadius=0
      boundBone {boneName=hairBack_04_r}
      boxExtents=<20 20 10>
      constraintSetup {angularLimitsMax=<1 15 -3>, angularLimitsMin=<-1 -15 -8>, angularTarget=<-1 0 0>}
      localJointOffset=<1 0 0>
     }]
    relativeSpaceBone {boneName=root}
   }
   n16 = Anim Dynamics - Bone: hairBack_05_r(ComponentPose=@n15, Alpha=1, ExternalForce="0,0,0") {
    angularDampingOverride=0.9999
    angularSpringConstant=100
    bGravityOverrideInSimSpace=true
    bOverrideAngularDamping=true
    simulationSpace=BoneRelative
    alphaBoolBlend {blendOption=Linear}
    boundBone {boneName=hairBack_05_r}
    componentAppliedLinearAccClamp=<3500 3500 3500>
    componentLinearAccScale=<0.6 0.6 0.8>
    componentLinearVelScale=<0.6 0.6 0.8>
    physicsBodyDefinitions [{
      sphereCollisionRadius=0
      boundBone {boneName=hairBack_05_r}
      boxExtents=<20 20 10>
      constraintSetup {angularLimitsMax=<1 15 -3>, angularLimitsMin=<-1 -15 -8>, angularTarget=<-1 0 0>}
      localJointOffset=<1 0 0>
     }]
    relativeSpaceBone {boneName=root}
   }
   n17 = Anim Dynamics - Bone: hairBack_06_r(ComponentPose=@n16, Alpha=1, ExternalForce="0,0,0") {
    angularDampingOverride=0.9999
    angularSpringConstant=100
    bGravityOverrideInSimSpace=true
    bOverrideAngularDamping=true
    simulationSpace=BoneRelative
    alphaBoolBlend {blendOption=Linear}
    boundBone {boneName=hairBack_06_r}
    componentAppliedLinearAccClamp=<3500 3500 3500>
    componentLinearAccScale=<0.6 0.6 0.8>
    componentLinearVelScale=<0.6 0.6 0.8>
    physicsBodyDefinitions [{
      sphereCollisionRadius=0
      boundBone {boneName=hairBack_06_r}
      boxExtents=<20 20 10>
      constraintSetup {angularLimitsMax=<1 15 -3>, angularLimitsMin=<-1 -15 -8>, angularTarget=<-1 0 0>}
      localJointOffset=<1 0 0>
     }]
    relativeSpaceBone {boneName=root}
   }
   Blend Poses by bool(
    BlendPose_0=Component To Local(ComponentPose=@n17)
    BlendPose_1=Use cached pose 'pre-backpack'()
    bActiveValue=Get Wearing Backpack() -> (Wearing Backpack:bool)
   ) {blendType=Linear}
   n18 = Get backbling_hair() -> (backbling_hair:double)
   n19 = "Constraint - hairBack_03_r (P)"(ComponentPose=Local To Component(), Alpha=@n18, ConstraintWeights_0=1) {
    alphaBoolBlend {blendOption=Linear}
    boneToModify {boneName=hairBack_03_r}
    constraintSetup [{targetBone {boneName=spine_04}}]
    constraintWeights [1]
   }
   n20 = "Constraint - hairBack_03_l (P)"(ComponentPose=@n19, Alpha=@n18, ConstraintWeights_0=1) {
    alphaBoolBlend {blendOption=Linear}
    boneToModify {boneName=hairBack_03_l}
    constraintSetup [{targetBone {boneName=spine_04}}]
    constraintWeights [1]
   }
   n21 = "Constraint - hairBack_05_r (P)"(ComponentPose=@n20, Alpha=@n18, ConstraintWeights_0=1) {
    alphaBoolBlend {blendOption=Linear}
    boneToModify {boneName=hairBack_05_r}
    constraintSetup [{targetBone {boneName=spine_04}}]
    constraintWeights [1]
   }
   n22 = "Constraint - hairBack_05_l (P)"(ComponentPose=@n21, Alpha=@n18, ConstraintWeights_0=1) {
    alphaBoolBlend {blendOption=Linear}
    boneToModify {boneName=hairBack_05_l}
    constraintSetup [{targetBone {boneName=spine_04}}]
    constraintWeights [1]
   }
   n23 = "Transform (Modify) Bone - Bone: hairBack_02_l"(ComponentPose=@n22, Alpha=@n18, Translation="0,0,0", Rotation="0,0,0", Scale="0.7,0.7,1") {
    rotationMode=BMM_Additive
    rotationSpace=BCS_BoneSpace
    scaleMode=BMM_Additive
    scaleSpace=BCS_BoneSpace
    translationMode=BMM_Additive
    translationSpace=BCS_BoneSpace
    alphaBoolBlend {blendOption=Linear}
    boneToModify {boneName=hairBack_02_l}
    scale=<0.7 0.7>
   }
   n24 = "Transform (Modify) Bone - Bone: hairBack_02_r"(ComponentPose=@n23, Alpha=@n18, Translation="0,0,0", Rotation="0,0,0", Scale="0.7,0.7,1") {
    rotationMode=BMM_Additive
    rotationSpace=BCS_BoneSpace
    scaleMode=BMM_Additive
    scaleSpace=BCS_BoneSpace
    translationMode=BMM_Additive
    translationSpace=BCS_BoneSpace
    alphaBoolBlend {blendOption=Linear}
    boneToModify {boneName=hairBack_02_r}
    scale=<0.7 0.7>
   }
   n25 = "Transform (Modify) Bone - Bone: hairBack_05_l"(ComponentPose=@n24, Alpha=@n18, Translation="0,0,0", Rotation="0,0,0", Scale="1,1,1") {
    rotationMode=BMM_Additive
    rotationSpace=BCS_BoneSpace
    scaleMode=BMM_Additive
    scaleSpace=BCS_BoneSpace
    translationMode=BMM_Additive
    translationSpace=BCS_BoneSpace
    alphaBoolBlend {blendOption=Linear}
    boneToModify {boneName=hairBack_05_l}
   }
   n26 = "Transform (Modify) Bone - Bone: hairBack_05_r"(ComponentPose=@n25, Alpha=@n18, Translation="0,0,0", Rotation="0,0,0", Scale="1,1,1") {
    rotationMode=BMM_Additive
    rotationSpace=BCS_BoneSpace
    scaleMode=BMM_Additive
    scaleSpace=BCS_BoneSpace
    translationMode=BMM_Additive
    translationSpace=BCS_BoneSpace
    alphaBoolBlend {blendOption=Linear}
    boneToModify {boneName=hairBack_05_r}
   }
   Component To Local(ComponentPose=@n26)
   n27 = "Constraint - hairBack_02_r (PT)"!(ComponentPose, Alpha=0.3, ConstraintWeights_0=1, ConstraintWeights_1:float=0) -> (Pose) {
    alphaBoolBlend {blendOption=Linear}
    boneToModify {boneName=hairBack_02_r}
    constraintSetup [{targetBone {boneName=spine_05}}, {transformType=Translation, targetBone {boneName=None}}]
    constraintWeights [1,1]
   }
   n28 = "Constraint - hairBack_03_r (P)"(ComponentPose=@n27, Alpha=0.7, ConstraintWeights_0=1) {
    alphaBoolBlend {blendOption=Linear}
    boneToModify {boneName=hairBack_03_r}
    constraintSetup [{targetBone {boneName=spine_05}}]
    constraintWeights [1]
   }
   n29 = "Constraint - hairBack_04_r (P)"(ComponentPose=@n28, Alpha=0.7, ConstraintWeights_0=1) {
    alphaBoolBlend {blendOption=Linear}
    boneToModify {boneName=hairBack_04_r}
    constraintSetup [{targetBone {boneName=spine_05}}]
    constraintWeights [1]
   }
   n30 = "Constraint - hairBack_02_l (PT)"!(ComponentPose=@n29, Alpha=0.3, ConstraintWeights_0=1, ConstraintWeights_1:float=0) -> (Pose) {
    alphaBoolBlend {blendOption=Linear}
    boneToModify {boneName=hairBack_02_l}
    constraintSetup [{targetBone {boneName=spine_05}}, {transformType=Translation, targetBone {boneName=None}}]
    constraintWeights [1,1]
   }
   n31 = "Constraint - hairBack_03_l (P)"(ComponentPose=@n30, Alpha=0.7, ConstraintWeights_0=1) {
    alphaBoolBlend {blendOption=Linear}
    boneToModify {boneName=hairBack_03_l}
    constraintSetup [{targetBone {boneName=spine_05}}]
    constraintWeights [1]
   }
   "Constraint - hairBack_04_l (P)"(ComponentPose=@n31, Alpha=0.7, ConstraintWeights_0=1) {
    alphaBoolBlend {blendOption=Linear}
    boneToModify {boneName=hairBack_04_l}
    constraintSetup [{targetBone {boneName=spine_05}}]
    constraintWeights [1]
   }
   n32 = "Constraint - dyn_hair_l_02 (P)"(Alpha=0.4, ConstraintWeights_0=1) {
    alphaBoolBlend {blendOption=Linear}
    alphaScaleBiasClamp {bInterpResult=true}
    boneToModify {boneName=dyn_hair_l_02}
    constraintSetup [{targetBone {boneName=spine_05}}]
    constraintWeights [1]
   }
   "Constraint - dyn_hair_r_02 (P)"(ComponentPose=@n32, Alpha=0.4, ConstraintWeights_0=1) {
    alphaBoolBlend {blendOption=Linear}
    alphaScaleBiasClamp {bInterpResult=true}
    boneToModify {boneName=dyn_hair_r_02}
    constraintSetup [{targetBone {boneName=spine_05}}]
    constraintWeights [1]
   }
   n33 = "Constraint - dyn_hair_l_03 (P)"(Alpha=0.4, ConstraintWeights_0=1) {
    alphaBoolBlend {blendOption=Linear}
    alphaScaleBiasClamp {bInterpResult=true}
    boneToModify {boneName=dyn_hair_l_03}
    constraintSetup [{targetBone {boneName=spine_05}}]
    constraintWeights [1]
   }
   "Constraint - dyn_hair_r_03 (P)"(ComponentPose=@n33, Alpha=0.4, ConstraintWeights_0=1) {
    alphaBoolBlend {blendOption=Linear}
    alphaScaleBiasClamp {bInterpResult=true}
    boneToModify {boneName=dyn_hair_r_03}
    constraintSetup [{targetBone {boneName=spine_05}}]
    constraintWeights [1]
   }
   Local To Component()
   n34 = Get headUp() -> (headUp:double)
   "Transform (Modify) Bone - Bone: driver_hairStrand_01_l"(Alpha=@n34, Translation="0,0,0", Rotation="0,-20,0", Scale="1,1,1") {
    rotationMode=BMM_Additive
    rotationSpace=BCS_BoneSpace
    alphaBoolBlend {blendOption=Linear}
    boneToModify {boneName=driver_hairStrand_01_l}
    rotation {yaw=-20}
   }
   "Transform (Modify) Bone - Bone: hairBack_04_l"(Alpha=1, Translation="0,0,0", Rotation="-10,0,0", Scale="1,1,1") {
    rotationMode=BMM_Additive
    rotationSpace=BCS_BoneSpace
    alphaBoolBlend {blendOption=Linear}
    boneToModify {boneName=hairBack_04_l}
    rotation {pitch=-10}
   }
   n35 = Get headDown() -> (headDown:double)
   n36 = "Transform (Modify) Bone - Bone: hairStrand_01_l"(Alpha=@n35, Translation="0,0,0", Rotation="0,20,0", Scale="1,1,1") {
    rotationMode=BMM_Additive
    rotationSpace=BCS_BoneSpace
    alphaBoolBlend {blendOption=Linear}
    boneToModify {boneName=hairStrand_01_l}
    rotation {yaw=20}
   }
   "Transform (Modify) Bone - Bone: hairStrand_03_l"(ComponentPose=@n36, Alpha=@n35, Translation="0,0,0", Rotation="0,-20,0", Scale="1,1,1") {
    rotationMode=BMM_Additive
    rotationSpace=BCS_BoneSpace
    alphaBoolBlend {blendOption=Linear}
    boneToModify {boneName=hairStrand_03_l}
    rotation {yaw=-20}
   }
   n37 = Get headUp()
   n38 = "Transform (Modify) Bone - Bone: hairBack_01_r"(Alpha=@n37, Translation="-1,-4,0", Rotation="5,-15,0", Scale="1,1,1") {
    rotationMode=BMM_Additive
    rotationSpace=BCS_BoneSpace
    translationMode=BMM_Additive
    translationSpace=BCS_BoneSpace
    alphaBoolBlend {blendOption=Linear}
    boneToModify {boneName=hairBack_01_r}
    rotation {pitch=5, yaw=-15}
    translation=<-1 -4>
   }
   n39 = "Transform (Modify) Bone - Bone: hairBack_02_r"(ComponentPose=@n38, Alpha=@n37, Translation="0,0,0", Rotation="0,25,0", Scale="1,1,1") {
    rotationMode=BMM_Additive
    rotationSpace=BCS_BoneSpace
    translationMode=BMM_Additive
    translationSpace=BCS_BoneSpace
    alphaBoolBlend {blendOption=Linear}
    boneToModify {boneName=hairBack_02_r}
    rotation {yaw=25}
   }
   n40 = "Transform (Modify) Bone - Bone: hairBack_01_l"(ComponentPose=@n39, Alpha=@n37, Translation="0,2,0", Rotation="0,-15,0", Scale="1,1,1") {
    rotationMode=BMM_Additive
    rotationSpace=BCS_BoneSpace
    translationMode=BMM_Additive
    translationSpace=BCS_BoneSpace
    alphaBoolBlend {blendOption=Linear}
    boneToModify {boneName=hairBack_01_l}
    rotation {yaw=-15}
    translation {y=2}
   }
   "Transform (Modify) Bone - Bone: hairBack_02_l"(ComponentPose=@n40, Alpha=@n37, Translation="0,0,0", Rotation="0,10,0", Scale="1,1,1") {
    rotationMode=BMM_Additive
    rotationSpace=BCS_BoneSpace
    translationMode=BMM_Additive
    translationSpace=BCS_BoneSpace
    alphaBoolBlend {blendOption=Linear}
    boneToModify {boneName=hairBack_02_l}
    rotation {yaw=10}
   }
   n41 = "Transform (Modify) Bone - Bone: hairBack_02_r"(Alpha=Get clavicleR() -> (clavicleR:double), Translation="0,-2,0", Rotation="0,-15,0", Scale="1,1,1") {
    rotationMode=BMM_Additive
    rotationSpace=BCS_BoneSpace
    translationMode=BMM_Additive
    translationSpace=BCS_BoneSpace
    alphaBoolBlend {blendOption=Linear}
    boneToModify {boneName=hairBack_02_r}
    rotation {yaw=-15}
    translation {y=-2}
   }
   n42 = "Transform (Modify) Bone - Bone: hairBack_02_l"(ComponentPose=@n41, Alpha=Get clavicleL() -> (clavicleL:double), Translation="0,2,0", Rotation="0,-15,0", Scale="1,1,1") {
    rotationMode=BMM_Additive
    rotationSpace=BCS_BoneSpace
    translationMode=BMM_Additive
    translationSpace=BCS_BoneSpace
    alphaBoolBlend {blendOption=Linear}
    boneToModify {boneName=hairBack_02_l}
    rotation {yaw=-15}
    translation {y=2}
   }
   "Transform (Modify) Bone - Bone: hairBack_02_l"(ComponentPose=@n42, Alpha=Get clavicleL_bck() -> (clavicleL_bck:double), Translation="0,2,0", Rotation="0,-15,0", Scale="1,1,1") {
    rotationMode=BMM_Additive
    rotationSpace=BCS_BoneSpace
    translationMode=BMM_Additive
    translationSpace=BCS_BoneSpace
    alphaBoolBlend {blendOption=Linear}
    boneToModify {boneName=hairBack_02_l}
    rotation {yaw=-15}
    translation {y=2}
   }
   RigidBody!(
    ComponentPose
    Alpha:float
    OverrideWorldGravity:Vector="0,0,-980"
    ExternalForce
    ComponentLinearAccScale:Vector="0.4,0.4,0.6"
    ComponentLinearVelScale:Vector="0.4,0.4,0.6"
   ) -> (Pose) {
    bClampLinearTranslationLimitToRefPose=true
    bOverrideWorldGravity=true
    overridePhysicsAsset=&/F_MED_Dark_elf_FaceAcc_Physics
    simulationSpace=BaseBoneSpace
    alphaBoolBlend {blendOption=Linear}
    baseBoneRef {boneName=spine_05}
    componentAppliedLinearAccClamp=<3000 3000 3000>
    componentLinearAccScale=<0.4 0.4 0.6>
    componentLinearVelScale=<0.4 0.4 0.6>
   }
   n43 = Copy Pose From Mesh(SourceMeshComponent=Get Skel_Mesh())
   n44 = Copy Pose From Mesh() {bUseAttachedParent=true}
   n45 = "Constraint - hairStrand_a_01_r (PP)"!(
    ComponentPose=Local To Component(LocalPose=Layered blend per bone(BasePose=@n44, BlendPoses_0=@n43) {layerSetup [{branchFilters [{boneName=upperarm_l}, {boneName=upperarm_r}]}]})
    Alpha=1
    ConstraintWeights_0=0.2
    ConstraintWeights_1:float=0.8
   ) -> (Pose) {
    alphaBoolBlend {blendOption=Linear}
    boneToModify {boneName=hairStrand_a_01_r}
    constraintSetup [{targetBone {boneName=spine_05}}, {targetBone {boneName=head}}]
    constraintWeights [1,1]
   }
   n46 = "Constraint - hairStrand_01_l (PP)"!(ComponentPose=@n45, Alpha=1, ConstraintWeights_0=0.6, ConstraintWeights_1:float=0.4) -> (Pose) {
    alphaBoolBlend {blendOption=Linear}
    boneToModify {boneName=hairStrand_01_l}
    constraintSetup [{targetBone {boneName=spine_05}}, {targetBone {boneName=head}}]
    constraintWeights [1,1]
   }
   n47 = "Constraint - hairStrand_a_02_r (PP)"!(ComponentPose=@n46, Alpha=0.7, ConstraintWeights_0=0.8, ConstraintWeights_1:float=0.2) -> (Pose) {
    alphaBoolBlend {blendOption=Linear}
    boneToModify {boneName=hairStrand_a_02_r}
    constraintSetup [{targetBone {boneName=spine_05}}, {targetBone {boneName=clavicle_r}}]
    constraintWeights [1,1]
   }
   n48 = "Constraint - hairStrand_02_l (PP)"!(ComponentPose=@n47, Alpha=0.7, ConstraintWeights_0=0.8, ConstraintWeights_1:float=0.2) -> (Pose) {
    alphaBoolBlend {blendOption=Linear}
    boneToModify {boneName=hairStrand_02_l}
    constraintSetup [{targetBone {boneName=spine_05}}, {targetBone {boneName=clavicle_r}}]
    constraintWeights [1,1]
   }
   n49 = Get bIsSkyDiving()
   n50 = "Transform (Modify) Bone - Bone: dyn_hair_bk_00"!(ComponentPose=@n48, bAlphaBoolEnabled:bool=@n49, Translation="0,0,0", Rotation="0,-35,0", Scale="1,1,1") -> (Pose) {
    alphaInputType=Bool
    rotationMode=BMM_Additive
    rotationSpace=BCS_BoneSpace
    alphaBoolBlend {blendInTime=0.1, blendOption=Linear, blendOutTime=0.1}
    boneToModify {boneName=dyn_hair_bk_00}
    rotation {yaw=-35}
   }
   n51 = "Transform (Modify) Bone - Bone: dyn_hair_l_00"!(ComponentPose=@n50, bAlphaBoolEnabled:bool=@n49, Translation="0,0,0", Rotation="0,-35,0", Scale="1,1,1") -> (Pose) {
    alphaInputType=Bool
    rotationMode=BMM_Additive
    rotationSpace=BCS_BoneSpace
    alphaBoolBlend {blendInTime=0.1, blendOption=Linear, blendOutTime=0.1}
    boneToModify {boneName=dyn_hair_l_00}
    rotation {yaw=-35}
   }
   n52 = "Transform (Modify) Bone - Bone: dyn_hair_r_00"!(ComponentPose=@n51, bAlphaBoolEnabled:bool=@n49, Translation="0,0,0", Rotation="0,-35,0", Scale="1,1,1") -> (Pose) {
    alphaInputType=Bool
    rotationMode=BMM_Additive
    rotationSpace=BCS_BoneSpace
    alphaBoolBlend {blendInTime=0.1, blendOption=Linear, blendOutTime=0.1}
    boneToModify {boneName=dyn_hair_r_00}
    rotation {yaw=-35}
   }
   n53 = Control Rig(Source=Component To Local(ComponentPose=@n52)) {
    alphaBoolBlend {blendOption=Linear}
    controlRigAssetReference {blueprintRigClass=&/F_MED_Dark_Elf_Hair_FaceAcc_CtrlRig.F_MED_Dark_Elf_Hair_FaceAcc_CtrlRig_C}
   }
   n54 = Anim Dynamics - Bone: hairStrand_a_02_r(ComponentPose=Local To Component(LocalPose=@n53), Alpha=1, ExternalForce="0,0,0") {
    angularDampingOverride=0.98
    angularSpringConstant=75
    bAngularSpring=true
    bGravityOverrideInSimSpace=true
    bOverrideAngularDamping=true
    simulationSpace=RootRelative
    alphaBoolBlend {blendOption=Linear}
    boundBone {boneName=hairStrand_a_02_r}
    componentAppliedLinearAccClamp=<3500 3500 3500>
    componentLinearAccScale=<1 1 1>
    componentLinearVelScale=<1 1 1>
    physicsBodyDefinitions [{
      sphereCollisionRadius=0
      boundBone {boneName=hairStrand_a_02_r}
      boxExtents=<20 20 15>
      constraintSetup {angularLimitsMax=<1 15 30>, angularLimitsMin=<-1 -15 0>, angularTarget=<1 0 0>}
      localJointOffset=<2 0 0>
     }]
    relativeSpaceBone {boneName=root}
   }
   n55 = Anim Dynamics - Bone: hairStrand_a_03_r(ComponentPose=@n54, Alpha=1, ExternalForce="0,0,0") {
    angularDampingOverride=0.98
    angularSpringConstant=75
    bAngularSpring=true
    bGravityOverrideInSimSpace=true
    bOverrideAngularDamping=true
    simulationSpace=RootRelative
    alphaBoolBlend {blendOption=Linear}
    boundBone {boneName=hairStrand_a_03_r}
    componentAppliedLinearAccClamp=<3500 3500 3500>
    componentLinearAccScale=<1 1 1>
    componentLinearVelScale=<1 1 1>
    physicsBodyDefinitions [{
      sphereCollisionRadius=0
      boundBone {boneName=hairStrand_a_03_r}
      boxExtents=<20 20 15>
      constraintSetup {angularLimitsMax=<1 15 30>, angularLimitsMin=<-1 -15 5>, angularTarget=<1 0 0>}
      localJointOffset=<2 -1.25 0>
     }]
    relativeSpaceBone {boneName=root}
   }
   n56 = Anim Dynamics - Bone: hairStrand_a_04_r(ComponentPose=@n55, Alpha=1, ExternalForce="0,0,0") {
    angularDampingOverride=0.98
    angularSpringConstant=75
    bAngularSpring=true
    bGravityOverrideInSimSpace=true
    bOverrideAngularDamping=true
    simulationSpace=RootRelative
    alphaBoolBlend {blendOption=Linear}
    boundBone {boneName=hairStrand_a_04_r}
    componentAppliedLinearAccClamp=<3500 3500 3500>
    componentLinearAccScale=<1 1 1>
    componentLinearVelScale=<1 1 1>
    physicsBodyDefinitions [{
      sphereCollisionRadius=0
      boundBone {boneName=hairStrand_a_04_r}
      boxExtents=<20 20 15>
      constraintSetup {angularLimitsMax=<1 15 30>, angularLimitsMin=<-1 -15 0>, angularTarget=<1 0 0>}
      localJointOffset=<2 0 0>
     }]
    relativeSpaceBone {boneName=root}
   }
   n57 = Anim Dynamics - Bone: hairStrand_02_l(ComponentPose=@n56, Alpha=1, ExternalForce="0,0,0") {
    angularDampingOverride=0.98
    angularSpringConstant=75
    bAngularSpring=true
    bGravityOverrideInSimSpace=true
    bOverrideAngularDamping=true
    simulationSpace=RootRelative
    alphaBoolBlend {blendOption=Linear}
    boundBone {boneName=hairStrand_02_l}
    componentAppliedLinearAccClamp=<3500 3500 3500>
    componentLinearAccScale=<1 1 1>
    componentLinearVelScale=<1 1 1>
    physicsBodyDefinitions [{
      sphereCollisionRadius=0
      boundBone {boneName=hairStrand_02_l}
      boxExtents=<20 20 15>
      constraintSetup {angularLimitsMax=<1 15 30>, angularLimitsMin=<-1 -15 0>, angularTarget=<1 0 0>}
      localJointOffset=<2 0 0>
     }]
    relativeSpaceBone {boneName=root}
   }
   n58 = Anim Dynamics - Bone: hairStrand_03_l(ComponentPose=@n57, Alpha=1, ExternalForce="0,0,0") {
    angularDampingOverride=0.98
    angularSpringConstant=75
    bAngularSpring=true
    bGravityOverrideInSimSpace=true
    bOverrideAngularDamping=true
    simulationSpace=RootRelative
    alphaBoolBlend {blendOption=Linear}
    boundBone {boneName=hairStrand_03_l}
    componentAppliedLinearAccClamp=<3500 3500 3500>
    componentLinearAccScale=<1 1 1>
    componentLinearVelScale=<1 1 1>
    physicsBodyDefinitions [{
      sphereCollisionRadius=0
      boundBone {boneName=hairStrand_03_l}
      boxExtents=<20 20 15>
      constraintSetup {angularLimitsMax=<1 15 30>, angularLimitsMin=<-1 -15 0>, angularTarget=<1 0 0>}
      localJointOffset=<2 0 0>
     }]
    relativeSpaceBone {boneName=root}
   }
   n59 = Anim Dynamics - Bone: hairStrand_04_l(ComponentPose=@n58, Alpha=1, ExternalForce="0,0,0") {
    angularDampingOverride=0.98
    angularSpringConstant=75
    bAngularSpring=true
    bGravityOverrideInSimSpace=true
    bOverrideAngularDamping=true
    simulationSpace=RootRelative
    alphaBoolBlend {blendOption=Linear}
    boundBone {boneName=hairStrand_04_l}
    componentAppliedLinearAccClamp=<3500 3500 3500>
    componentLinearAccScale=<1 1 1>
    componentLinearVelScale=<1 1 1>
    physicsBodyDefinitions [{
      sphereCollisionRadius=0
      boundBone {boneName=hairStrand_04_l}
      boxExtents=<20 20 15>
      constraintSetup {angularLimitsMax=<1 15 30>, angularLimitsMin=<-1 -15 0>, angularTarget=<1 0 0>}
      localJointOffset=<2 0 0>
     }]
    relativeSpaceBone {boneName=root}
   }
   n60 = Get AccScaleOverride() -> (AccScaleOverride:Vector)
   n61 = RigidBody(ComponentPose=@n59, OverrideWorldGravity=Get GravityOverride(), ComponentLinearAccScale=@n60, ComponentLinearVelScale=@n60) {
    bClampLinearTranslationLimitToRefPose=true
    bOverrideWorldGravity=true
    overridePhysicsAsset=&F_MED_Dark_elf_FaceAcc_Physics
    simulationSpace=BaseBoneSpace
    alphaBoolBlend {blendOption=Linear}
    baseBoneRef {boneName=spine_05}
    componentAppliedLinearAccClamp=<3000 3000 3000>
    componentLinearAccScale=<0.3 0.3 0.5>
    componentLinearVelScale=<0.3 0.3 0.5>
    overrideWorldGravity {z=-980}
    simSpaceSettings {
     maxAngularAcceleration=200
     maxAngularVelocity=5
     maxLinearAcceleration=1000
     maxLinearVelocity=300
     velocityScaleZ=0
     worldAlpha=1
     externalLinearDragV=<1 1>
    }
   }
   n62 = "Constraint - dyn_hair_bk_01 (P)"(ComponentPose=@n61, Alpha=0.4, ConstraintWeights_0=1) {
    alphaBoolBlend {blendOption=Linear}
    alphaScaleBiasClamp {bInterpResult=true}
    boneToModify {boneName=dyn_hair_bk_01}
    constraintSetup [{targetBone {boneName=spine_05}}]
    constraintWeights [1]
   }
   n63 = "Constraint - dyn_hair_l_01 (P)"(ComponentPose=@n62, Alpha=0.4, ConstraintWeights_0=1) {
    alphaBoolBlend {blendOption=Linear}
    alphaScaleBiasClamp {bInterpResult=true}
    boneToModify {boneName=dyn_hair_l_01}
    constraintSetup [{targetBone {boneName=spine_05}}]
    constraintWeights [1]
   }
   n64 = "Constraint - dyn_hair_r_01 (P)"(ComponentPose=@n63, Alpha=0.4, ConstraintWeights_0=1) {
    alphaBoolBlend {blendOption=Linear}
    alphaScaleBiasClamp {bInterpResult=true}
    boneToModify {boneName=dyn_hair_r_01}
    constraintSetup [{targetBone {boneName=spine_05}}]
    constraintWeights [1]
   }
   n65 = "Constraint - dyn_hair_bk_02 (P)"(ComponentPose=@n64, Alpha=0.4, ConstraintWeights_0=1) {
    alphaBoolBlend {blendOption=Linear}
    alphaScaleBiasClamp {bInterpResult=true}
    boneToModify {boneName=dyn_hair_bk_02}
    constraintSetup [{targetBone {boneName=spine_05}}]
    constraintWeights [1]
   }
   n66 = "Constraint - dyn_hair_bk_03 (P)"(ComponentPose=@n65, Alpha=0.4, ConstraintWeights_0=1) {
    alphaBoolBlend {blendOption=Linear}
    alphaScaleBiasClamp {bInterpResult=true}
    boneToModify {boneName=dyn_hair_bk_03}
    constraintSetup [{targetBone {boneName=spine_05}}]
    constraintWeights [1]
   }
   Output Pose(Result=Component To Local(ComponentPose=@n66)) {layerGroup=None}
   n67 = Control Rig(Source=Copy Pose From Mesh() {bUseAttachedParent=true}) {
    alphaBoolBlend {blendOption=Linear}
    controlRigAssetReference {blueprintRigClass=&/F_MED_Dark_Elf_FaceAcc_CtrlRig.F_MED_Dark_Elf_FaceAcc_CtrlRig_C}
   }
   n68 = "Transform (Modify) Bone - Bone: driver_hairBack_01_r"(ComponentPose=Local To Component(LocalPose=@n67), Alpha=1, Translation="0,0,0", Rotation="0,-55,0", Scale="1,1,1") {
    rotationMode=BMM_Additive
    rotationSpace=BCS_BoneSpace
    alphaBoolBlend {blendOption=Linear}
    boneToModify {boneName=driver_hairBack_01_r}
    rotation {yaw=-55}
   }
   n69 = "Transform (Modify) Bone - Bone: driver_hairBack_01_l"(ComponentPose=@n68, Alpha=1, Translation="0,0,0", Rotation="0,-55,0", Scale="1,1,1") {
    rotationMode=BMM_Additive
    rotationSpace=BCS_BoneSpace
    alphaBoolBlend {blendOption=Linear}
    boneToModify {boneName=driver_hairBack_01_l}
    rotation {yaw=-55}
   }
   n70 = "Transform (Modify) Bone - Bone: hairBack_03_r"(ComponentPose=@n69, Alpha=1, Translation="0,0,0", Rotation="0,15,0", Scale="1,1,1") {
    rotationMode=BMM_Additive
    rotationSpace=BCS_BoneSpace
    alphaBoolBlend {blendOption=Linear}
    boneToModify {boneName=hairBack_03_r}
    rotation {yaw=15}
   }
   n71 = "Transform (Modify) Bone - Bone: hairBack_03_l"(ComponentPose=@n70, Alpha=1, Translation="0,0,0", Rotation="0,15,0", Scale="1,1,1") {
    rotationMode=BMM_Additive
    rotationSpace=BCS_BoneSpace
    alphaBoolBlend {blendOption=Linear}
    boneToModify {boneName=hairBack_03_l}
    rotation {yaw=15}
   }
   n72 = "Transform (Modify) Bone - Bone: hairBack_04_r"(ComponentPose=@n71, Alpha=1, Translation="0,0,0", Rotation="0,30,0", Scale="1,1,1") {
    rotationMode=BMM_Additive
    rotationSpace=BCS_BoneSpace
    alphaBoolBlend {blendOption=Linear}
    boneToModify {boneName=hairBack_04_r}
    rotation {yaw=30}
   }
   n73 = "Transform (Modify) Bone - Bone: hairBack_04_l"(ComponentPose=@n72, Alpha=1, Translation="0,0,0", Rotation="0,30,0", Scale="1,1,1") {
    rotationMode=BMM_Additive
    rotationSpace=BCS_BoneSpace
    alphaBoolBlend {blendOption=Linear}
    boneToModify {boneName=hairBack_04_l}
    rotation {yaw=30}
   }
   n74 = RigidBody!(ComponentPose=@n73, bAlphaBoolEnabled:bool=Get No Backpack() -> (No Backpack:bool), ExternalForce) -> (Pose) {
    alphaInputType=Bool
    bClampLinearTranslationLimitToRefPose=true
    bOverrideWorldGravity=true
    overridePhysicsAsset=&/F_MED_Dark_Elf_PERF_HairPhysics
    simulationSpace=BaseBoneSpace
    alphaBoolBlend {blendOption=Linear}
    baseBoneRef {boneName=root}
    componentAppliedLinearAccClamp=<3000 3000 3000>
    componentLinearAccScale {z=0.475}
    componentLinearVelScale=<0.4 0.4 0.125>
    overrideWorldGravity {z=-980}
   }
   n75 = "Transform (Modify) Bone - Bone: hairStrand_b_01_r"(ComponentPose=@n74, Alpha=1, Translation="0,0,0", Rotation="-15,0,0", Scale="1,1,1") {
    rotationMode=BMM_Additive
    rotationSpace=BCS_BoneSpace
    alphaBoolBlend {blendOption=Linear}
    boneToModify {boneName=hairStrand_b_01_r}
    rotation {pitch=-15}
   }
   n76 = "Transform (Modify) Bone - Bone: hairStrand_a_02_r"(ComponentPose=@n75, Alpha=@n34, Translation="0,0,0", Rotation="0,-20,0", Scale="1,1,1") {
    rotationMode=BMM_Additive
    rotationSpace=BCS_BoneSpace
    alphaBoolBlend {blendOption=Linear}
    boneToModify {boneName=hairStrand_a_02_r}
    rotation {yaw=-20}
   }
   n77 = "Transform (Modify) Bone - Bone: hairStrand_03_l"(ComponentPose=@n76, Alpha=1, Translation="0,0,0", Rotation="0,10,0", Scale="1,1,1") {
    rotationMode=BMM_Additive
    rotationSpace=BCS_BoneSpace
    alphaBoolBlend {blendOption=Linear}
    boneToModify {boneName=hairStrand_03_l}
    rotation {yaw=10}
   }
   n78 = "Transform (Modify) Bone - Bone: hairStrand_a_02_r"(ComponentPose=@n77, Alpha=@n35, Translation="0,0,0", Rotation="0,10,0", Scale="1,1,1") {
    rotationMode=BMM_Additive
    rotationSpace=BCS_BoneSpace
    alphaBoolBlend {blendOption=Linear}
    boneToModify {boneName=hairStrand_a_02_r}
    rotation {yaw=10}
   }
   n79 = Get clavicleL()
   n80 = "Transform (Modify) Bone - Bone: driver_hairBack_01_r"(ComponentPose=@n78, Alpha=@n79, Translation="0,0,0", Rotation="0,-6,0", Scale="1,1,1") {
    rotationMode=BMM_Additive
    rotationSpace=BCS_BoneSpace
    alphaBoolBlend {blendOption=Linear}
    boneToModify {boneName=driver_hairBack_01_r}
    rotation {yaw=-6}
   }
   n81 = "Transform (Modify) Bone - Bone: hairBack_03_r"(ComponentPose=@n80, Alpha=@n79, Translation="0,0,0", Rotation="0,10,0", Scale="1,1,1") {
    rotationMode=BMM_Additive
    rotationSpace=BCS_BoneSpace
    alphaBoolBlend {blendOption=Linear}
    boneToModify {boneName=hairBack_03_r}
    rotation {yaw=10}
   }
   n82 = Get clavicleR()
   n83 = "Transform (Modify) Bone - Bone: hairBack_03_l"(ComponentPose=@n81, Alpha=@n82, Translation="0,0,0", Rotation="0,6,0", Scale="1,1,1") {
    rotationMode=BMM_Additive
    rotationSpace=BCS_BoneSpace
    alphaBoolBlend {blendOption=Linear}
    boneToModify {boneName=hairBack_03_l}
    rotation {yaw=6}
   }
   n84 = "Transform (Modify) Bone - Bone: driver_hairBack_01_l"(ComponentPose=@n83, Alpha=@n82, Translation="0,0,0", Rotation="0,-10,0", Scale="1,1,1") {
    rotationMode=BMM_Additive
    rotationSpace=BCS_BoneSpace
    alphaBoolBlend {blendOption=Linear}
    boneToModify {boneName=driver_hairBack_01_l}
    rotation {yaw=-10}
   }
   n85 = Get aimDown() -> (aimDown:double)
   n86 = "Transform (Modify) Bone - Bone: hairBack_02_r"(ComponentPose=@n84, Alpha=@n85, Translation="0,0,0", Rotation="0,12,0", Scale="1,1,1") {
    rotationMode=BMM_Additive
    rotationSpace=BCS_BoneSpace
    alphaBoolBlend {blendOption=Linear}
    boneToModify {boneName=hairBack_02_r}
    rotation {yaw=12}
   }
   n87 = "Transform (Modify) Bone - Bone: hairBack_02_l"(ComponentPose=@n86, Alpha=@n85, Translation="0,0,0", Rotation="0,12,0", Scale="1,1,1") {
    rotationMode=BMM_Additive
    rotationSpace=BCS_BoneSpace
    alphaBoolBlend {blendOption=Linear}
    boneToModify {boneName=hairBack_02_l}
    rotation {yaw=12}
   }
   n88 = "Transform (Modify) Bone - Bone: hairBack_04_r"(ComponentPose=@n87, Alpha=@n85, Translation="0,0,0", Rotation="0,-12,0", Scale="1,1,1") {
    rotationMode=BMM_Additive
    rotationSpace=BCS_BoneSpace
    alphaBoolBlend {blendOption=Linear}
    boneToModify {boneName=hairBack_04_r}
    rotation {yaw=-12}
   }
   n89 = "Transform (Modify) Bone - Bone: hairBack_04_l"(ComponentPose=@n88, Alpha=@n85, Translation="0,0,0", Rotation="0,-12,0", Scale="1,1,1") {
    rotationMode=BMM_Additive
    rotationSpace=BCS_BoneSpace
    alphaBoolBlend {blendOption=Linear}
    boneToModify {boneName=hairBack_04_l}
    rotation {yaw=-12}
   }
   n90 = "Transform (Modify) Bone - Bone: driver_hairStrand_01_l"(
    ComponentPose=Local To Component()
    Alpha=Get Spine 01 Idle Alpha Inv() -> (Spine 01 Idle Alpha Inv:double)
    Translation="0,0,0"
    Rotation="0,15,0"
    Scale="1,1,1"
   ) {
    rotationMode=BMM_Additive
    rotationSpace=BCS_BoneSpace
    alphaBoolBlend {blendOption=Linear}
    boneToModify {boneName=driver_hairStrand_01_l}
    rotation {yaw=15}
   }
   n91 = Get Spine 01 Idle Alpha() -> (Spine 01 Idle Alpha:double)
   n92 = "Transform (Modify) Bone - Bone: driver_hairStrand_b_01_r"(ComponentPose=@n90, Alpha=@n91, Translation="0,0,0", Rotation="0,35,0", Scale="1,1,1") {
    rotationMode=BMM_Additive
    rotationSpace=BCS_BoneSpace
    alphaBoolBlend {blendOption=Linear}
    boneToModify {boneName=driver_hairStrand_b_01_r}
    rotation {yaw=35}
   }
   n93 = "Transform (Modify) Bone - Bone: driver_hairStrand_01_l"(ComponentPose=@n92, Alpha=@n91, Translation="0,0,0", Rotation="0,35,0", Scale="1,1,1") {
    rotationMode=BMM_Additive
    rotationSpace=BCS_BoneSpace
    alphaBoolBlend {blendOption=Linear}
    boneToModify {boneName=driver_hairStrand_01_l}
    rotation {yaw=35}
   }
   n94 = RigidBody(ComponentPose=@n93, bAlphaBoolEnabled=Get No Backpack()) {
    alphaInputType=Bool
    bClampLinearTranslationLimitToRefPose=true
    bOverrideWorldGravity=true
    overridePhysicsAsset=&/F_MED_Dark_Elf_FrontEnd_PERF_HairPhysics
    simulationSpace=BaseBoneSpace
    alphaBoolBlend {blendOption=Linear}
    baseBoneRef {boneName=root}
    componentAppliedLinearAccClamp=<3000 3000 3000>
    componentLinearAccScale {z=0.475}
    componentLinearVelScale=<0.4 0.4 0.125>
    overrideWorldGravity {z=-980}
   }
   n95 = "Transform (Modify) Bone - Bone: driver_hairBack_01_r"(ComponentPose=@n94, Alpha=@n91, Translation="0,0,0", Rotation="0,15,0", Scale="1,1,1") {
    rotationMode=BMM_Additive
    rotationSpace=BCS_BoneSpace
    alphaBoolBlend {blendOption=Linear}
    boneToModify {boneName=driver_hairBack_01_r}
    rotation {yaw=15}
   }
   n96 = "Transform (Modify) Bone - Bone: hairBack_03_r"(ComponentPose=@n95, Alpha=@n91, Translation="0,0,0", Rotation="-10,-15,0", Scale="1,1,1") {
    rotationMode=BMM_Additive
    rotationSpace=BCS_BoneSpace
    alphaBoolBlend {blendOption=Linear}
    boneToModify {boneName=hairBack_03_r}
    rotation {pitch=-10, yaw=-15}
   }
   n97 = "Transform (Modify) Bone - Bone: hairBack_05_r"(ComponentPose=@n96, Alpha=@n91, Translation="0,0,0", Rotation="-10,-5,0", Scale="1,1,1") {
    rotationMode=BMM_Additive
    rotationSpace=BCS_BoneSpace
    alphaBoolBlend {blendOption=Linear}
    boneToModify {boneName=hairBack_05_r}
    rotation {pitch=-10, yaw=-5}
   }
   Save cached pose 'pre-backpack'(Pose=Blend Poses by bool(BlendPose_0=Component To Local(ComponentPose=@n97), BlendPose_1=Component To Local(ComponentPose=@n89), bActiveValue=Get bIsInFrontEnd()) {blendType=Linear}) {cacheName=pre-backpack}
   Get Wearing Backpack()
   Get backbling_hair()
   Get bIsInFrontEnd()
  ]}
 EventGraph {nodes [
   Event BlueprintUpdateAnimation()
   > n1 = Sequence!(execute) -> (then_0, then_1, then_2:Exec, then_3:Exec)
   n2 = Get Skel_Mesh()
   > Set clavicleR(clavicleR:double=GetAlphaFromReferenceTransformDelta(SkeletalMeshComponent=@n2, BoneName=clavicle_r, BaseName=spine_05, DeltaComponentType=EulerY, InputRange=(X=0,Y=30), TwistAxis="1,0,0")) -> (Output_Get:double)
   > Set clavicleL(clavicleL:double=GetAlphaFromReferenceTransformDelta(SkeletalMeshComponent=@n2, BoneName=clavicle_l, BaseName=spine_05, DeltaComponentType=EulerY, InputRange=(X=0,Y=40), TwistAxis="1,0,0")) -> (Output_Get:double)
   > Set clavicleR_bck(clavicleR_bck:double=GetAlphaFromReferenceTransformDelta(SkeletalMeshComponent=@n2, BoneName=clavicle_r, BaseName=spine_05, InputRange=(X=0,Y=35), TwistAxis="1,0,0")) -> (Output_Get:double)
   > Set clavicleL_bck(clavicleL_bck:double=GetAlphaFromReferenceTransformDelta(SkeletalMeshComponent=@n2, BoneName=clavicle_l, BaseName=spine_05, InputRange=(X=0,Y=-20), TwistAxis="1,0,0")) -> (Output_Get:double)
   > n3 = Branch(Condition=Get bIsTargeting())
   n4 = GetOwnerSkeletalMeshComponent()
   n5 = BreakRotator(InRot=BreakTransform(InTransform=GetDeltaTransformFromRefPose(self=@n4, BoneName=head, BaseName=neck_02)).Rotation)
   n6 = Set headDown(execute=@n7.then, headDown:double=MapRangeClamped(Value=@n5.Yaw, InRangeA=0, InRangeB=18)) -> (Output_Get:double)
   Set aimDown(execute=@n3.then, aimDown:double=@n6) -> (Output_Get:double)
   Set aimDown(execute=@n3.else, aimDown=0)
   n7 = Set headUp(execute=@n1.then_1, headUp:double=MapRangeClamped(Value=@n5.Yaw, InRangeA=0, InRangeB=-18)) -> (Output_Get:double)
   n8 = Gravity Override State Function DEPRECATED(
    execute=@n1.then_2
    CharacterPartAnimInstance=Self-Reference()
    GameTimeInSeconds=GetGameTimeInSeconds()
    GravityOverrideParams=Get Gravity Override Params()
   )
   > Set GravityOverride(GravityOverride=@n8)
   n9 = Branch(execute=@n1.then_3, Condition=CheckForBackbling())
   > Set AccScaleOverride(AccScaleOverride:Vector="0.1,0.1,0.1") -> (Output_Get:Vector)
   > Set isWearAlpha(isWearAlpha:double=0.5) -> (Output_Get:double)
   > Set isWearAlpha_2(isWearAlpha_2:double=0.2) -> (Output_Get:double)
   Set AccScaleOverride(execute=@n9.else, AccScaleOverride="0.3,0.3,0.5")
   > Set isWearAlpha(isWearAlpha=0)
   > Set isWearAlpha_2(isWearAlpha_2=0)
   Event OnCharacterCustomizationLoaded()
   > Delay(Duration=0.2)
   > ResetDynamics(InTeleportType=TeleportPhysics)
   > ResetDynamics(InTeleportType=ResetPhysics)
   n10 = CheckForBackbling()
   > n11 = Set Wearing Backpack(Wearing Backpack:bool=@n10) -> (Output_Get:bool)
   > Set No Backpack(No Backpack:bool=NOT Boolean(A=@n11)) -> (Output_Get:bool)
   > Set backbling_hair(backbling_hair:double="To Float (Boolean)"(InBool=@n11)) -> (Output_Get:double)
   Event BlueprintInitializeAnimation()
   > Set Skel_Mesh(Skel_Mesh=GetOwnerSkeletalMeshComponent())
   Branch(Condition=Get bIsInFrontEnd())
   n12 = GetOwnerSkeletalMeshComponent()
   n13 = BreakRotator(InRot=BreakTransform(InTransform=GetDeltaTransformFromRefPose(self=@n12, BoneName=pelvis, BaseName=root)).Rotation)
   > Set Spine 01 Idle Alpha(Spine 01 Idle Alpha:double=MapRangeClamped(Value=@n13.Pitch, InRangeA=0, InRangeB=7)) -> (Output_Get:double)
   > Set Spine 01 Idle Alpha Inv(Spine 01 Idle Alpha Inv:double=MapRangeClamped(Value=@n13.Pitch, InRangeA=7, InRangeB=0)) -> (Output_Get:double)
   n14 = GetDeltaTransformFromRefPose(self=GetOwnerSkeletalMeshComponent(), BoneName=clavicle_r, BaseName=spine_05)
   MapRangeClamped(Value=BreakRotator(InRot=BreakTransform(InTransform=@n14).Rotation).Yaw, InRangeA=-12, InRangeB=10)
   n15 = GetDeltaTransformFromRefPose(self=GetOwnerSkeletalMeshComponent(), BoneName=clavicle_l, BaseName=spine_05)
   MapRangeClamped(Value=BreakRotator(InRot=BreakTransform(InTransform=@n15).Rotation).Yaw, InRangeA=-6, InRangeB=8)
   AND Boolean(A=@n10, B=NOT Boolean(A=@n10.bIsBackblingACape))
  ]}
]
variables [
 AccScaleOverride {type=Vector}
 aimDown {type=number}
 backbling_hair {type=number}
 clavicleL {type=number}
 clavicleL_bck {type=number}
 clavicleR {type=number}
 clavicleR_bck {type=number}
 Gravity Override Params {type=GravityOverrideParamsStruct}
 GravityOverride {type=Vector}
 headDown {type=number}
 headUp {type=number}
 isWearAlpha {type=number}
 isWearAlpha_2 {type=number}
 No Backpack {type=boolean}
 Skel_Mesh {type=SkeletalMeshComponent}
 Spine 01 Idle Alpha {type=number}
 Spine 01 Idle Alpha Inv {type=number}
 Wearing Backpack {type=boolean}
]
